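# Control Rig for the female medium Chainmail body cosmetic, applying corrective adjustments to its shirt, sleeve, necklace and belt bones.
# Spherical pose readers on head, arms, clavicles and thighs write alpha variables, with look-up and both pistol-aim alphas multiplied and curve-remapped.
# These alphas weight bone offsets, some through AlphaInterp, alongside fixed scaling of shirt and upper-arm bones and aim-constrained upper-arm deform bones.
# A stationary-crouch branch also shifts left shirt bones by spine-down alpha.
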ControlRigBlueprint /BRCosmetics/Characters/Player/Female/Medium/Bodies/F_MED_Chainmail/Meshes/F_MED_Chainmail_CtrlRig
graphs [
 ContainedGraph {nodes [
   n1 = RigVMFunctionEntryNode()
   SequenceAggregate_1 = &RigVMFunction_Sequence(ExecuteContext=@n1.ExecuteContext)
   SequenceAggregate_1_1 = &RigVMFunction_Sequence(ExecuteContext=@SequenceAggregate_1.B)
   SequenceAggregate_1_2 = &RigVMFunction_Sequence(ExecuteContext=@SequenceAggregate_1_1.B)
   SequenceAggregate_1_3 = &RigVMFunction_Sequence(ExecuteContext=@SequenceAggregate_1_2.B)
   SequenceAggregate_1_4 = &RigVMFunction_Sequence(ExecuteContext=@SequenceAggregate_1_3.B)
   SequenceAggregate_1_5 = &RigVMFunction_Sequence(ExecuteContext=@SequenceAggregate_1_4.B)
   SequenceAggregate_1_6 = &RigVMFunction_Sequence(ExecuteContext=@SequenceAggregate_1_5.B)
   SequenceAggregate_1_7 = &RigVMFunction_Sequence(ExecuteContext=@SequenceAggregate_1_6.B)
   RigVMFunctionReturnNode(
    A=@SequenceAggregate_1.A
    B=@SequenceAggregate_1_1.A
    C=@SequenceAggregate_1_2.A
    D=@SequenceAggregate_1_3.A
    E=@SequenceAggregate_1_4.A
    F=@SequenceAggregate_1_5.A
    G=@SequenceAggregate_1_6.A
    H=@SequenceAggregate_1_7.A
    I=@SequenceAggregate_1_7.B
   )
  ]}
 ContainedGraph {nodes [
   n1 = RigVMFunctionEntryNode()
   SequenceAggregate = &RigVMFunction_Sequence(ExecuteContext=@n1.ExecuteContext)
   SequenceAggregate_1 = &RigVMFunction_Sequence(ExecuteContext=@SequenceAggregate.B)
   SequenceAggregate_2 = &RigVMFunction_Sequence(ExecuteContext=@SequenceAggregate_1.B)
   SequenceAggregate_3 = &RigVMFunction_Sequence(ExecuteContext=@SequenceAggregate_2.B)
   SequenceAggregate_4 = &RigVMFunction_Sequence(ExecuteContext=@SequenceAggregate_3.B)
   SequenceAggregate_5 = &RigVMFunction_Sequence(ExecuteContext=@SequenceAggregate_4.B)
   RigVMFunctionReturnNode(
    A=@SequenceAggregate.A
    B=@SequenceAggregate_1.A
    C=@SequenceAggregate_2.A
    D=@SequenceAggregate_3.A
    E=@SequenceAggregate_4.A
    F=@SequenceAggregate_5.A
    G=@SequenceAggregate_5.B
   )
  ]}
 ContainedGraph {nodes [
   n1 = RigVMFunctionEntryNode()
   SequenceAggregate_4 = &RigVMFunction_Sequence(ExecuteContext=@n1.ExecuteContext)
   SequenceAggregate_4_1 = &RigVMFunction_Sequence(ExecuteContext=@SequenceAggregate_4.B)
   RigVMFunctionReturnNode(A=@SequenceAggregate_4.A, B=@SequenceAggregate_4_1.A, C=@SequenceAggregate_4_1.B)
  ]}
 ContainedGraph {nodes [
   n1 = RigVMFunctionEntryNode()
   SequenceAggregate_5 = &RigVMFunction_Sequence(ExecuteContext=@n1.ExecuteContext)
   SequenceAggregate_5_1 = &RigVMFunction_Sequence(ExecuteContext=@SequenceAggregate_5.B)
   SequenceAggregate_5_2 = &RigVMFunction_Sequence(ExecuteContext=@SequenceAggregate_5_1.B)
   RigVMFunctionReturnNode(A=@SequenceAggregate_5.A, B=@SequenceAggregate_5_1.A, C=@SequenceAggregate_5_2.A, D=@SequenceAggregate_5_2.B)
  ]}
 RigVMFunctionLibrary {links [], nodes []}
 RigVMModel {nodes [
   RigUnit_BeginExecution = &RigUnit_BeginExecution()
   SequenceAggregate = RigVMAggregateNode(ExecuteContext=@RigUnit_BeginExecution.ExecutePin)
   &RigUnit_ModifyTransforms(
    ItemToModify=((Item=(Type=Bone,Name="dfrm_shirt_scale"),Transform=(Rotation=(X=0,Y=0,Z=0,W=1),Translation=(X=0,Y=0,Z=0),Scale3D=(X=0.75,Y=0.75,Z=1))))
    Weight=1
    ExecutePin=@SequenceAggregate.A
   )
   > &RigUnit_ModifyTransforms(
    ItemToModify=((Item=(Type=Bone,Name="dfrm_body_upperarm_scale_l"),Transform=(Rotation=(X=0,Y=0,Z=0,W=1),Translation=(X=0,Y=0,Z=0),Scale3D=(X=1,Y=0.8,Z=0.8))))
    Weight=1
   )
   > &RigUnit_ModifyTransforms(
    ItemToModify=((Item=(Type=Bone,Name="dfrm_body_upperarm_scale_r"),Transform=(Rotation=(X=0,Y=0,Z=0,W=1),Translation=(X=0,Y=0,Z=0),Scale3D=(X=1,Y=0.8,Z=0.8))))
    Weight=1
   )
   > &RigUnit_ModifyTransforms(
    ItemToModify=((Item=(Type=Bone,Name="dfrm_shirt_upperarm_scale_l"),Transform=(Rotation=(X=0,Y=0,Z=0,W=1),Translation=(X=0,Y=0,Z=0),Scale3D=(X=1,Y=0.75,Z=0.75))))
    Weight=1
   )
   > &RigUnit_ModifyTransforms(
    ItemToModify=((Item=(Type=Bone,Name="dfrm_shirt_upperarm_scale_r"),Transform=(Rotation=(X=0,Y=0,Z=0,W=1),Translation=(X=0,Y=0,Z=0),Scale3D=(X=1,Y=0.75,Z=0.75))))
    Weight=1
   )
   SequenceAggregate_1 = RigVMAggregateNode(ExecuteContext=@SequenceAggregate.B)
   n1 = &RigUnit_SphericalPoseReader(
    ActiveRegionScaleFactors=(PositiveWidth=2,NegativeWidth=0.01,PositiveHeight=2,NegativeHeight=2)
    ActiveRegionSize=0.1
    Debug=(bDrawDebug=false,DebugScale=25,DebugSegments=20,DebugThickness=0.25)
    DriverAxis=(X=0,Y=1,Z=0)
    DriverItem=(Type=Bone,Name="head")
    FalloffRegionScaleFactors=(PositiveWidth=1,NegativeWidth=0.3,PositiveHeight=1,NegativeHeight=1)
    FalloffSize=0.2
    OptionalParentItem=(Type=Bone)
    RotationOffset=(X=90,Y=0,Z=-12.5)
    ExecutePin=@SequenceAggregate_1.A
   )
   > RigVMVariableNode(Value=@n1.OutputParam) {variable=spr_look_up_alpha}
   n2 = &RigUnit_SphericalPoseReader(
    ActiveRegionScaleFactors=(PositiveWidth=1,NegativeWidth=1,PositiveHeight=1,NegativeHeight=1)
    ActiveRegionSize=0.05
    Debug=(bDrawDebug=false,DebugScale=25,DebugSegments=20,DebugThickness=0.25)
    DriverAxis=(X=-1,Y=0,Z=0)
    DriverItem=(Type=Bone,Name="upperarm_r")
    FalloffRegionScaleFactors=(PositiveWidth=1,NegativeWidth=1,PositiveHeight=1,NegativeHeight=1)
    FalloffSize=0.15
    OptionalParentItem=(Type=Bone)
    RotationOffset=(X=-82.5,Y=0,Z=-15)
    ExecutePin=@SequenceAggregate_1.B
   )
   > RigVMVariableNode(Value=@n2.OutputParam) {variable=spr_upperarm_r_pistol_aim_alpha}
   n3 = &RigUnit_SphericalPoseReader(
    ActiveRegionScaleFactors=(PositiveWidth=3.5,NegativeWidth=0.5,PositiveHeight=3,NegativeHeight=3)
    ActiveRegionSize=0.1
    Debug=(bDrawDebug=false,DebugScale=25,DebugSegments=20,DebugThickness=0.25)
    DriverAxis=(X=-1,Y=0,Z=0)
    DriverItem=(Type=Bone,Name="thigh_l")
    FalloffRegionScaleFactors=(PositiveWidth=1,NegativeWidth=1,PositiveHeight=1,NegativeHeight=1)
    FalloffSize=0.45
    OptionalParentItem=(Type=Bone)
    RotationOffset=(X=90,Y=0,Z=0)
    ExecutePin=@SequenceAggregate_1.D
   )
   > RigVMVariableNode(Value=@n3.OutputParam) {variable=spr_thigh_l_up_alpha}
   n4 = &RigUnit_SphericalPoseReader(
    ActiveRegionScaleFactors=(PositiveWidth=1,NegativeWidth=1,PositiveHeight=1,NegativeHeight=1)
    ActiveRegionSize=0.1
    Debug=(bDrawDebug=false,DebugScale=25,DebugSegments=20,DebugThickness=0.25)
    DriverAxis=(X=1,Y=0,Z=0)
    DriverItem=(Type=Bone,Name="upperarm_l")
    FalloffRegionScaleFactors=(PositiveWidth=1,NegativeWidth=1,PositiveHeight=1,NegativeHeight=1)
    FalloffSize=0.2
    OptionalParentItem=(Type=Bone)
    RotationOffset=(X=60,Y=0,Z=-32.5)
    ExecutePin=@SequenceAggregate_1.C
   )
   > RigVMVariableNode(Value=@n4.OutputParam) {variable=spr_upperarm_l_pistol_aim_alpha}
   n5 = &RigUnit_SphericalPoseReader(
    ActiveRegionScaleFactors=(PositiveWidth=3.5,NegativeWidth=0.5,PositiveHeight=3,NegativeHeight=3)
    ActiveRegionSize=0.1
    Debug=(bDrawDebug=false,DebugScale=25,DebugSegments=20,DebugThickness=0.25)
    DriverAxis=(X=1,Y=0,Z=0)
    DriverItem=(Type=Bone,Name="thigh_r")
    FalloffRegionScaleFactors=(PositiveWidth=1,NegativeWidth=1,PositiveHeight=1,NegativeHeight=1)
    FalloffSize=0.45
    FlipHeightScaling=false
    FlipWidthScaling=true
    OptionalParentItem=(Type=Bone)
    RotationOffset=(X=-90,Y=0,Z=0)
    ExecutePin=@SequenceAggregate_1.E
   )
   > RigVMVariableNode(Value=@n5.OutputParam) {variable=spr_thigh_r_up_alpha}
   n6 = &RigUnit_SphericalPoseReader(
    ActiveRegionScaleFactors=(PositiveWidth=1,NegativeWidth=1,PositiveHeight=1,NegativeHeight=1)
    ActiveRegionSize=0.1
    Debug=(bDrawDebug=false,DebugScale=25,DebugSegments=20,DebugThickness=0.25)
    DriverAxis=(X=0,Y=0,Z=-1)
    DriverItem=(Type=Bone,Name="clavicle_r")
    FalloffRegionScaleFactors=(PositiveWidth=1,NegativeWidth=1,PositiveHeight=1,NegativeHeight=1)
    FalloffSize=0.25
    OptionalParentItem=(Type=Bone)
    RotationOffset=(X=135,Y=0,Z=0)
    ExecutePin=@SequenceAggregate_1.F
   )
   > RigVMVariableNode(Value=@n6.OutputParam) {variable=spr_clavicle_r_twist_alpha}
   Multiply = &RigVMFunction_MathDoubleMul(B=RigVMVariableNode() {variable=spr_upperarm_l_pistol_aim_alpha}.Value, A=RigVMVariableNode() {variable=spr_look_up_alpha}.Value)
   n7 = &RigVMFunction_AnimEvalRichCurve(
    Curve=(EditorCurveData=(Keys=((InterpMode=RCIM_Cubic,TangentMode=RCTM_User,ArriveTangent=2.6526,LeaveTangent=2.6526),(InterpMode=RCIM_Cubic,Time=1,Value=1))))
    Value=&RigVMFunction_MathDoubleMul(A=@Multiply.Result, B=RigVMVariableNode() {variable=spr_upperarm_r_pistol_aim_alpha}.Value).Result
   )
   RigVMVariableNode(Value=@n7.Result, ExecuteContext=@SequenceAggregate_1.I) {variable=look_up_with_pistol_mult_alpha}
   n8 = &RigUnit_SphericalPoseReader(
    ActiveRegionScaleFactors=(PositiveWidth=1,NegativeWidth=1,PositiveHeight=1,NegativeHeight=1)
    ActiveRegionSize=0.1
    Debug=(bDrawDebug=false,DebugScale=25,DebugSegments=20,DebugThickness=0.25)
    DriverAxis=(X=1,Y=0,Z=0)
    DriverItem=(Type=Bone,Name="clavicle_l")
    FalloffRegionScaleFactors=(PositiveWidth=1,NegativeWidth=1,PositiveHeight=1,NegativeHeight=1)
    FalloffSize=0.2
    OptionalParentItem=(Type=Bone)
    RotationOffset=(X=15,Y=-46,Z=0)
    ExecutePin=@SequenceAggregate_1.G
   )
   > RigVMVariableNode(Value=@n8.OutputParam) {variable=spr_clavicle_l_up}
   SequenceAggregate_4 = RigVMAggregateNode(ExecuteContext=@SequenceAggregate.D)
   &RigUnit_ModifyTransforms(
    ItemToModify=((Item=(Type=Bone,Name="dfrm_clavicle_r"),Transform=(Rotation=(X=0,Y=0,Z=0,W=1),Translation=(X=0,Y=0,Z=0),Scale3D=(X=1,Y=1,Z=1))))
    ItemToModify.0.Transform.Rotation=&RigVMFunction_MathQuaternionFromEuler(Euler=(X=-30,Y=0,Z=0)).Result
    ExecutePin=@SequenceAggregate_4.A
    Weight=RigVMVariableNode() {variable=spr_clavicle_r_twist_alpha}.Value
   )
   &RigUnit_ModifyTransforms(
    ItemToModify=((Item=(Type=Bone,Name="dyn_util_belt_l"),Transform=(Rotation=(X=0,Y=0,Z=0,W=1),Translation=(X=0,Y=0,Z=1),Scale3D=(X=1,Y=1,Z=1))))
    ItemToModify.0.Transform.Rotation=&RigVMFunction_MathQuaternionFromEuler(Euler=(X=0,Y=0,Z=0)).Result
    Weight=RigVMVariableNode() {variable=spr_thigh_l_up_alpha}.Value
    ExecutePin=@SequenceAggregate_4.C
   )
   /Script/ControlRig.RigUnit_AimItem(
    DebugSettings=(Scale=10,WorldOffset=(Rotation=(X=0,Y=0,Z=0,W=1),Translation=(X=0,Y=0,Z=0),Scale3D=(X=1,Y=1,Z=1)))
    Item=(Type=Bone,Name="dfrm_upperarm_l")
    Primary=(Weight=1,Axis=(X=1,Y=0,Z=0),Target=(X=0,Y=0,Z=0),Kind=Location,Space=(Type=Bone,Name="lowerarm_l"))
    Secondary=(Weight=1,Axis=(X=0,Y=-1,Z=0),Target=(X=0,Y=0,Z=0),Kind=Location,Space=(Type=Bone,Name="dfrm_aim_upperarm_l"))
    Weight=1
    ExecutePin=@SequenceAggregate_4.B
   ) {is=struct RigVMUnitNode}
   > &RigUnit_AimItem(
    DebugSettings=(Scale=10,WorldOffset=(Rotation=(X=0,Y=0,Z=0,W=1),Translation=(X=0,Y=0,Z=0),Scale3D=(X=1,Y=1,Z=1)))
    Item=(Type=Bone,Name="dfrm_upperarm_r")
    Primary=(Weight=1,Axis=(X=1,Y=0,Z=0),Target=(X=0,Y=0,Z=0),Kind=Location,Space=(Type=Bone,Name="lowerarm_r"))
    Secondary=(Weight=1,Axis=(X=0,Y=-1,Z=0),Target=(X=0,Y=0,Z=0),Kind=Location,Space=(Type=Bone,Name="dfrm_aim_upperarm_r"))
    Weight=1
   )
   SequenceAggregate_5 = RigVMAggregateNode(ExecuteContext=@SequenceAggregate.G)
   n9 = &RigVMFunction_AlphaInterp(
    InterpSpeedDecreasing=10
    InterpSpeedIncreasing=10
    Scale=1
    bClampResult=False
    bInterpResult=False
    Value=RigVMVariableNode() {variable=spr_thigh_l_up_alpha}.Value
   )
   &RigUnit_ModifyTransforms(
    ItemToModify=((Item=(Type=Bone,Name="dyn_shirt_front_mstr_l"),Transform=(Rotation=(X=0,Y=0,Z=0,W=1),Translation=(X=0,Y=0,Z=0),Scale3D=(X=1,Y=1,Z=1))))
    ItemToModify.0.Transform.Rotation=&RigVMFunction_MathQuaternionFromEuler(Euler=(X=15,Y=0,Z=-35)).Result
    ExecutePin=@SequenceAggregate_5.A
    Weight=@n9.Result
   )
   > &RigUnit_ModifyTransforms(
    ItemToModify=((Item=(Type=Bone,Name="dyn_shirt_front_side_mstr_l"),Transform=(Rotation=(X=0,Y=0,Z=0,W=1),Translation=(X=0,Y=0,Z=0),Scale3D=(X=1,Y=1,Z=1))))
    ItemToModify.0.Transform.Rotation=&RigVMFunction_MathQuaternionFromEuler(Euler=(X=10,Y=0,Z=-10)).Result
    Weight=@n9.Result
   )
   > &RigUnit_ModifyTransforms(
    ItemToModify=((Item=(Type=Bone,Name="dyn_shirt_side_mstr_l"),Transform=(Rotation=(X=0,Y=0,Z=0,W=1),Translation=(X=0,Y=0,Z=0),Scale3D=(X=1,Y=1,Z=1))))
    ItemToModify.0.Transform.Rotation=&RigVMFunction_MathQuaternionFromEuler(Euler=(X=20,Y=0,Z=-15)).Result
    Weight=@n9.Result
   )
   > &RigUnit_ModifyTransforms(
    ItemToModify=((Item=(Type=Bone,Name="dyn_shirt_back_side_mstr_l"),Transform=(Rotation=(X=0,Y=0,Z=0,W=1),Translation=(X=0,Y=0,Z=0),Scale3D=(X=1,Y=1,Z=1))))
    ItemToModify.0.Transform.Rotation=&RigVMFunction_MathQuaternionFromEuler(Euler=(X=2.5,Y=5,Z=-10)).Result
    Weight=@n9.Result
   )
   > &RigUnit_ModifyTransforms(
    ItemToModify=((Item=(Type=Bone,Name="dyn_shirt_back_mstr"),Transform=(Rotation=(X=0,Y=0,Z=0,W=1),Translation=(X=0,Y=0,Z=0),Scale3D=(X=1,Y=1,Z=1))))
    ItemToModify.0.Transform.Rotation=&RigVMFunction_MathQuaternionFromEuler(Euler=(X=0,Y=0,Z=0)).Result
    Weight=@n9.Result
   )
   n10 = &RigVMFunction_ControlFlowBranch(Condition=RigVMVariableNode() {variable=is_moving_2d}.Value, ExecuteContext=@SequenceAggregate_5.B)
   n11 = &RigVMFunction_ControlFlowBranch(ExecuteContext=@n10.False, Condition=RigVMVariableNode() {variable=is_crouching}.Value)
   &RigUnit_ModifyTransforms(
    ItemToModify=((Item=(Type=Bone,Name="dyn_shirt_front_mstr_l"),Transform=(Rotation=(X=0,Y=0,Z=0,W=1),Translation=(X=2,Y=-3.5,Z=-6),Scale3D=(X=1,Y=1,Z=1))))
    ExecutePin=@n11.True
    ItemToModify.0.Transform.Rotation=&RigVMFunction_MathQuaternionFromEuler(Euler=(X=0,Y=20,Z=0)).Result
    Weight=RigVMVariableNode() {variable=spine_rotate_down_alpha}.Value
   )
   > &RigUnit_ModifyTransforms(
    ItemToModify=((Item=(Type=Bone,Name="dyn_shirt_side_mstr_l"),Transform=(Rotation=(X=0,Y=0,Z=0,W=1),Translation=(X=1,Y=-1.75,Z=-3),Scale3D=(X=1,Y=1,Z=1))))
    ItemToModify.0.Transform.Rotation=&RigVMFunction_MathQuaternionFromEuler(Euler=(X=0,Y=10,Z=0)).Result
    Weight=RigVMVariableNode() {variable=spine_rotate_down_alpha}.Value
   )
   > &RigUnit_ModifyTransforms(
    ItemToModify=((Item=(Type=Bone,Name="dyn_shirt_back_side_mstr_l"),Transform=(Rotation=(X=0,Y=0,Z=0,W=1),Translation=(X=0,Y=-1,Z=-1.5),Scale3D=(X=1,Y=1,Z=1))))
    Weight=RigVMVariableNode() {variable=spine_rotate_down_alpha}.Value
   )
   RerouteNode_18 = RigVMDispatchNode(Elements=(X=0.8,Y=0.8,Z=0.8)) {template="DISPATCH_RigVMDispatch_MakeStruct(in Elements,out Struct)"}
   &RigUnit_ModifyTransforms(
    ItemToModify=((Item=(Type=Bone,Name="dfrm_sphere_necklace_08"),Transform=(Rotation=(X=0,Y=0,Z=0,W=1),Translation=(X=1,Y=1,Z=-1),Scale3D=(X=0.8,Y=0.8,Z=0.8))))
    ItemToModify.0.Transform.Scale3D=@RerouteNode_18.Struct
    Weight=RigVMVariableNode() {variable=look_up_with_pistol_mult_alpha}.Value
    ExecutePin=@SequenceAggregate.C
   )
   > &RigUnit_ModifyTransforms(
    ItemToModify=((Item=(Type=Bone,Name="dfrm_sphere_necklace_09"),Transform=(Rotation=(X=0,Y=0,Z=0,W=1),Translation=(X=1,Y=1,Z=-1),Scale3D=(X=0.7,Y=0.7,Z=0.7))))
    Weight=RigVMVariableNode() {variable=look_up_with_pistol_mult_alpha}.Value
    ItemToModify.0.Transform.Scale3D=@RerouteNode_18.Struct
   )
   > &RigUnit_ModifyTransforms(
    ItemToModify=((Item=(Type=Bone,Name="dfrm_sphere_necklace_10"),Transform=(Rotation=(X=0,Y=0,Z=0,W=1),Translation=(X=2,Y=1,Z=-1),Scale3D=(X=0.7,Y=0.7,Z=0.7))))
    Weight=RigVMVariableNode() {variable=look_up_with_pistol_mult_alpha}.Value
    ItemToModify.0.Transform.Scale3D=@RerouteNode_18.Struct
   )
   n12 = &RigVMFunction_AlphaInterp(
    InterpSpeedDecreasing=10
    InterpSpeedIncreasing=10
    Scale=1
    bClampResult=False
    Value=RigVMVariableNode() {variable=spr_clavicle_l_up}.Value
   )
   > &RigUnit_ModifyTransforms(
    ItemToModify=((Item=(Type=Bone,Name="dfrm_sphere_necklace_03"),Transform=(Rotation=(X=0,Y=0,Z=0,W=1),Translation=(X=0,Y=-0.5,Z=1),Scale3D=(X=1,Y=1,Z=1))))
    Weight=@n12.Result
   )
   > &RigUnit_ModifyTransforms(
    ItemToModify=((Item=(Type=Bone,Name="dfrm_sphere_necklace_04"),Transform=(Rotation=(X=0,Y=0,Z=0,W=1),Translation=(X=1,Y=-0.5,Z=1),Scale3D=(X=1,Y=1,Z=1))))
    Weight=@n12.Result
   )
   > &RigUnit_ModifyTransforms(
    ItemToModify=((Item=(Type=Bone,Name="dfrm_sphere_necklace_05"),Transform=(Rotation=(X=0,Y=0,Z=0,W=1),Translation=(X=2.5,Y=-0.75,Z=0.5),Scale3D=(X=1,Y=1,Z=1))))
    Weight=@n12.Result
   )
   > &RigUnit_ModifyTransforms(
    ItemToModify=((Item=(Type=Bone,Name="dfrm_sphere_necklace_06"),Transform=(Rotation=(X=0,Y=0,Z=0,W=1),Translation=(X=2.25,Y=-0.75,Z=0.75),Scale3D=(X=1,Y=1,Z=1))))
    Weight=@n12.Result
   )
   SequenceAggregate_2 = &RigVMFunction_Sequence(ExecuteContext=@SequenceAggregate.F)
   AlphaInterp = &RigVMFunction_AlphaInterp(
    InterpSpeedDecreasing=10
    InterpSpeedIncreasing=10
    Scale=1
    bClampResult=False
    Value=RigVMVariableNode() {variable=spr_thigh_r_up_alpha}.Value
   )
   &RigUnit_ModifyTransforms(
    ItemToModify=((Item=(Type=Bone,Name="dyn_shirt_front_mstr_r"),Transform=(Rotation=(X=0,Y=0,Z=0,W=1),Translation=(X=0,Y=0,Z=0),Scale3D=(X=1,Y=1,Z=1))))
    ItemToModify.0.Transform.Rotation=&RigVMFunction_MathQuaternionFromEuler(Euler=(X=0,Y=-7.5,Z=-15)).Result
    ExecutePin=@SequenceAggregate_2.A
    Weight=@AlphaInterp.Result
   )
   > &RigUnit_ModifyTransforms(
    ItemToModify=((Item=(Type=Bone,Name="dyn_shirt_front_side_mstr_r"),Transform=(Rotation=(X=0,Y=0,Z=0,W=1),Translation=(X=0,Y=0,Z=0),Scale3D=(X=1,Y=1,Z=1))))
    ItemToModify.0.Transform.Rotation=&RigVMFunction_MathQuaternionFromEuler(Euler=(X=0,Y=-7.5,Z=-15)).Result
    Weight=@AlphaInterp.Result
   )
   > &RigUnit_ModifyTransforms(
    ItemToModify=((Item=(Type=Bone,Name="dyn_shirt_side_mstr_r"),Transform=(Rotation=(X=0,Y=0,Z=0,W=1),Translation=(X=0,Y=0,Z=0),Scale3D=(X=1,Y=1,Z=1))))
    ItemToModify.0.Transform.Rotation=&RigVMFunction_MathQuaternionFromEuler(Euler=(X=0,Y=0,Z=-5)).Result
    Weight=@AlphaInterp.Result
   )
   > &RigUnit_ModifyTransforms(
    ItemToModify=((Item=(Type=Bone,Name="dyn_shirt_back_side_mstr_r"),Transform=(Rotation=(X=0,Y=0,Z=0,W=1),Translation=(X=0,Y=0,Z=0),Scale3D=(X=1,Y=1,Z=1))))
    ItemToModify.0.Transform.Rotation=&RigVMFunction_MathQuaternionFromEuler(Euler=(X=0,Y=0,Z=-2.5)).Result
    Weight=@AlphaInterp.Result
   )
   > &RigUnit_ModifyTransforms(
    ItemToModify=((Item=(Type=Bone,Name="dyn_shirt_back_mstr"),Transform=(Rotation=(X=0,Y=0,Z=0,W=1),Translation=(X=0,Y=0,Z=0),Scale3D=(X=1,Y=1,Z=1))))
    ItemToModify.0.Transform.Rotation=&RigVMFunction_MathQuaternionFromEuler(Euler=(X=0,Y=0,Z=-2.5)).Result
    Weight=@AlphaInterp.Result
   )
   &RigUnit_ModifyTransforms(
    ItemToModify=((Item=(Type=Bone,Name="dyn_shirt_front_02_r"),Transform=(Rotation=(X=0,Y=0,Z=0,W=1),Translation=(X=0,Y=0,Z=0),Scale3D=(X=1,Y=1,Z=1))))
    ItemToModify.0.Transform.Rotation=&RigVMFunction_MathQuaternionFromEuler(Euler=(X=0,Y=0,Z=-7.5)).Result
    ExecutePin=@SequenceAggregate_2.B
    Weight=@AlphaInterp.Result
   )
   > &RigUnit_ModifyTransforms(
    ItemToModify=((Item=(Type=Bone,Name="dyn_shirt_front_side_02_r"),Transform=(Rotation=(X=0,Y=0,Z=0,W=1),Translation=(X=0,Y=0,Z=0),Scale3D=(X=1,Y=1,Z=1))))
    ItemToModify.0.Transform.Rotation=&RigVMFunction_MathQuaternionFromEuler(Euler=(X=0,Y=-7.5,Z=-5)).Result
    Weight=@AlphaInterp.Result
   )
   > &RigUnit_ModifyTransforms(
    ItemToModify=((Item=(Type=Bone,Name="dyn_shirt_side_02_r"),Transform=(Rotation=(X=0,Y=0,Z=0,W=1),Translation=(X=0,Y=0,Z=0),Scale3D=(X=1,Y=1,Z=1))))
    ItemToModify.0.Transform.Rotation=&RigVMFunction_MathQuaternionFromEuler(Euler=(X=0,Y=0,Z=-2.5)).Result
    Weight=@AlphaInterp.Result
   )
   > &RigUnit_ModifyTransforms(
    ItemToModify=((Item=(Type=Bone,Name="dyn_shirt_back_side_02_r"),Transform=(Rotation=(X=0,Y=0,Z=0,W=1),Translation=(X=0,Y=0,Z=0),Scale3D=(X=1,Y=1,Z=1))))
    ItemToModify.0.Transform.Rotation=&RigVMFunction_MathQuaternionFromEuler(Euler=(X=0,Y=0,Z=-2.5)).Result
    Weight=@AlphaInterp.Result
   )
   > &RigUnit_ModifyTransforms(
    ItemToModify=((Item=(Type=Bone,Name="dyn_shirt_back_02"),Transform=(Rotation=(X=0,Y=0,Z=0,W=1),Translation=(X=0,Y=0,Z=0),Scale3D=(X=1,Y=1,Z=1))))
    ItemToModify.0.Transform.Rotation=&RigVMFunction_MathQuaternionFromEuler(Euler=(X=0,Y=0,Z=-2.5)).Result
    Weight=@AlphaInterp.Result
   )
  ]}
]
hierarchy [
 root {loc=<0 0 0>, rot=<0 0 0 1>}
 pelvis {parent=root, loc=<0 -0.5301 98.0571>, rot=<0.0127 0.707 -0.0127 -0.707>}
 spine_01 {parent=pelvis, loc=<6.1208 -0.9328 0>, rot=<0 0 0.034 -0.9994>}
 spine_02 {parent=spine_01, loc=<6.5996 -0.2095 0>, rot=<0 0 0.0238 -0.9997>}
 spine_03 {parent=spine_02, loc=<6.4914 -0.1237 0>, rot=<0 0 0.0149 -0.9999>}
 spine_04 {parent=spine_03, loc=<6.9598 0.0794 0>, rot=<0 0 -0.0132 -0.9999>}
 spine_05 {parent=spine_04, loc=<6.9987 0 0>, rot=<0 0 -0.0015 -1>}
 clavicle_l {parent=spine_05, loc=<5.4831 -0.1053 -3.2946>, rot=<0.0465 -0.7289 -0.009 -0.6829>}
 upperarm_l {parent=clavicle_l, loc=<9.8658 -1.2136 -0.4105>, rot=<0.0164 -0.3245 -0.0269 -0.9454>}
 lowerarm_l {parent=upperarm_l, loc=<26.465 -0.8345 0.259>, rot=<0.0036 0.0161 -0.2298 -0.9731>}
 hand_l {parent=lowerarm_l, loc=<22.7355 -0.3203 -0.6105>, rot=<-0.6784 -0.0223 0.0735 -0.7307>}
 thumb_01_l {parent=hand_l, loc=<1.3118 -0.3191 -2.459>, rot=<0.6546 -0.4042 0.0664 -0.6354>}
 thumb_02_l {parent=thumb_01_l, loc=<5.9336 -0.0448 0.0401>, rot=<-0.0037 0.0213 0.298 -0.9543>}
 thumb_03_l {parent=thumb_02_l, loc=<2.2223 0.3682 -0.0907>, rot=<0.0118 -0.013 0.0066 -0.9998>}
 index_metacarpal_l {parent=hand_l, loc=<1.6535 0.0183 -2.96>, rot=<0.0158 -0.0124 -0.0173 -0.9996>}
 index_01_l {parent=index_metacarpal_l, loc=<8.3008 0.0357 -0.0524>, rot=<-0.0044 0.0485 0.0924 -0.9945>}
 index_02_l {parent=index_01_l, loc=<3.3587 -0.3004 0.0452>, rot=<-0.002 0.0027 0.3225 -0.9466>}
 index_03_l {parent=index_02_l, loc=<2.5578 -0.123 -0.0252>, rot=<-0.002 0.0067 0.2558 -0.9667>}
 middle_metacarpal_l {parent=hand_l, loc=<2.1856 -0.1725 -0.5984>, rot=<0.006 0.0168 -0.0329 -0.9993>}
 middle_01_l {parent=middle_metacarpal_l, loc=<7.3275 -0.2447 0.0129>, rot=<-0.0208 0.0042 0.0858 -0.9961>}
 middle_02_l {parent=middle_01_l, loc=<4.0123 0.1401 0.1175>, rot=<-0.0218 0.0258 0.3512 -0.9357>}
 middle_03_l {parent=middle_02_l, loc=<2.6077 0.0534 0.1373>, rot=<0.0082 0.0188 0.1708 -0.9851>}
 ring_metacarpal_l {parent=hand_l, loc=<2.2074 -0.3625 1.3638>, rot=<-0.0284 0.0722 -0.0458 -0.9959>}
 ring_01_l {parent=ring_metacarpal_l, loc=<7.0295 -0.6197 -0.1446>, rot=<-0.0489 -0.0252 0.1892 -0.9804>}
 ring_02_l {parent=ring_01_l, loc=<3.6174 0.4647 -0.1027>, rot=<0 0 0.2806 -0.9598>}
 ring_03_l {parent=ring_02_l, loc=<2.4152 0.114 0.0199>, rot=<0 0 0.1189 -0.9929>}
 pinky_metacarpal_l {parent=hand_l, loc=<1.7616 -0.9753 3.3814>, rot=<-0.0397 0.0814 -0.069 -0.9935>}
 pinky_01_l {parent=pinky_metacarpal_l, loc=<7.1709 -0.4831 -0.1094>, rot=<-0.0654 -0.0035 0.2242 -0.9723>}
 pinky_02_l {parent=pinky_01_l, loc=<2.8274 0.0893 -0.1043>, rot=<0.0037 0.0011 0.2854 -0.9584>}
 pinky_03_l {parent=pinky_02_l, loc=<2.2046 0.0307 -0.0078>, rot=<-0.0002 0.0098 0.1445 -0.9895>}
 deform_wrist_dn_base_l {parent=hand_l, loc=<0 -1.5326 0>, rot=<0 0 0 -1>}
 deform_wrist_dn_l {parent=deform_wrist_dn_base_l, loc=<0 0 0>, rot=<0 0 0 1>}
 deform_wrist_up_base_l {parent=hand_l, loc=<0 2.3009 0>, rot=<0 0 0 -1>}
 deform_wrist_up_l {parent=deform_wrist_up_base_l, loc=<0 0 0>, rot=<0 0 0 1>}
 lowerarm_twist_02_l {parent=lowerarm_l, loc=<10.6707 0 0>, rot=<0 0 0 1>}
 lowerarm_twist_01_l {parent=lowerarm_l, loc=<18.316 0 0>, rot=<0 0 0 1>}
 deform_elbow_in_base_l {parent=lowerarm_l, loc=<0.1834 3.6062 -0.2572>, rot=<0 0 0 -1>}
 deform_elbow_in_l {parent=deform_elbow_in_base_l, loc=<0 0 0>, rot=<0 0 0 1>}
 deform_elbow_out_base_l {parent=lowerarm_l, loc=<0 -4 0>, rot=<0 0 0 1>}
 deform_elbow_out_l {parent=deform_elbow_out_base_l, loc=<0 0 0>, rot=<0 0 0 1>}
 upperarm_twist_01_l {parent=upperarm_l, loc=<7.2135 -0.1134 0>, rot=<0 0 0 1>}
 deform_upperarm_bk_base_l {parent=upperarm_twist_01_l, loc=<-6.4053 -4.3746 0.314>, rot=<0 0 0 -1>}
 deform_upperarm_bk_l {parent=deform_upperarm_bk_base_l, loc=<0 0 0>, rot=<0 0 0 1>}
 deform_upperarm_fr_base_l {parent=upperarm_twist_01_l, loc=<-5.6312 5.092 0.3354>, rot=<0 0 0 -1>}
 deform_upperarm_fr_l {parent=deform_upperarm_fr_base_l, loc=<0 0 0>, rot=<0 0 0 1>}
 deform_upperarm_up_base_l {parent=upperarm_twist_01_l, loc=<-8.891 0.4825 3.9517>, rot=<0 0 0 -1>}
 deform_upperarm_up_l {parent=deform_upperarm_up_base_l, loc=<0 0 0>, rot=<0 0 0 1>}
 upperarm_twist_02_l {parent=upperarm_l, loc=<19.7998 -0.1134 0>, rot=<0 0 0 1>}
 dfrm_body_upperarm_scale_l {parent=upperarm_l, loc=<4.1441 -0.2805 -0.5276>, rot=<0.7016 0.0038 0.0212 -0.7123>}
 dfrm_clavicle_l {parent=clavicle_l, loc=<0 0 0>, rot=<0 0 0 1>}
 dfrm_aim_upperarm_l {parent=dfrm_clavicle_l, loc=<18.9397 -0.4874 11.5117>, rot=<0.685 -0.2174 -0.2345 -0.6546>}
 dfrm_upperarm_l {parent=dfrm_clavicle_l, loc=<9.8658 -1.2136 -0.4105>, rot=<0.685 -0.2174 -0.2345 -0.6546>}
 dfrm_sleeve_l {parent=dfrm_upperarm_l, loc=<0 0 0>, rot=<0 0 0 1>}
 dfrm_shirt_upperarm_scale_l {parent=dfrm_sleeve_l, loc=<7.6699 1.9198 0.2163>, rot=<0.0178 -0.0066 0.0507 -0.9985>}
 attach_cape {parent=spine_05, loc=<9.3216 -8.4501 0>, rot=<0.0616 -0.7044 -0.0616 -0.7044>}
 attach_backpack {parent=spine_05, loc=<0 -8.35 0>, rot=<-0.0093 -0.707 0.0093 -0.707>}
 neck_01 {parent=spine_05, loc=<11.3138 0.1529 0>, rot=<0 0 -0.2003 -0.9797>}
 neck_02 {parent=neck_01, loc=<4.5627 0 0>, rot=<0 0 0.0504 -0.9987>}
 head {parent=neck_02, loc=<5.3914 0 0>, rot=<0 0 0.115 -0.9934>}
 clavicle_r {parent=spine_05, loc=<5.4831 -0.1053 3.2946>, rot=<0.7289 0.0465 -0.6829 0.009>}
 upperarm_r {parent=clavicle_r, loc=<-9.8658 1.2136 0.4105>, rot=<0.0164 -0.3245 -0.0269 -0.9454>}
 lowerarm_r {parent=upperarm_r, loc=<-26.465 0.8345 -0.259>, rot=<0.0036 0.0161 -0.2298 -0.9731>}
 hand_r {parent=lowerarm_r, loc=<-22.7355 0.3203 0.6106>, rot=<-0.6784 -0.0223 0.0735 -0.7307>}
 thumb_01_r {parent=hand_r, loc=<-1.3118 0.3191 2.459>, rot=<0.6546 -0.4042 0.0664 -0.6354>}
 thumb_02_r {parent=thumb_01_r, loc=<-5.9336 0.0448 -0.0401>, rot=<-0.0037 0.0213 0.298 -0.9543>}
 thumb_03_r {parent=thumb_02_r, loc=<-2.2223 -0.3682 0.0907>, rot=<0.0118 -0.013 0.0066 -0.9998>}
 index_metacarpal_r {parent=hand_r, loc=<-1.6535 -0.0183 2.96>, rot=<0.0158 -0.0124 -0.0173 -0.9996>}
 index_01_r {parent=index_metacarpal_r, loc=<-8.3008 -0.0357 0.0524>, rot=<-0.0044 0.0485 0.0924 -0.9945>}
 index_02_r {parent=index_01_r, loc=<-3.3587 0.3004 -0.0452>, rot=<-0.002 0.0027 0.3225 -0.9466>}
 index_03_r {parent=index_02_r, loc=<-2.5578 0.123 0.0252>, rot=<-0.002 0.0067 0.2558 -0.9667>}
 middle_metacarpal_r {parent=hand_r, loc=<-2.1856 0.1725 0.5984>, rot=<0.006 0.0168 -0.0329 -0.9993>}
 middle_01_r {parent=middle_metacarpal_r, loc=<-7.3275 0.2447 -0.0129>, rot=<-0.0208 0.0042 0.0858 -0.9961>}
 middle_02_r {parent=middle_01_r, loc=<-4.0123 -0.1401 -0.1175>, rot=<-0.0218 0.0258 0.3512 -0.9357>}
 middle_03_r {parent=middle_02_r, loc=<-2.6077 -0.0534 -0.1373>, rot=<0.0082 0.0188 0.1708 -0.9851>}
 ring_metacarpal_r {parent=hand_r, loc=<-2.2074 0.3625 -1.3638>, rot=<-0.0284 0.0722 -0.0458 -0.9959>}
 ring_01_r {parent=ring_metacarpal_r, loc=<-7.0295 0.6197 0.1446>, rot=<-0.0489 -0.0252 0.1892 -0.9804>}
 ring_02_r {parent=ring_01_r, loc=<-3.6174 -0.4647 0.1027>, rot=<0 0 0.2806 -0.9598>}
 ring_03_r {parent=ring_02_r, loc=<-2.4152 -0.114 -0.0199>, rot=<0 0 0.1189 -0.9929>}
 pinky_metacarpal_r {parent=hand_r, loc=<-1.7616 0.9753 -3.3814>, rot=<-0.0397 0.0814 -0.069 -0.9935>}
 pinky_01_r {parent=pinky_metacarpal_r, loc=<-7.1709 0.4831 0.1094>, rot=<-0.0654 -0.0035 0.2242 -0.9723>}
 pinky_02_r {parent=pinky_01_r, loc=<-2.8274 -0.0893 0.1043>, rot=<0.0037 0.0011 0.2854 -0.9584>}
 pinky_03_r {parent=pinky_02_r, loc=<-2.2046 -0.0307 0.0078>, rot=<-0.0002 0.0098 0.1445 -0.9895>}
 deform_wrist_dn_base_r {parent=hand_r, loc=<-0.0001 1.5328 0>, rot=<0 0 0 -1>}
 deform_wrist_dn_r {parent=deform_wrist_dn_base_r, loc=<0 0 0>, rot=<0 0 0 1>}
 deform_wrist_up_base_r {parent=hand_r, loc=<-0.0002 -2.3007 -0.0001>, rot=<0 0 0 -1>}
 deform_wrist_up_r {parent=deform_wrist_up_base_r, loc=<0 0 0>, rot=<0 0 0 1>}
 lowerarm_twist_02_r {parent=lowerarm_r, loc=<-10.6707 0 0>, rot=<0 0 0 -1>}
 lowerarm_twist_01_r {parent=lowerarm_r, loc=<-18.316 0 0>, rot=<0 0 0 -1>}
 deform_elbow_in_base_r {parent=lowerarm_r, loc=<-0.1834 -3.6062 0.2572>, rot=<0 0 0 -1>}
 deform_elbow_in_r {parent=deform_elbow_in_base_r, loc=<0 0 0>, rot=<0 0 0 1>}
 deform_elbow_out_base_r {parent=lowerarm_r, loc=<0 4 -0.0001>, rot=<0 0 0 -1>}
 deform_elbow_out_r {parent=deform_elbow_out_base_r, loc=<0 0 0>, rot=<0 0 0 1>}
 upperarm_twist_01_r {parent=upperarm_r, loc=<-7.2135 0.1134 0>, rot=<0 0 0 -1>}
 deform_upperarm_bk_base_r {parent=upperarm_twist_01_r, loc=<6.4053 4.3746 -0.314>, rot=<0 0 0 -1>}
 deform_upperarm_bk_r {parent=deform_upperarm_bk_base_r, loc=<0 0 0>, rot=<0 0 0 1>}
 deform_upperarm_fr_base_r {parent=upperarm_twist_01_r, loc=<5.6312 -5.092 -0.3354>, rot=<0 0 0 -1>}
 deform_upperarm_fr_r {parent=deform_upperarm_fr_base_r, loc=<0 0 0>, rot=<0 0 0 1>}
 deform_upperarm_up_base_r {parent=upperarm_twist_01_r, loc=<8.891 -0.4825 -3.9517>, rot=<0 0 0 -1>}
 deform_upperarm_up_r {parent=deform_upperarm_up_base_r, loc=<0 0 0>, rot=<0 0 0 1>}
 upperarm_twist_02_r {parent=upperarm_r, loc=<-19.7998 0.1134 0>, rot=<0 0 0 -1>}
 dfrm_body_upperarm_scale_r {parent=upperarm_r, loc=<-4.1441 0.2805 0.5276>, rot=<-0.0038 0.7016 0.7123 0.0212>}
 dfrm_clavicle_r {parent=clavicle_r, loc=<0 0 0>, rot=<0 0 0 1>}
 dfrm_upperarm_r {parent=dfrm_clavicle_r, loc=<-9.8658 1.2136 0.4105>, rot=<-0.2174 -0.685 -0.6546 0.2345>}
 dfrm_sleeve_r {parent=dfrm_upperarm_r, loc=<0 0 0>, rot=<0 0 0 1>}
 dfrm_shirt_upperarm_scale_r {parent=dfrm_sleeve_r, loc=<7.7118 2.0143 -0.2206>, rot=<-0.0172 0.0068 0.0533 -0.9984>}
 dfrm_aim_upperarm_r {parent=clavicle_r, loc=<-18.9397 0.4874 -11.5117>, rot=<-0.2174 -0.685 -0.6546 0.2345>}
 dfrm_necklace_01_l {parent=spine_05, loc=<-7.607 12.8547 -0.2563>, rot=<0.839 -0.1289 -0.5277 -0.033>}
 dfrm_necklace_02_l {parent=dfrm_necklace_01_l, loc=<2.7027 0 0>, rot=<-0.032 -0.2595 -0.0713 -0.9626>}
 dfrm_necklace_03_l {parent=dfrm_necklace_02_l, loc=<2.7497 0 0>, rot=<-0.0873 -0.1514 -0.0305 -0.9841>}
 dfrm_necklace_04_l {parent=dfrm_necklace_03_l, loc=<2.7543 0.021 0.1558>, rot=<-0.0191 -0.0211 0.0033 -0.9996>}
 dfrm_necklace_05_l {parent=dfrm_necklace_04_l, loc=<3.6821 0 0>, rot=<0.04 -0.0452 -0.0011 -0.9982>}
 dfrm_necklace_06_l {parent=dfrm_necklace_05_l, loc=<3.68 0 0>, rot=<0 0 0 1>}
 dfrm_necklace_01_r {parent=spine_05, loc=<-7.607 12.8547 -0.2563>, rot=<0.8321 -0.0529 0.5465 -0.0791>}
 dfrm_necklace_02_r {parent=dfrm_necklace_01_r, loc=<2.6894 0 0>, rot=<0.1573 0.2342 -0.0973 -0.9545>}
 dfrm_necklace_03_r {parent=dfrm_necklace_02_r, loc=<2.7517 0 0>, rot=<0.174 0.1276 -0.0227 -0.9762>}
 dfrm_necklace_04_r {parent=dfrm_necklace_03_r, loc=<2.7587 0 0>, rot=<-0.0401 0.0718 0.0013 -0.9966>}
 dfrm_necklace_05_r {parent=dfrm_necklace_04_r, loc=<3.6805 0 0>, rot=<0.0004 0.0387 0.0272 -0.9989>}
 dfrm_necklace_06_r {parent=dfrm_necklace_05_r, loc=<3.6818 0 0>, rot=<0 0 0 1>}
 dyn_neckace_mstr {parent=spine_05, loc=<-0.5594 9.0185 0>, rot=<-0.0137 0.011 0.9689 0.2467>}
 dyn_neckace_01 {parent=dyn_neckace_mstr, loc=<0 0 0>, rot=<0 0 0 1>}
 dyn_necklace_02 {parent=dyn_neckace_01, loc=<8.0282 0 0>, rot=<0.0112 0.0135 -0.2638 -0.9644>}
 dfrm_sphere_necklace_mstr {parent=spine_05, loc=<7.2677 2.1291 0>, rot=<0.0284 -0.7065 -0.0284 -0.7065>}
 dfrm_sphere_necklace_mstr_01 {parent=dfrm_sphere_necklace_mstr, loc=<-0.0823 6.4384 -6.0588>, rot=<0 0 0 1>}
 dfrm_sphere_necklace_01 {parent=dfrm_sphere_necklace_mstr_01, loc=<0 0 0>, rot=<0 0 0 1>}
 dfrm_sphere_necklace_mstr_02 {parent=dfrm_sphere_necklace_mstr, loc=<4.2425 6.1927 -5.2437>, rot=<0 0 0 1>}
 dfrm_sphere_necklace_02 {parent=dfrm_sphere_necklace_mstr_02, loc=<0 0 0>, rot=<0 0 0 1>}
 dfrm_sphere_necklace_mstr_03 {parent=dfrm_sphere_necklace_mstr, loc=<7.2321 4.4094 -2.895>, rot=<0 0 0 1>}
 dfrm_sphere_necklace_03 {parent=dfrm_sphere_necklace_mstr_03, loc=<0 0 0>, rot=<0 0 0 1>}
 dfrm_sphere_necklace_mstr_04 {parent=dfrm_sphere_necklace_mstr, loc=<8.6081 1.9524 -0.2869>, rot=<0 0 0 1>}
 dfrm_sphere_necklace_04 {parent=dfrm_sphere_necklace_mstr_04, loc=<0 0 0>, rot=<0 0 0 1>}
 dfrm_sphere_necklace_mstr_05 {parent=dfrm_sphere_necklace_mstr, loc=<8.6165 -0.9746 2.1551>, rot=<0 0 0 1>}
 dfrm_sphere_necklace_05 {parent=dfrm_sphere_necklace_mstr_05, loc=<0 0 0>, rot=<0 0 0 1>}
 dfrm_sphere_necklace_mstr_06 {parent=dfrm_sphere_necklace_mstr, loc=<7.3265 -4.3145 3.5327>, rot=<0 0 0 1>}
 dfrm_sphere_necklace_06 {parent=dfrm_sphere_necklace_mstr_06, loc=<0 0 0>, rot=<0 0 0 1>}
 dfrm_sphere_necklace_mstr_07 {parent=dfrm_sphere_necklace_mstr, loc=<4.1363 -6.6578 3.8243>, rot=<0 0 0 1>}
 dfrm_sphere_necklace_07 {parent=dfrm_sphere_necklace_mstr_07, loc=<0 0 0>, rot=<0 0 0 1>}
 dfrm_sphere_necklace_mstr_08 {parent=dfrm_sphere_necklace_mstr, loc=<0.1774 -7.6538 3.8859>, rot=<0 0 0 1>}
 dfrm_sphere_necklace_08 {parent=dfrm_sphere_necklace_mstr_08, loc=<0 0 0>, rot=<0 0 0 1>}
 dfrm_sphere_necklace_mstr_09 {parent=dfrm_sphere_necklace_mstr, loc=<-4.1363 -6.6578 3.8243>, rot=<0 0 0 1>}
 dfrm_sphere_necklace_09 {parent=dfrm_sphere_necklace_mstr_09, loc=<0 0 0>, rot=<0 0 0 1>}
 dfrm_sphere_necklace_mstr_10 {parent=dfrm_sphere_necklace_mstr, loc=<-7.3265 -4.3145 3.5327>, rot=<0 0 0 1>}
 dfrm_sphere_necklace_10 {parent=dfrm_sphere_necklace_mstr_10, loc=<0 0 0>, rot=<0 0 0 1>}
 dfrm_sphere_necklace_mstr_11 {parent=dfrm_sphere_necklace_mstr, loc=<-8.6165 -0.9746 2.1551>, rot=<0 0 0 1>}
 dfrm_sphere_necklace_11 {parent=dfrm_sphere_necklace_mstr_11, loc=<0 0 0>, rot=<0 0 0 1>}
 dfrm_sphere_necklace_mstr_12 {parent=dfrm_sphere_necklace_mstr, loc=<-8.6081 1.9524 -0.2869>, rot=<0 0 0 1>}
 dfrm_sphere_necklace_12 {parent=dfrm_sphere_necklace_mstr_12, loc=<0 0 0>, rot=<0 0 0 1>}
 dfrm_sphere_necklace_mstr_13 {parent=dfrm_sphere_necklace_mstr, loc=<-7.2321 4.4094 -2.895>, rot=<0 0 0 1>}
 dfrm_sphere_necklace_13 {parent=dfrm_sphere_necklace_mstr_13, loc=<0 0 0>, rot=<0 0 0 1>}
 dfrm_sphere_necklace_mstr_14 {parent=dfrm_sphere_necklace_mstr, loc=<-4.2425 6.1927 -5.2437>, rot=<0 0 0 1>}
 dfrm_sphere_necklace_14 {parent=dfrm_sphere_necklace_mstr_14, loc=<0 0 0>, rot=<0 0 0 1>}
 deform_pec_aim_base_l {parent=spine_04, loc=<3.2456 9.3309 -1.2237>, rot=<0.0603 -0.4033 0.3318 -0.8507>}
 deform_pec_aim_l {parent=deform_pec_aim_base_l, loc=<0 0 0>, rot=<0 0 0 1>}
 deform_pec_base_l {parent=deform_pec_aim_l, loc=<9.3665 0 0>, rot=<0 0 0 1>}
 deform_pec_l {parent=deform_pec_base_l, loc=<0 0 0>, rot=<0 0 0 1>}
 deform_pec_aim_base_r {parent=spine_04, loc=<3.2452 9.3309 1.2237>, rot=<-0.4033 -0.0603 0.8507 0.3318>}
 deform_pec_aim_r {parent=deform_pec_aim_base_r, loc=<0 0 0>, rot=<0 0 0 1>}
 deform_pec_base_r {parent=deform_pec_aim_r, loc=<-9.3667 -0.0003 -0.0004>, rot=<0 0 0 1>}
 deform_pec_r {parent=deform_pec_base_r, loc=<0 0 0>, rot=<0 0 0 1>}
 dyn_shirt_mstr {parent=spine_03, loc=<0 0 0>, rot=<0 0 0 1>}
 dyn_shirt_front_cnstr_l {parent=dyn_shirt_mstr, loc=<1.3816 8.8732 -8.324>, rot=<0.2834 -0.2653 -0.6996 -0.5998>}
 dyn_shirt_front_mstr_l {parent=dyn_shirt_front_cnstr_l, loc=<2.2122 0 0>, rot=<-0.0742 -0.0711 -0.6785 -0.7274>}
 dyn_shirt_front_01_l {parent=dyn_shirt_front_mstr_l, loc=<0 0 0>, rot=<0 0 0 1>}
 dyn_shirt_front_02_l {parent=dyn_shirt_front_01_l, loc=<7.9882 0 0>, rot=<0 0 0.0443 -0.999>}
 dyn_skirt_front_01_l {parent=dyn_shirt_front_02_l, loc=<7.9882 0 0>, rot=<0.1214 -0.0283 0.0262 -0.9919>}
 dyn_skirt_front_02_l {parent=dyn_skirt_front_01_l, loc=<11 0 0>, rot=<0.0443 0.0002 -0.0054 -0.999>}
 dyn_skirt_front_03_l {parent=dyn_skirt_front_02_l, loc=<11 0 0>, rot=<0.0341 -0.0018 -0.0072 -0.9994>}
 dyn_shirt_front_side_cnstr_l {parent=dyn_shirt_mstr, loc=<1.9857 3.8633 -10.1693>, rot=<-0.4432 0.4461 0.6082 0.4845>}
 dyn_shirt_front_side_mstr_l {parent=dyn_shirt_front_side_cnstr_l, loc=<2.0444 0 0>, rot=<0.0146 0.0273 -0.6451 -0.7635>}
 dyn_shirt_front_side_01_l {parent=dyn_shirt_front_side_mstr_l, loc=<0 0 0>, rot=<0 0 0 1>}
 dyn_shirt_front_side_02_l {parent=dyn_shirt_front_side_01_l, loc=<7.9882 0 0>, rot=<0 0 0.0651 -0.9979>}
 dyn_skirt_front_side_01_l {parent=dyn_shirt_front_side_02_l, loc=<7.9882 0 0>, rot=<0.0279 -0.0246 0.0275 -0.9989>}
 dyn_skirt_front_side_02_l {parent=dyn_skirt_front_side_01_l, loc=<9.3048 0 0>, rot=<0 0 -0.0063 -1>}
 dyn_skirt_front_side_03_l {parent=dyn_skirt_front_side_02_l, loc=<9.3048 0 0>, rot=<0.0344 0 0 -0.9994>}
 dyn_shirt_side_cnstr_l {parent=dyn_shirt_mstr, loc=<3.2772 -1.6844 -10.7109>, rot=<-0.4073 0.5492 0.6302 0.368>}
 dyn_shirt_side_mstr_l {parent=dyn_shirt_side_cnstr_l, loc=<2.2886 0 0>, rot=<-0.1317 -0.0855 -0.5483 -0.8214>}
 dyn_shirt_side_01_l {parent=dyn_shirt_side_mstr_l, loc=<0 0 0>, rot=<0 0 0 1>}
 dyn_shirt_side_02_l {parent=dyn_shirt_side_01_l, loc=<7.9882 0 0>, rot=<0 0 0.0513 -0.9987>}
 dyn_skirt_side_01_l {parent=dyn_shirt_side_02_l, loc=<7.9882 0 0>, rot=<-0.0567 -0.0065 0.06 -0.9966>}
 dyn_skirt_side_02_l {parent=dyn_skirt_side_01_l, loc=<8.1335 0 0>, rot=<0 0 0.0242 -0.9997>}
 dyn_skirt_side_03_l {parent=dyn_skirt_side_02_l, loc=<8.1335 0 0>, rot=<0 0 0 1>}
 dyn_shirt_back_side_cnstr_l {parent=dyn_shirt_mstr, loc=<3.7618 -5.5471 -8.2699>, rot=<0.5916 -0.7651 -0.217 -0.1326>}
 dyn_shirt_back_side_mstr_l {parent=dyn_shirt_back_side_cnstr_l, loc=<2.6592 0 0>, rot=<-0.0259 -0.0199 -0.621 -0.7832>}
 dyn_shirt_back_side_01_l {parent=dyn_shirt_back_side_mstr_l, loc=<0 0 0>, rot=<0 0 0 1>}
 dyn_shirt_back_side_02_l {parent=dyn_shirt_back_side_01_l, loc=<7.9882 0 0>, rot=<0 0 0.0756 -0.9971>}
 dyn_skirt_back_side_01_l {parent=dyn_shirt_back_side_02_l, loc=<7.9882 0 0>, rot=<0.0876 -0.0047 0.0765 -0.9932>}
 dyn_skirt_back_side_02_l {parent=dyn_skirt_back_side_01_l, loc=<8.262 0 0>, rot=<-0.08 0.0002 0.0081 -0.9968>}
 dyn_skirt_back_side_03_l {parent=dyn_skirt_back_side_02_l, loc=<8.262 0 0>, rot=<-0.0136 -0.0004 -0.0264 -0.9996>}
 dyn_shirt_back_cnstr {parent=dyn_shirt_mstr, loc=<4.4868 -7.2898 -0.1813>, rot=<0.0387 -0.706 -0.0387 -0.706>}
 dyn_shirt_back_mstr {parent=dyn_shirt_back_cnstr, loc=<-0.1813 -2.2059 -0.9228>, rot=<-0.0213 -0.7103 -0.0215 -0.7032>}
 dyn_shirt_back_01 {parent=dyn_shirt_back_mstr, loc=<0 0 0>, rot=<0 0 0 1>}
 dyn_shirt_back_02 {parent=dyn_shirt_back_01, loc=<7.9882 0 0>, rot=<0 0 0.0441 -0.999>}
 dyn_skirt_back_01 {parent=dyn_shirt_back_02, loc=<7.9882 0 -0.0004>, rot=<0 0 0.0516 -0.9987>}
 dyn_skirt_back_02 {parent=dyn_skirt_back_01, loc=<8.9093 0 0>, rot=<0 0 0 1>}
 dyn_skirt_back_03 {parent=dyn_skirt_back_02, loc=<8.9093 0 0>, rot=<0 0 0 1>}
 dyn_shirt_front_cnstr_r {parent=dyn_shirt_mstr, loc=<1.3816 8.8733 8.324>, rot=<-0.2834 0.2654 -0.6997 -0.5998>}
 dyn_shirt_front_mstr_r {parent=dyn_shirt_front_cnstr_r, loc=<2.2122 0 0>, rot=<0.0742 0.0711 -0.6784 -0.7274>}
 dyn_shirt_front_01_r {parent=dyn_shirt_front_mstr_r, loc=<0 0 0>, rot=<0 0 0 1>}
 dyn_shirt_front_02_r {parent=dyn_shirt_front_01_r, loc=<7.9876 0 0>, rot=<0 0 0.0443 -0.999>}
 dyn_skirt_front_01_r {parent=dyn_shirt_front_02_r, loc=<7.989 0 0>, rot=<-0.1214 0.0283 0.0263 -0.9919>}
 dyn_skirt_front_02_r {parent=dyn_skirt_front_01_r, loc=<10.9996 0 0>, rot=<-0.0443 -0.0002 -0.0054 -0.999>}
 dyn_skirt_front_03_r {parent=dyn_skirt_front_02_r, loc=<11 0 0>, rot=<-0.0341 0.0018 -0.0072 -0.9994>}
 dyn_shirt_front_side_cnstr_r {parent=dyn_shirt_mstr, loc=<1.9859 3.8634 10.1693>, rot=<0.4432 -0.4461 0.6082 0.4844>}
 dyn_shirt_front_side_mstr_r {parent=dyn_shirt_front_side_cnstr_r, loc=<2.0444 0 0>, rot=<-0.0146 -0.0273 -0.645 -0.7636>}
 dyn_shirt_front_side_01_r {parent=dyn_shirt_front_side_mstr_r, loc=<0 0 0>, rot=<0 0 0 1>}
 dyn_shirt_front_side_02_r {parent=dyn_shirt_front_side_01_r, loc=<7.9875 0 0>, rot=<0 0 0.0651 -0.9979>}
 dyn_skirt_front_side_01_r {parent=dyn_shirt_front_side_02_r, loc=<7.9884 0 0>, rot=<-0.0279 0.0246 0.0275 -0.9989>}
 dyn_skirt_front_side_02_r {parent=dyn_skirt_front_side_01_r, loc=<9.3049 0 0>, rot=<0 0 -0.0063 -1>}
 dyn_skirt_front_side_03_r {parent=dyn_skirt_front_side_02_r, loc=<9.3048 0 0>, rot=<-0.0344 0 0 -0.9994>}
 dyn_shirt_side_cnstr_r {parent=dyn_shirt_mstr, loc=<3.277 -1.6844 10.7109>, rot=<0.4073 -0.5492 0.6302 0.368>}
 dyn_shirt_side_mstr_r {parent=dyn_shirt_side_cnstr_r, loc=<2.2886 0 0>, rot=<0.1317 0.0855 -0.5483 -0.8214>}
 dyn_shirt_side_01_r {parent=dyn_shirt_side_mstr_r, loc=<0 0 0>, rot=<0 0 0 1>}
 dyn_shirt_side_02_r {parent=dyn_shirt_side_01_r, loc=<7.9885 0 0>, rot=<0 0 0.0513 -0.9987>}
 dyn_skirt_side_01_r {parent=dyn_shirt_side_02_r, loc=<7.9878 0 0>, rot=<0.0567 0.0065 0.06 -0.9966>}
 dyn_skirt_side_02_r {parent=dyn_skirt_side_01_r, loc=<8.1335 0 0>, rot=<0 0 0.0242 -0.9997>}
 dyn_skirt_side_03_r {parent=dyn_skirt_side_02_r, loc=<8.1335 0 0>, rot=<0 0 0 1>}
 dyn_shirt_back_side_cnstr_r {parent=dyn_shirt_mstr, loc=<3.7621 -5.5471 8.2699>, rot=<0.5915 -0.7651 0.217 0.1326>}
 dyn_shirt_back_side_mstr_r {parent=dyn_shirt_back_side_cnstr_r, loc=<2.6593 0 0>, rot=<0.0259 0.0199 -0.6209 -0.7832>}
 dyn_shirt_back_side_01_r {parent=dyn_shirt_back_side_mstr_r, loc=<0 0 0>, rot=<0 0 0 1>}
 dyn_shirt_back_side_02_r {parent=dyn_shirt_back_side_01_r, loc=<7.9885 0 0>, rot=<0 0 0.0756 -0.9971>}
 dyn_skirt_back_side_01_r {parent=dyn_shirt_back_side_02_r, loc=<7.9875 0 0>, rot=<-0.0876 0.0047 0.0765 -0.9932>}
 dyn_skirt_back_side_02_r {parent=dyn_skirt_back_side_01_r, loc=<8.2623 0 0>, rot=<0.08 -0.0002 0.0081 -0.9968>}
 dyn_skirt_back_side_03_r {parent=dyn_skirt_back_side_02_r, loc=<8.262 0 0>, rot=<0.0136 0.0004 -0.0264 -0.9996>}
 dfrm_shirt_scale {parent=spine_03, loc=<5.3381 0.0213 0>, rot=<0.0387 -0.706 -0.0387 -0.706>}
 thigh_l {parent=pelvis, loc=<-1.0855 -0.2109 -8.7053>, rot=<0.0009 0.0355 0.0155 -0.9993>}
 calf_l {parent=thigh_l, loc=<-42.5284 -0.0019 -0.7297>, rot=<0.0006 -0.0161 -0.0391 -0.9991>}
 foot_l {parent=calf_l, loc=<-46.7401 -2.5014 -0.592>, rot=<0.0427 -0.0238 0.0353 -0.9982>}
 ball_l {parent=foot_l, loc=<-3.5658 10.8066 -0.1544>, rot=<0.0001 -0.0001 -0.7063 -0.7079>}
 calf_twist_02_l {parent=calf_l, loc=<-19.1546 -1.7943 -0.2638>, rot=<0.0237 0.0095 -0.0035 -0.9997>}
 dyn_inner_lace_mstr_l {parent=calf_twist_02_l, loc=<-1.4125 3.5028 3.8235>, rot=<0.2487 0.1632 0.9434 0.1468>}
 dyn_inner_lace_01_l {parent=dyn_inner_lace_mstr_l, loc=<0 0 0>, rot=<0 0 0 1>}
 dyn_inner_lace_02_l {parent=dyn_inner_lace_01_l, loc=<5.9434 0 0>, rot=<0.0992 -0.0679 -0.0137 -0.9927>}
 dyn_inner_lace_03_l {parent=dyn_inner_lace_02_l, loc=<5.9434 0 0>, rot=<-0.0242 -0.0004 0.0153 -0.9996>}
 dyn_inner_lace_04_l {parent=dyn_inner_lace_03_l, loc=<5.9434 0 0>, rot=<0 0 0 1>}
 dyn_outer_lace_mstr_l {parent=calf_twist_02_l, loc=<-1.5963 4.3891 -4.2971>, rot=<-0.1737 0.0576 0.9646 0.1899>}
 dyn_outer_lace_01_l {parent=dyn_outer_lace_mstr_l, loc=<0 0 0>, rot=<0 0 0 1>}
 dyn_outer_lace_02_l {parent=dyn_outer_lace_01_l, loc=<4.6042 0 0>, rot=<-0.0076 0.031 0.0303 -0.999>}
 dyn_outer_lace_03_l {parent=dyn_outer_lace_02_l, loc=<4.6056 0 0>, rot=<0 0.0002 -0.0123 -0.9999>}
 dyn_outer_lace_04_l {parent=dyn_outer_lace_03_l, loc=<4.6067 0 0>, rot=<0.0052 0.0001 -0.057 -0.9984>}
 calf_twist_01_l {parent=calf_l, loc=<-35.8483 -1.9197 -0.5794>, rot=<0.0237 0.0095 -0.0035 -0.9997>}
 deform_knee_in_base_l {parent=calf_l, loc=<0.8031 -5.0231 0.2456>, rot=<0 0 0 -1>}
 deform_knee_in_l {parent=deform_knee_in_base_l, loc=<0 0 0>, rot=<0 0 0 1>}
 deform_knee_out_base_l {parent=calf_l, loc=<0 4 0>, rot=<0 0 0 -1>}
 deform_knee_out_l {parent=deform_knee_out_base_l, loc=<0 0 0>, rot=<0 0 0 1>}
 thigh_twist_01_l {parent=thigh_l, loc=<-27.952 0 0>, rot=<0.0013 0 0 -1>}
 deform_groin_base_l {parent=thigh_l, loc=<0.4801 5.6776 -0.2486>, rot=<0 0 0 -1>}
 deform_groin_l {parent=deform_groin_base_l, loc=<0 0 0>, rot=<0 0 0 1>}
 thigh_r {parent=pelvis, loc=<-1.0855 -0.2109 8.7053>, rot=<0.0355 -0.0009 0.9993 0.0155>}
 calf_r {parent=thigh_r, loc=<42.5284 0.0019 0.7297>, rot=<0.0006 -0.0161 -0.0391 -0.9991>}
 foot_r {parent=calf_r, loc=<46.7401 2.5014 0.592>, rot=<0.0427 -0.0238 0.0353 -0.9982>}
 ball_r {parent=foot_r, loc=<3.5658 -10.8066 0.1544>, rot=<0.0001 -0.0001 -0.7063 -0.7079>}
 calf_twist_02_r {parent=calf_r, loc=<19.1546 1.7943 0.2638>, rot=<0.0237 0.0095 -0.0035 -0.9997>}
 dyn_outer_lace_mstr_r {parent=calf_twist_02_r, loc=<1.5757 -4.3841 4.2724>, rot=<-0.2031 0.188 0.0486 -0.9597>}
 dyn_outer_lace_01_r {parent=dyn_outer_lace_mstr_r, loc=<0 0 0>, rot=<0 0 0 1>}
 dyn_outer_lace_02_r {parent=dyn_outer_lace_01_r, loc=<6.2729 0 0>, rot=<0.0793 -0.0649 0.0037 -0.9947>}
 dyn_outer_lace_03_r {parent=dyn_outer_lace_02_r, loc=<6.2729 0 0>, rot=<0.0235 -0.0375 0.0067 -0.999>}
 dyn_outer_lace_04_r {parent=dyn_outer_lace_03_r, loc=<6.3409 0 0>, rot=<-0.6245 0.0569 0.7759 -0.0692>}
 dyn_inner_lace_mstr_r {parent=calf_twist_02_r, loc=<1.6153 -3.3654 -3.8568>, rot=<-0.101 -0.1849 0.0923 -0.9732>}
 dyn_inner_lace_01_r {parent=dyn_inner_lace_mstr_r, loc=<0 0 0>, rot=<0 0 0 1>}
 dyn_inner_lace_02_r {parent=dyn_inner_lace_01_r, loc=<6.9771 0 0>, rot=<-0.0133 0.0851 -0.0065 -0.9963>}
 dyn_inner_lace_03_r {parent=dyn_inner_lace_02_r, loc=<6.9771 0 0>, rot=<-0.0003 0.0085 0.0386 -0.9992>}
 dyn_inner_lace_04_r {parent=dyn_inner_lace_03_r, loc=<7.0637 0 0>, rot=<0 0 0 1>}
 calf_twist_01_r {parent=calf_r, loc=<35.8483 1.9197 0.5794>, rot=<0.0237 0.0095 -0.0035 -0.9997>}
 deform_knee_in_base_r {parent=calf_r, loc=<-0.8031 5.0231 -0.2457>, rot=<0 0 0 -1>}
 deform_knee_in_r {parent=deform_knee_in_base_r, loc=<0 0 0>, rot=<0 0 0 1>}
 deform_knee_out_base_r {parent=calf_r, loc=<0 -4 0>, rot=<0 0 0 -1>}
 deform_knee_out_r {parent=deform_knee_out_base_r, loc=<0 0 0>, rot=<0 0 0 1>}
 thigh_twist_01_r {parent=thigh_r, loc=<27.952 0 0>, rot=<0.0013 0 0 -1>}
 deform_groin_base_r {parent=thigh_r, loc=<-0.4801 -5.6776 0.2486>, rot=<0 0 0 -1>}
 deform_groin_r {parent=deform_groin_base_r, loc=<0 0 0>, rot=<0 0 0 1>}
 deform_glute_base_l {parent=pelvis, loc=<-7.3466 -9.0213 -9.1245>, rot=<0.0172 0.031 -0.4477 -0.8935>}
 deform_glute_l {parent=deform_glute_base_l, loc=<0 0 0>, rot=<0 0 0 1>}
 deform_glute_base_r {parent=pelvis, loc=<-7.3464 -9.0213 9.1245>, rot=<0.031 -0.0172 0.8935 -0.4477>}
 deform_glute_r {parent=deform_glute_base_r, loc=<0 0 0>, rot=<0 0 0 1>}
 ring_01_mstr_l {parent=pelvis, loc=<-3.4618 7.3133 -8.4141>, rot=<0 0 0.9998 0.0179>}
 dyn_ring_01_l {parent=ring_01_mstr_l, loc=<0 0 0>, rot=<0 0 0 1>}
 ring_02_mstr_l {parent=pelvis, loc=<-8.7029 6.7934 -12.4325>, rot=<-0.0621 0.2816 0.9522 0.1006>}
 dyn_ring_02_l {parent=ring_02_mstr_l, loc=<0 0 0>, rot=<0 0 0 1>}
 ring_03_mstr_l {parent=pelvis, loc=<-9.3917 6.9898 -5.795>, rot=<-0.028 -0.1714 0.9843 0.0304>}
 dyn_ring_03_l {parent=ring_03_mstr_l, loc=<0 0 0>, rot=<0 0 0 1>}
 dyn_ring_mstr {parent=pelvis, loc=<-1.4815 8.42 0>, rot=<0 0 0.9998 0.0179>}
 dyn_ring {parent=dyn_ring_mstr, loc=<0 0 0>, rot=<0 0 0 1>}
 ring_01_mstr_r {parent=pelvis, loc=<-3.4618 7.3133 8.4141>, rot=<0 0 0.9998 0.0179>}
 dyn_ring_01_r {parent=ring_01_mstr_r, loc=<0 0 0>, rot=<0 0 0 1>}
 ring_02_mstr_r {parent=pelvis, loc=<-8.703 6.7934 12.4325>, rot=<0.0621 -0.2816 0.9522 0.1006>}
 dyn_ring_02_r {parent=ring_02_mstr_r, loc=<0 0 0>, rot=<0 0 0 1>}
 ring_03_mstr_r {parent=pelvis, loc=<-9.3917 6.9898 5.795>, rot=<0.028 0.1714 0.9843 0.0304>}
 dyn_ring_03_r {parent=ring_03_mstr_r, loc=<0 0 0>, rot=<0 0 0 1>}
 dyn_util_belt_r {parent=pelvis, loc=<4.726 7.6809 8.523>, rot=<-0.0127 -0.707 0.0127 -0.707>}
 dyn_util_belt_l {parent=pelvis, loc=<4.726 7.6809 -8.523>, rot=<-0.0127 -0.707 0.0127 -0.707>}
]
variables [
 is_crouching {type=bool, default=False}
 is_moving_2d {type=bool, default=False}
 look_up_with_pistol_mult_alpha {type=double, default=0}
 spine_rotate_down_alpha {type=double, default=0}
 spr_clavicle_l_up {type=double, default=0}
 spr_clavicle_r_twist_alpha {type=double, default=0}
 spr_look_up_alpha {type=double, default=0}
 spr_thigh_l_up_alpha {type=double, default=0}
 spr_thigh_r_up_alpha {type=double, default=0}
 spr_upperarm_l_pistol_aim_alpha {type=double, default=0}
 spr_upperarm_r_pistol_aim_alpha {type=double, default=0}
]
