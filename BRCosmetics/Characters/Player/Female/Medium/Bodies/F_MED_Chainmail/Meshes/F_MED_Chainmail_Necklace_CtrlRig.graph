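# Control Rig that deforms the chainmail necklace on the medium female player body.
# Weighted parent constraints place each side's control nulls between spine, clavicle, pec and upper-arm bones and the left top necklace control.
# Both top controls ride the dyn_necklace_02 bone.
# Per side, five controls drive a six-bone chain through stretchy B-spline IK; right-side weights differ from left and partly reference left-side elements.
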ControlRigBlueprint /BRCosmetics/Characters/Player/Female/Medium/Bodies/F_MED_Chainmail/Meshes/F_MED_Chainmail_Necklace_CtrlRig
graphs [
 RigVMFunctionLibrary {links [], nodes []}
 RigVMModel {nodes [
   RigUnit_BeginExecution = &RigUnit_BeginExecution()
   &RigUnit_ParentConstraint(
    AdvancedSettings=(RotationOrderForFilter=XZY)
    Child=(Type=Null,Name="dfrm_necklace_03_l_ctrl_null")
    Parents=((Item=(Type=Control,Name="dfrm_necklace_01_l_ctrl"),Weight=0.7),(Item=(Type=Bone,Name="clavicle_l"),Weight=0.1),(Item=(Type=Bone,Name="deform_pec_aim_l"),Weight=0.2))
    Weight=1
    ExecutePin=@RigUnit_BeginExecution.ExecutePin
   )
   > &RigUnit_ParentConstraint(
    AdvancedSettings=(RotationOrderForFilter=XZY)
    Child=(Type=Null,Name="dfrm_necklace_04_l_ctrl_null")
    Parents=((Item=(Type=Control,Name="dfrm_necklace_01_l_ctrl"),Weight=0.1),(Item=(Type=Bone,Name="spine_04"),Weight=0.2),(Item=(Type=Bone,Name="spine_05"),Weight=0.2),(Item=(Type=Bone,Name="deform_pec_l"),Weight=0.5))
    Weight=1
   )
   > &RigUnit_ParentConstraint(
    AdvancedSettings=(RotationOrderForFilter=XZY)
    Child=(Type=Null,Name="dfrm_necklace_05_l_ctrl_null")
    Parents=((Item=(Type=Bone,Name="deform_pec_l"),Weight=0.5),(Item=(Type=Bone,Name="spine_04"),Weight=0.27),(Item=(Type=Bone,Name="spine_05"),Weight=0.2),(Item=(Type=Bone,Name="clavicle_l"),Weight=0.03))
    Weight=1
   )
   > &RigUnit_ParentConstraint(
    AdvancedSettings=(RotationOrderForFilter=XZY)
    Child=(Type=Null,Name="dfrm_necklace_06_l_ctrl_null")
    Parents=((Item=(Type=Bone,Name="spine_05"),Weight=0.54),(Item=(Type=Bone,Name="clavicle_l"),Weight=0.25),(Item=(Type=Bone,Name="spine_04"),Weight=0.14),(Item=(Type=Bone,Name="deform_pec_l"),Weight=0.037),(Item=(Type=Bone,Name="deform_upperarm_fr_l"),Weight=0.033))
    Weight=1
   )
   > &RigUnit_ParentConstraint(
    AdvancedSettings=(RotationOrderForFilter=XZY)
    Child=(Type=Null,Name="dfrm_necklace_03_r_ctrl_null")
    Parents=((Item=(Type=Control,Name="dfrm_necklace_01_l_ctrl"),Weight=0.5),(Item=(Type=Bone,Name="deform_pec_r"),Weight=0.3),(Item=(Type=Bone,Name="spine_03"),Weight=0.2))
    Weight=1
   )
   > &RigUnit_ParentConstraint(
    AdvancedSettings=(RotationOrderForFilter=XZY)
    Child=(Type=Null,Name="dfrm_necklace_04_r_ctrl_null")
    Parents=((Item=(Type=Control,Name="dfrm_necklace_01_l_ctrl"),Weight=0.1),(Item=(Type=Bone,Name="spine_03"),Weight=0.3),(Item=(Type=Bone,Name="deform_pec_r"),Weight=0.6))
    Weight=1
   )
   > &RigUnit_ParentConstraint(
    AdvancedSettings=(RotationOrderForFilter=XZY)
    Child=(Type=Null,Name="dfrm_necklace_05_r_ctrl_null")
    Parents=((Item=(Type=Bone,Name="deform_pec_r"),Weight=0.6),(Item=(Type=Bone,Name="spine_03"),Weight=0.15),(Item=(Type=Bone,Name="spine_05"),Weight=0.2),(Item=(Type=Bone,Name="clavicle_l"),Weight=0.03))
    Weight=1
   )
   > &RigUnit_ParentConstraint(
    AdvancedSettings=(RotationOrderForFilter=XZY)
    Child=(Type=Null,Name="dfrm_necklace_06_r_ctrl_null")
    Parents=((Item=(Type=Bone,Name="spine_05"),Weight=0.54),(Item=(Type=Bone,Name="clavicle_r"),Weight=0.25),(Item=(Type=Bone,Name="spine_04"),Weight=0.14),(Item=(Type=Bone,Name="deform_pec_r"),Weight=0.037),(Item=(Type=Bone,Name="deform_upperarm_fr_r"),Weight=0.033))
    Weight=1
   )
   RerouteNode = RigVMDispatchNode(Values=((Type=Control,Name="dfrm_necklace_01_l_ctrl"),((),Type=Control,Name="dfrm_necklace_03_l_ctrl"),((),Type=Control,Name="dfrm_necklace_04_l_ctrl"),((),Type=Control,Name="dfrm_necklace_05_l_ctrl"),((),Type=Control,Name="dfrm_necklace_06_l_ctrl"))) {template="DISPATCH_RigVMDispatch_ArrayMake(in Values,out Array)"}
   RerouteNode_1 = RigVMDispatchNode(Values=((Type=Bone,Name="dfrm_necklace_01_l"),((),Type=Bone,Name="dfrm_necklace_02_l"),((),Type=Bone,Name="dfrm_necklace_03_l"),((),Type=Bone,Name="dfrm_necklace_04_l"),((),Type=Bone,Name="dfrm_necklace_05_l"),((),Type=Bone,Name="dfrm_necklace_06_l"))) {template="DISPATCH_RigVMDispatch_ArrayMake(in Values,out Array)"}
   > SplineIK = RigVMFunctionReferenceNode(
    Primary Axis=(X=1,Y=0,Z=0)
    Samples Per Segment=16
    Secondary Spline Direction=(X=0,Y=5,Z=0)
    Spline Mode=BSpline
    Stretch=true
    Up Axis=(X=0,Y=1,Z=0)
    Controls=@RerouteNode.Array
    Bones=@RerouteNode_1.Array
   )
   RerouteNode_2 = RigVMDispatchNode(Values=((Type=Control,Name="dfrm_necklace_01_r_ctrl"),((),Type=Control,Name="dfrm_necklace_03_r_ctrl"),((),Type=Control,Name="dfrm_necklace_04_r_ctrl"),((),Type=Control,Name="dfrm_necklace_05_r_ctrl"),((),Type=Control,Name="dfrm_necklace_06_r_ctrl"))) {template="DISPATCH_RigVMDispatch_ArrayMake(in Values,out Array)"}
   RerouteNode_3 = RigVMDispatchNode(Values=((Type=Bone,Name="dfrm_necklace_01_r"),((),Type=Bone,Name="dfrm_necklace_02_r"),((),Type=Bone,Name="dfrm_necklace_03_r"),((),Type=Bone,Name="dfrm_necklace_04_r"),((),Type=Bone,Name="dfrm_necklace_05_r"),((),Type=Bone,Name="dfrm_necklace_06_r"))) {template="DISPATCH_RigVMDispatch_ArrayMake(in Values,out Array)"}
   > SplineIK_1 = RigVMFunctionReferenceNode(
    Primary Axis=(X=1,Y=0,Z=0)
    Samples Per Segment=16
    Secondary Spline Direction=(X=0,Y=5,Z=0)
    Spline Mode=BSpline
    Stretch=true
    Up Axis=(X=0,Y=1,Z=0)
    Controls=@RerouteNode_2.Array
    Bones=@RerouteNode_3.Array
   )
  ]}
]
hierarchy [
 root {loc=<0 0 0>, rot=<0 0 0 1>}
 pelvis {parent=root, loc=<0 -0.5301 98.0571>, rot=<0.0127 0.707 -0.0127 -0.707>}
 spine_01 {parent=pelvis, loc=<6.1208 -0.9328 0>, rot=<0 0 0.034 -0.9994>}
 spine_02 {parent=spine_01, loc=<6.5996 -0.2095 0>, rot=<0 0 0.0238 -0.9997>}
 spine_03 {parent=spine_02, loc=<6.4914 -0.1237 0>, rot=<0 0 0.0149 -0.9999>}
 spine_04 {parent=spine_03, loc=<6.9598 0.0794 0>, rot=<0 0 -0.0132 -0.9999>}
 spine_05 {parent=spine_04, loc=<6.9987 0 0>, rot=<0 0 -0.0015 -1>}
 clavicle_l {parent=spine_05, loc=<5.4831 -0.1053 -3.2946>, rot=<0.0465 -0.7289 -0.009 -0.6829>}
 upperarm_l {parent=clavicle_l, loc=<9.8658 -1.2136 -0.4105>, rot=<0.0164 -0.3245 -0.0269 -0.9454>}
 lowerarm_l {parent=upperarm_l, loc=<26.465 -0.8345 0.259>, rot=<0.0036 0.0161 -0.2298 -0.9731>}
 hand_l {parent=lowerarm_l, loc=<22.7355 -0.3203 -0.6105>, rot=<-0.6784 -0.0223 0.0735 -0.7307>}
 thumb_01_l {parent=hand_l, loc=<1.3118 -0.3191 -2.459>, rot=<0.6546 -0.4042 0.0664 -0.6354>}
 thumb_02_l {parent=thumb_01_l, loc=<5.9336 -0.0448 0.0401>, rot=<-0.0037 0.0213 0.298 -0.9543>}
 thumb_03_l {parent=thumb_02_l, loc=<2.2223 0.3682 -0.0907>, rot=<0.0118 -0.013 0.0066 -0.9998>}
 index_metacarpal_l {parent=hand_l, loc=<1.6535 0.0183 -2.96>, rot=<0.0158 -0.0124 -0.0173 -0.9996>}
 index_01_l {parent=index_metacarpal_l, loc=<8.3008 0.0357 -0.0524>, rot=<-0.0044 0.0485 0.0924 -0.9945>}
 index_02_l {parent=index_01_l, loc=<3.3587 -0.3004 0.0452>, rot=<-0.002 0.0027 0.3225 -0.9466>}
 index_03_l {parent=index_02_l, loc=<2.5578 -0.123 -0.0252>, rot=<-0.002 0.0067 0.2558 -0.9667>}
 middle_metacarpal_l {parent=hand_l, loc=<2.1856 -0.1725 -0.5984>, rot=<0.006 0.0168 -0.0329 -0.9993>}
 middle_01_l {parent=middle_metacarpal_l, loc=<7.3275 -0.2447 0.0129>, rot=<-0.0208 0.0042 0.0858 -0.9961>}
 middle_02_l {parent=middle_01_l, loc=<4.0123 0.1401 0.1175>, rot=<-0.0218 0.0258 0.3512 -0.9357>}
 middle_03_l {parent=middle_02_l, loc=<2.6077 0.0534 0.1373>, rot=<0.0082 0.0188 0.1708 -0.9851>}
 ring_metacarpal_l {parent=hand_l, loc=<2.2074 -0.3625 1.3638>, rot=<-0.0284 0.0722 -0.0458 -0.9959>}
 ring_01_l {parent=ring_metacarpal_l, loc=<7.0295 -0.6197 -0.1446>, rot=<-0.0489 -0.0252 0.1892 -0.9804>}
 ring_02_l {parent=ring_01_l, loc=<3.6174 0.4647 -0.1027>, rot=<0 0 0.2806 -0.9598>}
 ring_03_l {parent=ring_02_l, loc=<2.4152 0.114 0.0199>, rot=<0 0 0.1189 -0.9929>}
 pinky_metacarpal_l {parent=hand_l, loc=<1.7616 -0.9753 3.3814>, rot=<-0.0397 0.0814 -0.069 -0.9935>}
 pinky_01_l {parent=pinky_metacarpal_l, loc=<7.1709 -0.4831 -0.1094>, rot=<-0.0654 -0.0035 0.2242 -0.9723>}
 pinky_02_l {parent=pinky_01_l, loc=<2.8274 0.0893 -0.1043>, rot=<0.0037 0.0011 0.2854 -0.9584>}
 pinky_03_l {parent=pinky_02_l, loc=<2.2046 0.0307 -0.0078>, rot=<-0.0002 0.0098 0.1445 -0.9895>}
 deform_wrist_dn_base_l {parent=hand_l, loc=<0 -1.5326 0>, rot=<0 0 0 -1>}
 deform_wrist_dn_l {parent=deform_wrist_dn_base_l, loc=<0 0 0>, rot=<0 0 0 1>}
 deform_wrist_up_base_l {parent=hand_l, loc=<0 2.3009 0>, rot=<0 0 0 -1>}
 deform_wrist_up_l {parent=deform_wrist_up_base_l, loc=<0 0 0>, rot=<0 0 0 1>}
 lowerarm_twist_02_l {parent=lowerarm_l, loc=<10.6707 0 0>, rot=<0 0 0 1>}
 lowerarm_twist_01_l {parent=lowerarm_l, loc=<18.316 0 0>, rot=<0 0 0 1>}
 deform_elbow_in_base_l {parent=lowerarm_l, loc=<0.1834 3.6062 -0.2572>, rot=<0 0 0 -1>}
 deform_elbow_in_l {parent=deform_elbow_in_base_l, loc=<0 0 0>, rot=<0 0 0 1>}
 deform_elbow_out_base_l {parent=lowerarm_l, loc=<0 -4 0>, rot=<0 0 0 1>}
 deform_elbow_out_l {parent=deform_elbow_out_base_l, loc=<0 0 0>, rot=<0 0 0 1>}
 upperarm_twist_01_l {parent=upperarm_l, loc=<7.2135 -0.1134 0>, rot=<0 0 0 1>}
 deform_upperarm_bk_base_l {parent=upperarm_twist_01_l, loc=<-6.4053 -4.3746 0.314>, rot=<0 0 0 -1>}
 deform_upperarm_bk_l {parent=deform_upperarm_bk_base_l, loc=<0 0 0>, rot=<0 0 0 1>}
 deform_upperarm_fr_base_l {parent=upperarm_twist_01_l, loc=<-5.6312 5.092 0.3354>, rot=<0 0 0 -1>}
 deform_upperarm_fr_l {parent=deform_upperarm_fr_base_l, loc=<0 0 0>, rot=<0 0 0 1>}
 deform_upperarm_up_base_l {parent=upperarm_twist_01_l, loc=<-8.891 0.4825 3.9517>, rot=<0 0 0 -1>}
 deform_upperarm_up_l {parent=deform_upperarm_up_base_l, loc=<0 0 0>, rot=<0 0 0 1>}
 upperarm_twist_02_l {parent=upperarm_l, loc=<19.7998 -0.1134 0>, rot=<0 0 0 1>}
 attach_cape {parent=spine_05, loc=<9.3216 -8.4501 0>, rot=<0.0616 -0.7044 -0.0616 -0.7044>}
 attach_backpack {parent=spine_05, loc=<0 -8.35 0>, rot=<-0.0093 -0.707 0.0093 -0.707>}
 neck_01 {parent=spine_05, loc=<11.3138 0.1529 0>, rot=<0 0 -0.2003 -0.9797>}
 neck_02 {parent=neck_01, loc=<4.5627 0 0>, rot=<0 0 0.0504 -0.9987>}
 head {parent=neck_02, loc=<5.3914 0 0>, rot=<0 0 0.115 -0.9934>}
 clavicle_r {parent=spine_05, loc=<5.4831 -0.1053 3.2946>, rot=<0.7289 0.0465 -0.6829 0.009>}
 upperarm_r {parent=clavicle_r, loc=<-9.8658 1.2136 0.4105>, rot=<0.0164 -0.3245 -0.0269 -0.9454>}
 lowerarm_r {parent=upperarm_r, loc=<-26.465 0.8345 -0.259>, rot=<0.0036 0.0161 -0.2298 -0.9731>}
 hand_r {parent=lowerarm_r, loc=<-22.7355 0.3203 0.6106>, rot=<-0.6784 -0.0223 0.0735 -0.7307>}
 thumb_01_r {parent=hand_r, loc=<-1.3118 0.3191 2.459>, rot=<0.6546 -0.4042 0.0664 -0.6354>}
 thumb_02_r {parent=thumb_01_r, loc=<-5.9336 0.0448 -0.0401>, rot=<-0.0037 0.0213 0.298 -0.9543>}
 thumb_03_r {parent=thumb_02_r, loc=<-2.2223 -0.3682 0.0907>, rot=<0.0118 -0.013 0.0066 -0.9998>}
 index_metacarpal_r {parent=hand_r, loc=<-1.6535 -0.0183 2.96>, rot=<0.0158 -0.0124 -0.0173 -0.9996>}
 index_01_r {parent=index_metacarpal_r, loc=<-8.3008 -0.0357 0.0524>, rot=<-0.0044 0.0485 0.0924 -0.9945>}
 index_02_r {parent=index_01_r, loc=<-3.3587 0.3004 -0.0452>, rot=<-0.002 0.0027 0.3225 -0.9466>}
 index_03_r {parent=index_02_r, loc=<-2.5578 0.123 0.0252>, rot=<-0.002 0.0067 0.2558 -0.9667>}
 middle_metacarpal_r {parent=hand_r, loc=<-2.1856 0.1725 0.5984>, rot=<0.006 0.0168 -0.0329 -0.9993>}
 middle_01_r {parent=middle_metacarpal_r, loc=<-7.3275 0.2447 -0.0129>, rot=<-0.0208 0.0042 0.0858 -0.9961>}
 middle_02_r {parent=middle_01_r, loc=<-4.0123 -0.1401 -0.1175>, rot=<-0.0218 0.0258 0.3512 -0.9357>}
 middle_03_r {parent=middle_02_r, loc=<-2.6077 -0.0534 -0.1373>, rot=<0.0082 0.0188 0.1708 -0.9851>}
 ring_metacarpal_r {parent=hand_r, loc=<-2.2074 0.3625 -1.3638>, rot=<-0.0284 0.0722 -0.0458 -0.9959>}
 ring_01_r {parent=ring_metacarpal_r, loc=<-7.0295 0.6197 0.1446>, rot=<-0.0489 -0.0252 0.1892 -0.9804>}
 ring_02_r {parent=ring_01_r, loc=<-3.6174 -0.4647 0.1027>, rot=<0 0 0.2806 -0.9598>}
 ring_03_r {parent=ring_02_r, loc=<-2.4152 -0.114 -0.0199>, rot=<0 0 0.1189 -0.9929>}
 pinky_metacarpal_r {parent=hand_r, loc=<-1.7616 0.9753 -3.3814>, rot=<-0.0397 0.0814 -0.069 -0.9935>}
 pinky_01_r {parent=pinky_metacarpal_r, loc=<-7.1709 0.4831 0.1094>, rot=<-0.0654 -0.0035 0.2242 -0.9723>}
 pinky_02_r {parent=pinky_01_r, loc=<-2.8274 -0.0893 0.1043>, rot=<0.0037 0.0011 0.2854 -0.9584>}
 pinky_03_r {parent=pinky_02_r, loc=<-2.2046 -0.0307 0.0078>, rot=<-0.0002 0.0098 0.1445 -0.9895>}
 deform_wrist_dn_base_r {parent=hand_r, loc=<-0.0001 1.5328 0>, rot=<0 0 0 -1>}
 deform_wrist_dn_r {parent=deform_wrist_dn_base_r, loc=<0 0 0>, rot=<0 0 0 1>}
 deform_wrist_up_base_r {parent=hand_r, loc=<-0.0002 -2.3007 -0.0001>, rot=<0 0 0 -1>}
 deform_wrist_up_r {parent=deform_wrist_up_base_r, loc=<0 0 0>, rot=<0 0 0 1>}
 lowerarm_twist_02_r {parent=lowerarm_r, loc=<-10.6707 0 0>, rot=<0 0 0 -1>}
 lowerarm_twist_01_r {parent=lowerarm_r, loc=<-18.316 0 0>, rot=<0 0 0 -1>}
 deform_elbow_in_base_r {parent=lowerarm_r, loc=<-0.1834 -3.6062 0.2572>, rot=<0 0 0 -1>}
 deform_elbow_in_r {parent=deform_elbow_in_base_r, loc=<0 0 0>, rot=<0 0 0 1>}
 deform_elbow_out_base_r {parent=lowerarm_r, loc=<0 4 -0.0001>, rot=<0 0 0 -1>}
 deform_elbow_out_r {parent=deform_elbow_out_base_r, loc=<0 0 0>, rot=<0 0 0 1>}
 upperarm_twist_01_r {parent=upperarm_r, loc=<-7.2135 0.1134 0>, rot=<0 0 0 -1>}
 deform_upperarm_bk_base_r {parent=upperarm_twist_01_r, loc=<6.4053 4.3746 -0.314>, rot=<0 0 0 -1>}
 deform_upperarm_bk_r {parent=deform_upperarm_bk_base_r, loc=<0 0 0>, rot=<0 0 0 1>}
 deform_upperarm_fr_base_r {parent=upperarm_twist_01_r, loc=<5.6312 -5.092 -0.3354>, rot=<0 0 0 -1>}
 deform_upperarm_fr_r {parent=deform_upperarm_fr_base_r, loc=<0 0 0>, rot=<0 0 0 1>}
 deform_upperarm_up_base_r {parent=upperarm_twist_01_r, loc=<8.891 -0.4825 -3.9517>, rot=<0 0 0 -1>}
 deform_upperarm_up_r {parent=deform_upperarm_up_base_r, loc=<0 0 0>, rot=<0 0 0 1>}
 upperarm_twist_02_r {parent=upperarm_r, loc=<-19.7998 0.1134 0>, rot=<0 0 0 -1>}
 dfrm_necklace_01_l {parent=spine_05, loc=<-7.607 12.8547 -0.2563>, rot=<0.839 -0.1289 -0.5277 -0.033>}
 dfrm_necklace_02_l {parent=dfrm_necklace_01_l, loc=<2.7027 0 0>, rot=<-0.032 -0.2595 -0.0713 -0.9626>}
 dfrm_necklace_03_l {parent=dfrm_necklace_02_l, loc=<2.7497 0 0>, rot=<-0.0873 -0.1514 -0.0305 -0.9841>}
 dfrm_necklace_04_l {parent=dfrm_necklace_03_l, loc=<2.7543 0.021 0.1558>, rot=<-0.0191 -0.0211 0.0033 -0.9996>}
 dfrm_necklace_05_l {parent=dfrm_necklace_04_l, loc=<3.6821 0 0>, rot=<0.04 -0.0452 -0.0011 -0.9982>}
 dfrm_necklace_06_l {parent=dfrm_necklace_05_l, loc=<3.68 0 0>, rot=<0 0 0 1>}
 dfrm_necklace_01_r {parent=spine_05, loc=<-7.607 12.8547 -0.2563>, rot=<0.8321 -0.0529 0.5465 -0.0791>}
 dfrm_necklace_02_r {parent=dfrm_necklace_01_r, loc=<2.6894 0 0>, rot=<0.1573 0.2342 -0.0973 -0.9545>}
 dfrm_necklace_03_r {parent=dfrm_necklace_02_r, loc=<2.7517 0 0>, rot=<0.174 0.1276 -0.0227 -0.9762>}
 dfrm_necklace_04_r {parent=dfrm_necklace_03_r, loc=<2.7587 0 0>, rot=<-0.0401 0.0718 0.0013 -0.9966>}
 dfrm_necklace_05_r {parent=dfrm_necklace_04_r, loc=<3.6805 0 0>, rot=<0.0004 0.0387 0.0272 -0.9989>}
 dfrm_necklace_06_r {parent=dfrm_necklace_05_r, loc=<3.6818 0 0>, rot=<0 0 0 1>}
 dyn_neckace_mstr {parent=spine_05, loc=<-0.5594 9.0185 0>, rot=<-0.0137 0.011 0.9689 0.2467>}
 dyn_neckace_01 {parent=dyn_neckace_mstr, loc=<0 0 0>, rot=<0 0 0 1>}
 dyn_necklace_02 {parent=dyn_neckace_01, loc=<8.0282 0 0>, rot=<0.0112 0.0135 -0.2638 -0.9644>}
 dyn_necklace_02_ctrl_null {parent=dyn_necklace_02, type=Null, loc=<0 0 0>, rot=<0 0 0 1>}
 dyn_necklace_02_ctrl {
  controlType=EULER_TRANSFORM
  parent=dyn_necklace_02_ctrl_null
  type=Control
  loc=<0.2563 9.8535 124.6066>
  rot=<-0.7069 -0.0159 0.7069 0.0159>
 }
 dfrm_necklace_01_l_ctrl_null {parent=dyn_necklace_02_ctrl, type=Null, loc=<0 0 0>, rot=<0.1437 0.8366 -0.0423 0.527>}
 dfrm_necklace_01_l_ctrl {controlType=EULER_TRANSFORM, parent=dfrm_necklace_01_l_ctrl_null, type=Control, loc=<0 0 0>, rot=<0 0 0 1>}
 dfrm_necklace_01_r_ctrl_null {parent=dyn_necklace_02_ctrl, type=Null, loc=<0 0 0>, rot=<0.0675 0.831 -0.0694 -0.5478>}
 dfrm_necklace_01_r_ctrl {controlType=EULER_TRANSFORM, parent=dfrm_necklace_01_r_ctrl_null, type=Control, loc=<0 0 0>, rot=<0 0 0 1>}
 deform_pec_aim_base_l {parent=spine_04, loc=<3.2456 9.3309 -1.2237>, rot=<0.0603 -0.4033 0.3318 -0.8507>}
 deform_pec_aim_l {parent=deform_pec_aim_base_l, loc=<0 0 0>, rot=<0 0 0 1>}
 deform_pec_base_l {parent=deform_pec_aim_l, loc=<9.3665 0 0>, rot=<0 0 0 1>}
 deform_pec_l {parent=deform_pec_base_l, loc=<0 0 0>, rot=<0 0 0 1>}
 deform_pec_aim_base_r {parent=spine_04, loc=<3.2452 9.3309 1.2237>, rot=<-0.4033 -0.0603 0.8507 0.3318>}
 deform_pec_aim_r {parent=deform_pec_aim_base_r, loc=<0 0 0>, rot=<0 0 0 1>}
 deform_pec_base_r {parent=deform_pec_aim_r, loc=<-9.3667 -0.0003 -0.0004>, rot=<0 0 0 1>}
 deform_pec_r {parent=deform_pec_base_r, loc=<0 0 0>, rot=<0 0 0 1>}
 thigh_l {parent=pelvis, loc=<-1.0855 -0.2109 -8.7053>, rot=<0.0009 0.0355 0.0155 -0.9993>}
 calf_l {parent=thigh_l, loc=<-42.5284 -0.0019 -0.7297>, rot=<0.0006 -0.0161 -0.0391 -0.9991>}
 foot_l {parent=calf_l, loc=<-46.7401 -2.5014 -0.592>, rot=<0.0427 -0.0238 0.0353 -0.9982>}
 ball_l {parent=foot_l, loc=<-3.5658 10.8066 -0.1544>, rot=<0.0001 -0.0001 -0.7063 -0.7079>}
 calf_twist_02_l {parent=calf_l, loc=<-19.1546 -1.7943 -0.2638>, rot=<0.0237 0.0095 -0.0035 -0.9997>}
 dyn_inner_lace_mstr_l {parent=calf_twist_02_l, loc=<-1.4125 3.5028 3.8235>, rot=<0.2487 0.1632 0.9434 0.1468>}
 dyn_inner_lace_01_l {parent=dyn_inner_lace_mstr_l, loc=<0 0 0>, rot=<0 0 0 1>}
 dyn_inner_lace_02_l {parent=dyn_inner_lace_01_l, loc=<5.9434 0 0>, rot=<0.0992 -0.0679 -0.0137 -0.9927>}
 dyn_inner_lace_03_l {parent=dyn_inner_lace_02_l, loc=<5.9434 0 0>, rot=<-0.0242 -0.0004 0.0153 -0.9996>}
 dyn_inner_lace_04_l {parent=dyn_inner_lace_03_l, loc=<5.9434 0 0>, rot=<0 0 0 1>}
 dyn_outer_lace_mstr_l {parent=calf_twist_02_l, loc=<-1.5963 4.3891 -4.2971>, rot=<-0.1737 0.0576 0.9646 0.1899>}
 dyn_outer_lace_01_l {parent=dyn_outer_lace_mstr_l, loc=<0 0 0>, rot=<0 0 0 1>}
 dyn_outer_lace_02_l {parent=dyn_outer_lace_01_l, loc=<4.6042 0 0>, rot=<-0.0076 0.031 0.0303 -0.999>}
 dyn_outer_lace_03_l {parent=dyn_outer_lace_02_l, loc=<4.6056 0 0>, rot=<0 0.0002 -0.0123 -0.9999>}
 dyn_outer_lace_04_l {parent=dyn_outer_lace_03_l, loc=<4.6067 0 0>, rot=<0.0052 0.0001 -0.057 -0.9984>}
 calf_twist_01_l {parent=calf_l, loc=<-35.8483 -1.9197 -0.5794>, rot=<0.0237 0.0095 -0.0035 -0.9997>}
 deform_knee_in_base_l {parent=calf_l, loc=<0.8031 -5.0231 0.2456>, rot=<0 0 0 -1>}
 deform_knee_in_l {parent=deform_knee_in_base_l, loc=<0 0 0>, rot=<0 0 0 1>}
 deform_knee_out_base_l {parent=calf_l, loc=<0 4 0>, rot=<0 0 0 -1>}
 deform_knee_out_l {parent=deform_knee_out_base_l, loc=<0 0 0>, rot=<0 0 0 1>}
 thigh_twist_01_l {parent=thigh_l, loc=<-27.952 0 0>, rot=<0.0013 0 0 -1>}
 deform_groin_base_l {parent=thigh_l, loc=<0.4801 5.6776 -0.2486>, rot=<0 0 0 -1>}
 deform_groin_l {parent=deform_groin_base_l, loc=<0 0 0>, rot=<0 0 0 1>}
 thigh_r {parent=pelvis, loc=<-1.0855 -0.2109 8.7053>, rot=<0.0355 -0.0009 0.9993 0.0155>}
 calf_r {parent=thigh_r, loc=<42.5284 0.0019 0.7297>, rot=<0.0006 -0.0161 -0.0391 -0.9991>}
 foot_r {parent=calf_r, loc=<46.7401 2.5014 0.592>, rot=<0.0427 -0.0238 0.0353 -0.9982>}
 ball_r {parent=foot_r, loc=<3.5658 -10.8066 0.1544>, rot=<0.0001 -0.0001 -0.7063 -0.7079>}
 calf_twist_02_r {parent=calf_r, loc=<19.1546 1.7943 0.2638>, rot=<0.0237 0.0095 -0.0035 -0.9997>}
 dyn_outer_lace_mstr_r {parent=calf_twist_02_r, loc=<1.5757 -4.3841 4.2724>, rot=<-0.2031 0.188 0.0486 -0.9597>}
 dyn_outer_lace_01_r {parent=dyn_outer_lace_mstr_r, loc=<0 0 0>, rot=<0 0 0 1>}
 dyn_outer_lace_02_r {parent=dyn_outer_lace_01_r, loc=<6.2729 0 0>, rot=<0.0793 -0.0649 0.0037 -0.9947>}
 dyn_outer_lace_03_r {parent=dyn_outer_lace_02_r, loc=<6.2729 0 0>, rot=<0.0235 -0.0375 0.0067 -0.999>}
 dyn_outer_lace_04_r {parent=dyn_outer_lace_03_r, loc=<6.3409 0 0>, rot=<-0.6245 0.0569 0.7759 -0.0692>}
 dyn_inner_lace_mstr_r {parent=calf_twist_02_r, loc=<1.6153 -3.3654 -3.8568>, rot=<-0.101 -0.1849 0.0923 -0.9732>}
 dyn_inner_lace_01_r {parent=dyn_inner_lace_mstr_r, loc=<0 0 0>, rot=<0 0 0 1>}
 dyn_inner_lace_02_r {parent=dyn_inner_lace_01_r, loc=<6.9771 0 0>, rot=<-0.0133 0.0851 -0.0065 -0.9963>}
 dyn_inner_lace_03_r {parent=dyn_inner_lace_02_r, loc=<6.9771 0 0>, rot=<-0.0003 0.0085 0.0386 -0.9992>}
 dyn_inner_lace_04_r {parent=dyn_inner_lace_03_r, loc=<7.0637 0 0>, rot=<0 0 0 1>}
 calf_twist_01_r {parent=calf_r, loc=<35.8483 1.9197 0.5794>, rot=<0.0237 0.0095 -0.0035 -0.9997>}
 deform_knee_in_base_r {parent=calf_r, loc=<-0.8031 5.0231 -0.2457>, rot=<0 0 0 -1>}
 deform_knee_in_r {parent=deform_knee_in_base_r, loc=<0 0 0>, rot=<0 0 0 1>}
 deform_knee_out_base_r {parent=calf_r, loc=<0 -4 0>, rot=<0 0 0 -1>}
 deform_knee_out_r {parent=deform_knee_out_base_r, loc=<0 0 0>, rot=<0 0 0 1>}
 thigh_twist_01_r {parent=thigh_r, loc=<27.952 0 0>, rot=<0.0013 0 0 -1>}
 deform_groin_base_r {parent=thigh_r, loc=<-0.4801 -5.6776 0.2486>, rot=<0 0 0 -1>}
 deform_groin_r {parent=deform_groin_base_r, loc=<0 0 0>, rot=<0 0 0 1>}
 deform_glute_base_l {parent=pelvis, loc=<-7.3466 -9.0213 -9.1245>, rot=<0.0172 0.031 -0.4477 -0.8935>}
 deform_glute_l {parent=deform_glute_base_l, loc=<0 0 0>, rot=<0 0 0 1>}
 deform_glute_base_r {parent=pelvis, loc=<-7.3464 -9.0213 9.1245>, rot=<0.031 -0.0172 0.8935 -0.4477>}
 deform_glute_r {parent=deform_glute_base_r, loc=<0 0 0>, rot=<0 0 0 1>}
 ring_01_mstr_l {parent=pelvis, loc=<-3.4618 7.3133 -8.4141>, rot=<0 0 0.9998 0.0179>}
 dyn_ring_01_l {parent=ring_01_mstr_l, loc=<0 0 0>, rot=<0 0 0 1>}
 ring_02_mstr_l {parent=pelvis, loc=<-8.7029 6.7934 -12.4325>, rot=<-0.0621 0.2816 0.9522 0.1006>}
 dyn_ring_02_l {parent=ring_02_mstr_l, loc=<0 0 0>, rot=<0 0 0 1>}
 ring_03_mstr_l {parent=pelvis, loc=<-9.3917 6.9898 -5.795>, rot=<-0.028 -0.1714 0.9843 0.0304>}
 dyn_ring_03_l {parent=ring_03_mstr_l, loc=<0 0 0>, rot=<0 0 0 1>}
 dyn_ring_mstr {parent=pelvis, loc=<-1.4815 8.42 0>, rot=<0 0 0.9998 0.0179>}
 dyn_ring {parent=dyn_ring_mstr, loc=<0 0 0>, rot=<0 0 0 1>}
 ring_01_mstr_r {parent=pelvis, loc=<-3.4618 7.3133 8.4141>, rot=<0 0 0.9998 0.0179>}
 dyn_ring_01_r {parent=ring_01_mstr_r, loc=<0 0 0>, rot=<0 0 0 1>}
 ring_02_mstr_r {parent=pelvis, loc=<-8.703 6.7934 12.4325>, rot=<0.0621 -0.2816 0.9522 0.1006>}
 dyn_ring_02_r {parent=ring_02_mstr_r, loc=<0 0 0>, rot=<0 0 0 1>}
 ring_03_mstr_r {parent=pelvis, loc=<-9.3917 6.9898 5.795>, rot=<0.028 0.1714 0.9843 0.0304>}
 dyn_ring_03_r {parent=ring_03_mstr_r, loc=<0 0 0>, rot=<0 0 0 1>}
 dfrm_necklace_03_l_ctrl_null {parent=pelvis, type=Null, loc=<29.8586 7.4845 -4.1395>, rot=<0.9412 -0.3129 -0.1219 -0.0359>}
 dfrm_necklace_03_l_ctrl {controlType=EULER_TRANSFORM, parent=dfrm_necklace_03_l_ctrl_null, type=Control, loc=<0 0 0>, rot=<0 0 0 1>}
 dfrm_necklace_05_l_ctrl_null {parent=pelvis, type=Null, loc=<34.8407 3.7511 -5.7597>, rot=<0.9479 -0.3116 -0.0656 0.0044>}
 dfrm_necklace_05_l_ctrl {controlType=EULER_TRANSFORM, parent=dfrm_necklace_05_l_ctrl_null, type=Control, loc=<0 0 0>, rot=<0 0 0 1>}
 dfrm_necklace_06_l_ctrl_null {parent=pelvis, type=Null, loc=<37.7744 1.575 -6.2073>, rot=<0.9479 -0.3116 -0.0656 0.0044>}
 dfrm_necklace_06_l_ctrl {controlType=EULER_TRANSFORM, parent=dfrm_necklace_06_l_ctrl_null, type=Control, loc=<0 0 0>, rot=<0 0 0 1>}
 dfrm_necklace_03_r_ctrl_null {parent=pelvis, type=Null, loc=<29.7806 7.5772 3.7452>, rot=<0.927 -0.3223 0.1428 0.1279>}
 dfrm_necklace_03_r_ctrl {controlType=EULER_TRANSFORM, parent=dfrm_necklace_03_r_ctrl_null, type=Control, loc=<0 0 0>, rot=<0 0 0 1>}
 dfrm_necklace_05_r_ctrl_null {parent=pelvis, type=Null, loc=<34.7066 3.8359 5.4747>, rot=<0.9508 -0.3002 0.05 0.0575>}
 dfrm_necklace_05_r_ctrl {controlType=EULER_TRANSFORM, parent=dfrm_necklace_05_r_ctrl_null, type=Control, loc=<0 0 0>, rot=<0 0 0 1>}
 dfrm_necklace_06_r_ctrl_null {parent=pelvis, type=Null, loc=<37.7065 1.7554 5.9521>, rot=<0.9508 -0.3002 0.05 0.0575>}
 dfrm_necklace_06_r_ctrl {controlType=EULER_TRANSFORM, parent=dfrm_necklace_06_r_ctrl_null, type=Control, loc=<0 0 0>, rot=<0 0 0 1>}
 dfrm_necklace_04_l_ctrl_null {parent=pelvis, type=Null, loc=<31.9469 5.8916 -4.9838>, rot=<0.9438 -0.3128 -0.0959 -0.0476>}
 dfrm_necklace_04_l_ctrl {controlType=EULER_TRANSFORM, parent=dfrm_necklace_04_l_ctrl_null, type=Control, loc=<0 0 0>, rot=<0 0 0 1>}
 dfrm_necklace_04_r_ctrl_null {parent=pelvis, type=Null, loc=<31.8538 6.0295 4.7027>, rot=<0.9397 -0.3235 0.0884 0.0673>}
 dfrm_necklace_04_r_ctrl {controlType=EULER_TRANSFORM, parent=dfrm_necklace_04_r_ctrl_null, type=Control, loc=<0 0 0>, rot=<0 0 0 1>}
]
