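# Physics asset for the necklace of the female medium Chainmail cosmetic body, defining capsule bodies for its dynamic bones.
# Constraints chain dyn_necklace_02 to dyn_neckace_01, and that to the small dyn_neckace_mstr root body.
# Both joints share tight angular limits, allowing modest swing and almost no twist.
# A large, unconstrained spine_04 capsule covers the upper torso, apparently as a collider for the necklace.
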PhysicsAsset /BRCosmetics/Characters/Player/Female/Medium/Bodies/F_MED_Chainmail/Meshes/F_MED_Chainmail_Necklace_Physics
bodies {
 dyn_neckace_01 [{shapeType=Capsule, length=4.5899, radius=2.2286, shapeName=dyn_neckace_01_capsule, center=<7.2769 -1.5244 0>}]
 dyn_neckace_mstr [{shapeType=Capsule, length=0, radius=0.5, shapeName=dyn_neckace_mstr_capsule}]
 dyn_necklace_02 [{shapeType=Capsule, length=2.7627, radius=1.1599, shapeName=dyn_necklace_02_capsule, center=<5.7243 0 0>}]
 spine_04 [{shapeType=Capsule, length=7.6972, radius=12.6362, shapeName=spine_04_capsule, center=<-3.2371 -0.0321 0>}]
}
constraints [
 {bone1Name=dyn_neckace_01, bone2Name=dyn_neckace_mstr, swing1LimitDegrees=20, swing2LimitDegrees=10, twistLimitDegrees=2}
 {bone1Name=dyn_necklace_02, bone2Name=dyn_neckace_01, swing1LimitDegrees=20, swing2LimitDegrees=10, twistLimitDegrees=2}
]
