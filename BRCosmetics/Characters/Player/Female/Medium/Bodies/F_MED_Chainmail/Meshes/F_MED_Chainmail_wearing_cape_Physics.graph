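# Physics asset for the cape-wearing variant of the female medium chainmail body, simulating its shirt and skirt bones; it has no cape bodies.
# Capsule chains ring the hips, running from tiny shirt root bodies through three skirt segments, with limited swing and tight twist.
# Free-angular constraints tie neighboring middle skirt segments from front to back on each side, leaving the front center unlinked.
# Pelvis, spine_02 and thigh capsules supply unconstrained body volumes.
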PhysicsAsset /BRCosmetics/Characters/Player/Female/Medium/Bodies/F_MED_Chainmail/Meshes/F_MED_Chainmail_wearing_cape_Physics
bodies {
 dyn_shirt_back_01 [{shapeType=Capsule, length=0, radius=0.5, shapeName=dyn_shirt_back_01_capsule}]
 dyn_shirt_back_02 [
  {
   shapeType=Capsule
   length=9.6563
   radius=2.276
   shapeName=dyn_skirt_back_side_01_l_capsule
   center=<2.6639 -0.9794 0>
   rotation=<rot 0.0002 0.0003 -10.9478>
  }
  {
   shapeType=Capsule
   length=9.6563
   radius=2.276
   shapeName=dyn_skirt_back_side_01_l_capsule
   center=<7.132 -0.9794 0>
   rotation=<rot 0.0002 0.0003 -10.9478>
  }
 ]
 dyn_shirt_back_side_01_l [{shapeType=Capsule, length=0, radius=0.5, shapeName=dyn_shirt_back_side_01_l_capsule}]
 dyn_shirt_back_side_01_r [{shapeType=Capsule, length=0, radius=0.5, shapeName=dyn_shirt_back_side_01_r_capsule}]
 dyn_shirt_back_side_02_l [
  {
   shapeType=Capsule
   length=7.9174
   radius=1.9594
   shapeName=dyn_shirt_back_side_02_l_capsule
   center=<2.5037 -0.9794 0>
   rotation=<rot 0 0.0001 -2.1594>
  }
  {
   shapeType=Capsule
   length=7.9174
   radius=1.9594
   shapeName=dyn_skirt_back_side_03_l_capsule
   center=<6.4316 -0.9794 0>
   rotation=<rot 0 0.0001 -2.1594>
  }
 ]
 dyn_shirt_back_side_02_r [
  {
   shapeType=Capsule
   length=7.9174
   radius=1.9594
   shapeName=dyn_shirt_back_side_02_l_capsule
   center=<2.5037 -0.9794 0>
   rotation=<rot 0 0.0001 -2.1594>
  }
  {
   shapeType=Capsule
   length=7.9174
   radius=1.9594
   shapeName=dyn_skirt_back_side_03_l_capsule
   center=<6.4316 -0.9794 0>
   rotation=<rot 0 0.0001 -2.1594>
  }
 ]
 dyn_shirt_front_01_l [{shapeType=Capsule, length=0, radius=0.5, shapeName=dyn_shirt_front_01_l_capsule}]
 dyn_shirt_front_01_r [{shapeType=Capsule, length=0, radius=0.5, shapeName=dyn_shirt_front_01_r_capsule}]
 dyn_shirt_front_02_l [
  {
   shapeType=Capsule
   length=7.8218
   radius=2.1325
   shapeName=dyn_skirt_front_03_l_capsule
   center=<4.6692 -1.203 0>
   rotation=<rot -90 166.8094 -170.1177>
  }
  {
   shapeType=Capsule
   length=7.8218
   radius=2.1325
   shapeName=dyn_skirt_front_03_l_capsule
   center=<4.7178 -0.9571 -3.2713>
   rotation=<rot -90 180 176.6917>
  }
 ]
 dyn_shirt_front_02_r [
  {
   shapeType=Capsule
   length=7.8218
   radius=2.1325
   shapeName=dyn_skirt_front_03_l_capsule
   center=<4.7093 -0.5087 0>
   rotation=<rot -90 8.8807 -12.1889>
  }
  {
   shapeType=Capsule
   length=7.8218
   radius=2.1325
   shapeName=dyn_skirt_front_03_l_capsule
   center=<4.6607 -1.9458 3.2713>
   rotation=<rot -90 180 176.6917>
  }
 ]
 dyn_shirt_front_side_01_l [{shapeType=Capsule, length=0, radius=0.5, shapeName=dyn_shirt_front_side_01_l_capsule}]
 dyn_shirt_front_side_01_r [{shapeType=Capsule, length=0, radius=0.5, shapeName=dyn_shirt_front_side_01_r_capsule}]
 dyn_shirt_front_side_02_l [
  {
   shapeType=Capsule
   length=5.369
   radius=2.1325
   shapeName=dyn_skirt_front_03_l_capsule
   center=<3.8408 -1.8944 0>
   rotation=<rot 90 180 -177.9631>
  }
  {
   shapeType=Capsule
   length=5.369
   radius=2.1325
   shapeName=dyn_skirt_front_03_l_capsule
   center=<3.8422 -1.9267 3.1464>
   rotation=<rot 90 -179.8881 -178.572>
  }
 ]
 dyn_shirt_front_side_02_r [
  {
   shapeType=Capsule
   length=5.369
   radius=2.1325
   shapeName=dyn_skirt_front_03_l_capsule
   center=<3.8408 -1.8944 0>
   rotation=<rot 90 180 -177.9631>
  }
  {
   shapeType=Capsule
   length=5.369
   radius=2.1325
   shapeName=dyn_skirt_front_03_l_capsule
   center=<3.8429 -1.8949 -3.1464>
   rotation=<rot 90 180 -178.684>
  }
 ]
 dyn_shirt_side_01_l [{shapeType=Capsule, length=0, radius=0.5, shapeName=dyn_shirt_side_01_l_capsule}]
 dyn_shirt_side_01_r [{shapeType=Capsule, length=0, radius=0.5, shapeName=dyn_shirt_side_01_r_capsule}]
 dyn_shirt_side_02_l [
  {shapeType=Capsule, length=9.9366, radius=2, shapeName=dyn_shirt_side_02_l_capsule, center=<3.1804 -0.5633 0>}
  {shapeType=Capsule, length=9.9366, radius=2, shapeName=dyn_skirt_side_02_l_capsule, center=<6.8342 -0.5633 0>}
 ]
 dyn_shirt_side_02_r [
  {shapeType=Capsule, length=9.9366, radius=2, shapeName=dyn_shirt_side_02_l_capsule, center=<3.1804 -0.5633 0>}
  {shapeType=Capsule, length=9.9366, radius=2, shapeName=dyn_skirt_side_02_l_capsule, center=<6.8342 -0.5633 0>}
 ]
 dyn_skirt_back_01 [
  {
   shapeType=Capsule
   length=8.947
   radius=2.9852
   shapeName=dyn_skirt_back_side_01_l_capsule
   center=<2.6639 -0.9794 0>
   rotation=<rot 0.0002 0.0003 -10.9478>
  }
  {
   shapeType=Capsule
   length=9.6563
   radius=2.276
   shapeName=dyn_skirt_back_side_01_l_capsule
   center=<7.132 -0.9794 0>
   rotation=<rot 0.0002 0.0003 -10.9478>
  }
 ]
 dyn_skirt_back_02 [
  {
   shapeType=Capsule
   length=9.6563
   radius=2.276
   shapeName=dyn_skirt_back_side_01_l_capsule
   center=<2.6639 -0.9794 0>
   rotation=<rot 0.0002 0.0003 -10.9478>
  }
  {
   shapeType=Capsule
   length=9.6563
   radius=2.276
   shapeName=dyn_skirt_back_side_01_l_capsule
   center=<7.132 -0.9794 0>
   rotation=<rot 0.0002 0.0003 -10.9478>
  }
 ]
 dyn_skirt_back_03 [
  {
   shapeType=Capsule
   length=9.6563
   radius=2.276
   shapeName=dyn_skirt_back_side_01_l_capsule
   center=<7.132 -0.9794 0>
   rotation=<rot 0.0002 0.0003 -10.9478>
  }
  {
   shapeType=Capsule
   length=9.6563
   radius=2.276
   shapeName=dyn_skirt_back_side_01_l_capsule
   center=<2.5858 -0.9794 0>
   rotation=<rot 0.0002 0.0003 -10.9478>
  }
 ]
 dyn_skirt_back_side_01_l [
  {
   shapeType=Capsule
   length=7.2081
   radius=2.6687
   shapeName=dyn_skirt_back_side_01_l_capsule
   center=<2.5037 -0.9794 0>
   rotation=<rot 0.0002 0.0003 -10.9478>
  }
  {
   shapeType=Capsule
   length=7.9174
   radius=1.9594
   shapeName=dyn_skirt_back_side_03_l_capsule
   center=<6.669 -0.9813 0.01>
   rotation=<rot 3.9453 -0.7637 -10.9742>
  }
 ]
 dyn_skirt_back_side_01_r [
  {
   shapeType=Capsule
   length=7.2081
   radius=2.6687
   shapeName=dyn_skirt_back_side_01_l_capsule
   center=<2.5037 -0.9794 0>
   rotation=<rot -0.0002 0.0003 10.9478>
  }
  {
   shapeType=Capsule
   length=7.9174
   radius=1.9594
   shapeName=dyn_skirt_back_side_03_l_capsule
   center=<6.669 -0.9813 0.01>
   rotation=<rot -3.9453 -0.7637 10.9742>
  }
 ]
 dyn_skirt_back_side_02_l [
  {
   shapeType=Capsule
   length=7.9174
   radius=1.9594
   shapeName=dyn_skirt_back_side_02_l_capsule
   center=<2.5037 -0.9794 0>
   rotation=<rot 5.5572 -0.7719 -7.9235>
  }
  {
   shapeType=Capsule
   length=7.6909
   radius=2.1859
   shapeName=dyn_skirt_back_side_03_l_capsule
   center=<6.4316 -0.9794 0>
   rotation=<rot 8.0334 -1.1199 -7.9647>
  }
 ]
 dyn_skirt_back_side_02_r [
  {
   shapeType=Capsule
   length=7.9174
   radius=1.9594
   shapeName=dyn_skirt_back_side_02_l_capsule
   center=<2.5037 -0.9794 0>
   rotation=<rot -5.5572 -0.7719 -7.9235>
  }
  {
   shapeType=Capsule
   length=7.6909
   radius=2.1859
   shapeName=dyn_skirt_back_side_03_l_capsule
   center=<6.4316 -0.9794 0>
   rotation=<rot -8.0334 -1.1199 -7.9647>
  }
 ]
 dyn_skirt_back_side_03_l [
  {
   shapeType=Capsule
   length=7.6707
   radius=2.2061
   shapeName=dyn_skirt_back_side_03_l_capsule
   center=<2.5037 -0.9794 0>
   rotation=<rot 10.2735 -1.4065 -7.8797>
  }
  {
   shapeType=Capsule
   length=7.6707
   radius=2.2061
   shapeName=dyn_skirt_back_side_03_l_capsule
   center=<6.4316 -0.9794 0>
   rotation=<rot 11.1077 -1.531 -7.9011>
  }
 ]
 dyn_skirt_back_side_03_r [
  {
   shapeType=Capsule
   length=7.6707
   radius=2.2061
   shapeName=dyn_skirt_back_side_03_l_capsule
   center=<2.5037 -0.9794 0>
   rotation=<rot -10.2735 -1.4065 -7.8797>
  }
  {
   shapeType=Capsule
   length=7.6707
   radius=2.2061
   shapeName=dyn_skirt_back_side_03_l_capsule
   center=<6.4316 -0.9794 0>
   rotation=<rot -11.1077 -1.531 -7.9011>
  }
 ]
 dyn_skirt_front_01_l [
  {
   shapeType=Capsule
   length=7.8218
   radius=2.1325
   shapeName=dyn_skirt_front_03_l_capsule
   center=<4.6254 -1.5838 0>
   rotation=<rot -90 180 -178.4829>
  }
  {
   shapeType=Capsule
   length=7.7719
   radius=2.1824
   shapeName=dyn_skirt_front_03_l_capsule
   center=<4.1035 -0.5591 -4.3204>
   rotation=<rot -90 180 -178.0579>
  }
 ]
 dyn_skirt_front_01_r [
  {
   shapeType=Capsule
   length=7.8218
   radius=2.1325
   shapeName=dyn_skirt_front_03_l_capsule
   center=<4.6293 -1.7312 0>
   rotation=<rot -90 -179.7016 -178.7813>
  }
  {
   shapeType=Capsule
   length=7.7719
   radius=2.1824
   shapeName=dyn_skirt_front_03_l_capsule
   center=<4.1085 -0.7065 2.8612>
   rotation=<rot -90 180 -178.0579>
  }
 ]
 dyn_skirt_front_02_l [
  {
   shapeType=Capsule
   length=9.0449
   radius=2.1325
   shapeName=dyn_skirt_front_03_l_capsule
   center=<4.6317 -1.2207 0>
   rotation=<rot -90 179.8813 -179.8818>
  }
  {
   shapeType=Capsule
   length=8.0703
   radius=2.234
   shapeName=dyn_skirt_front_03_l_capsule
   center=<4.2096 -0.8811 -3.8532>
   rotation=<rot -90 178.3124 -178.3126>
  }
 ]
 dyn_skirt_front_02_r [
  {
   shapeType=Capsule
   length=9.0449
   radius=2.1325
   shapeName=dyn_skirt_front_03_l_capsule
   center=<4.6317 -1.3681 0>
   rotation=<rot -90 14.899 -14.8995>
  }
  {
   shapeType=Capsule
   length=8.0703
   radius=2.234
   shapeName=dyn_skirt_front_03_l_capsule
   center=<4.2096 -1.0285 3.8532>
   rotation=<rot -90 90 -90.0002>
  }
 ]
 dyn_skirt_front_03_l [
  {
   shapeType=Capsule
   length=10.583
   radius=2.2836
   shapeName=dyn_skirt_front_03_l_capsule
   center=<4.6317 -1.4152 0>
   rotation=<rot -90 179.8416 -179.8418>
  }
  {
   shapeType=Capsule
   length=7.7623
   radius=2.3333
   shapeName=dyn_skirt_front_03_l_capsule
   center=<3.1688 -0.8778 -4.5201>
   rotation=<rot -90 144.7357 -144.7359>
  }
 ]
 dyn_skirt_front_03_r [
  {
   shapeType=Capsule
   length=10.583
   radius=2.2836
   shapeName=dyn_skirt_front_03_l_capsule
   center=<4.6317 -1.5626 0>
   rotation=<rot -90 179.8416 -179.8418>
  }
  {
   shapeType=Capsule
   length=7.7623
   radius=2.3333
   shapeName=dyn_skirt_front_03_l_capsule
   center=<3.1688 -1.0252 4.5201>
   rotation=<rot -90 179.9736 -179.9738>
  }
 ]
 dyn_skirt_front_side_01_l [{
   shapeType=Capsule
   length=8.5476
   radius=2.1325
   shapeName=dyn_skirt_front_03_l_capsule
   center=<5.4197 -1.7789 0>
   rotation=<rot 90 -165.6666 -168.575>
  }]
 dyn_skirt_front_side_01_r [{
   shapeType=Capsule
   length=8.5476
   radius=2.1325
   shapeName=dyn_skirt_front_03_l_capsule
   center=<5.4197 -1.7789 0>
   rotation=<rot 90 -165.6666 -168.575>
  }]
 dyn_skirt_front_side_02_l [
  {
   shapeType=Capsule
   length=8.5476
   radius=2.1325
   shapeName=dyn_skirt_front_03_l_capsule
   center=<2.7723 -0.3008 0>
   rotation=<rot -22.6428 -0.0001 -0.0001>
  }
  {
   shapeType=Capsule
   length=8.5476
   radius=2.1325
   shapeName=dyn_skirt_front_03_l_capsule
   center=<7.2292 -0.3008 0>
   rotation=<rot -27.1055 -0.0001 -0.0001>
  }
 ]
 dyn_skirt_front_side_02_r [
  {
   shapeType=Capsule
   length=8.5476
   radius=2.1325
   shapeName=dyn_skirt_front_03_l_capsule
   center=<2.7723 -0.3008 0>
   rotation=<rot 22.6428 -0.0001 -0.0001>
  }
  {
   shapeType=Capsule
   length=8.5476
   radius=2.1325
   shapeName=dyn_skirt_front_03_l_capsule
   center=<7.2292 -0.3008 0>
   rotation=<rot 27.1055 -0.0001 -0.0001>
  }
 ]
 dyn_skirt_front_side_03_l [
  {
   shapeType=Capsule
   length=8.5476
   radius=2.1325
   shapeName=dyn_skirt_front_03_l_capsule
   center=<2.5896 -0.3008 0>
   rotation=<rot -28.7042 -0.0001 -0.0001>
  }
  {
   shapeType=Capsule
   length=8.5476
   radius=2.1325
   shapeName=dyn_skirt_front_03_l_capsule
   center=<7.1744 -0.3008 0>
   rotation=<rot -27.9172 -0.0001 -0.0001>
  }
 ]
 dyn_skirt_front_side_03_r [
  {
   shapeType=Capsule
   length=8.5476
   radius=2.1325
   shapeName=dyn_skirt_front_03_l_capsule
   center=<2.5896 -0.3008 0>
   rotation=<rot 28.7042 -0.0001 -0.0001>
  }
  {
   shapeType=Capsule
   length=8.5476
   radius=2.1325
   shapeName=dyn_skirt_front_03_l_capsule
   center=<7.1744 -0.3008 0>
   rotation=<rot 27.9172 -0.0001 -0.0001>
  }
 ]
 dyn_skirt_side_01_l [
  {shapeType=Capsule, length=9.9366, radius=2, shapeName=dyn_skirt_side_01_l_capsule, center=<2.648 -0.5633 0>}
  {shapeType=Capsule, length=9.9366, radius=2, shapeName=dyn_skirt_side_02_l_capsule, center=<6.6716 -0.5633 0>}
 ]
 dyn_skirt_side_01_r [
  {shapeType=Capsule, length=9.9366, radius=2, shapeName=dyn_skirt_side_01_l_capsule, center=<2.648 -0.5633 0>}
  {shapeType=Capsule, length=9.9366, radius=2, shapeName=dyn_skirt_side_02_l_capsule, center=<6.6716 -0.5633 0>, rotation=<rot 6.6688 0 0>}
 ]
 dyn_skirt_side_02_l [
  {shapeType=Capsule, length=9.9366, radius=2, shapeName=dyn_skirt_side_02_l_capsule, center=<2.3554 -0.5633 0>}
  {shapeType=Capsule, length=9.9366, radius=2, shapeName=dyn_skirt_side_02_l_capsule, center=<6.3872 -0.5633 0>}
 ]
 dyn_skirt_side_02_r [
  {shapeType=Capsule, length=9.9366, radius=2, shapeName=dyn_skirt_side_03_l_capsule, center=<6.666 -0.5633 -0.1163>, rotation=<rot 8.6479 0 0>}
  {
   shapeType=Capsule
   length=9.9366
   radius=2
   shapeName=dyn_skirt_side_03_l_capsule
   center=<2.2108 0.5633 0>
   rotation=<rot 8.9187 0.0004 0.0003>
  }
 ]
 dyn_skirt_side_03_l [
  {
   shapeType=Capsule
   length=9.9366
   radius=2
   shapeName=dyn_skirt_side_03_l_capsule
   center=<6.666 -0.5633 -0.1163>
   rotation=<rot -8.6479 0 0>
  }
  {shapeType=Capsule, length=9.9366, radius=2, shapeName=dyn_skirt_side_03_l_capsule, center=<2.2108 -0.5633 0>}
 ]
 dyn_skirt_side_03_r [
  {shapeType=Capsule, length=9.9366, radius=2, shapeName=dyn_skirt_side_03_l_capsule, center=<6.666 -0.5633 -0.1163>, rotation=<rot 8.6479 0 0>}
  {
   shapeType=Capsule
   length=9.9366
   radius=2
   shapeName=dyn_skirt_side_03_l_capsule
   center=<2.2108 0.5633 0>
   rotation=<rot 8.9047 0.0001 0.0001>
  }
 ]
 pelvis [{shapeType=Capsule, length=11.7149, radius=11.8423, shapeName=pelvis_capsule, center=<-4.566 -0.0056 0>}]
 spine_02 [{shapeType=Capsule, length=19.6311, radius=10.3689, shapeName=spine_02_capsule, rotation=<rot -90 178.9849 -178.9851>}]
 thigh_l [{
   shapeType=Capsule
   length=25.1942
   radius=9.2569
   shapeName=thigh_l_capsule
   center=<-20.7626 -1.4828 1.0251>
   rotation=<rot 89.5225 -30.1182 -24.3819>
  }]
 thigh_r [{
   shapeType=Capsule
   length=25.1942
   radius=9.2569
   shapeName=thigh_l_capsule
   center=<20.7626 1.4828 -1.0251>
   rotation=<rot 89.5225 -30.1182 155.6181>
  }]
}
constraints [
 {bone1Name=dyn_skirt_front_01_l, bone2Name=dyn_shirt_front_02_l, swing1LimitDegrees=30, swing2LimitDegrees=30, twistLimitDegrees=2}
 {bone1Name=dyn_skirt_front_02_l, bone2Name=dyn_skirt_front_01_l, swing1LimitDegrees=30, swing2LimitDegrees=15, twistLimitDegrees=2}
 {bone1Name=dyn_skirt_front_03_l, bone2Name=dyn_skirt_front_02_l, swing1LimitDegrees=30, swing2LimitDegrees=10, twistLimitDegrees=2}
 {bone1Name=dyn_skirt_front_side_01_l, bone2Name=dyn_shirt_front_side_02_l, swing1LimitDegrees=20, swing2LimitDegrees=10, twistLimitDegrees=2}
 {bone1Name=dyn_skirt_front_side_03_l, bone2Name=dyn_skirt_front_side_02_l, swing1LimitDegrees=20, swing2LimitDegrees=10, twistLimitDegrees=2}
 {bone1Name=dyn_skirt_front_side_02_l, bone2Name=dyn_skirt_front_side_01_l, swing1LimitDegrees=20, swing2LimitDegrees=10, twistLimitDegrees=1}
 {bone1Name=dyn_skirt_side_03_l, bone2Name=dyn_skirt_side_02_l, swing1LimitDegrees=20, swing2LimitDegrees=10, twistLimitDegrees=2}
 {bone1Name=dyn_skirt_side_02_l, bone2Name=dyn_skirt_side_01_l, swing1LimitDegrees=20, swing2LimitDegrees=15, twistLimitDegrees=2}
 {bone1Name=dyn_skirt_side_01_l, bone2Name=dyn_shirt_side_02_l, swing1LimitDegrees=20, swing2LimitDegrees=10, twistLimitDegrees=2}
 {bone1Name=dyn_skirt_back_side_01_l, bone2Name=dyn_shirt_back_side_02_l, swing1LimitDegrees=10, swing2LimitDegrees=7.5, twistLimitDegrees=2}
 {bone1Name=dyn_skirt_back_side_02_l, bone2Name=dyn_skirt_back_side_01_l, swing1LimitDegrees=20, swing2LimitDegrees=5, twistLimitDegrees=2}
 {bone1Name=dyn_skirt_back_side_03_l, bone2Name=dyn_skirt_back_side_02_l, swing1LimitDegrees=10, swing2LimitDegrees=5, twistLimitDegrees=2}
 {bone1Name=dyn_skirt_back_01, bone2Name=dyn_shirt_back_02, swing1LimitDegrees=10, swing2LimitDegrees=7.5, twistLimitDegrees=2}
 {bone1Name=dyn_skirt_back_02, bone2Name=dyn_skirt_back_01, swing1LimitDegrees=20, swing2LimitDegrees=5, twistLimitDegrees=2}
 {bone1Name=dyn_skirt_back_03, bone2Name=dyn_skirt_back_02, swing1LimitDegrees=10, swing2LimitDegrees=5, twistLimitDegrees=2}
 {bone1Name=dyn_skirt_front_03_r, bone2Name=dyn_skirt_front_02_r, swing1LimitDegrees=20, swing2LimitDegrees=10, twistLimitDegrees=2}
 {bone1Name=dyn_skirt_front_02_r, bone2Name=dyn_skirt_front_01_r, swing1LimitDegrees=30, swing2LimitDegrees=15, twistLimitDegrees=2}
 {bone1Name=dyn_skirt_front_01_r, bone2Name=dyn_shirt_front_02_r, swing1LimitDegrees=30, swing2LimitDegrees=30, twistLimitDegrees=2}
 {bone1Name=dyn_skirt_front_side_01_r, bone2Name=dyn_shirt_front_side_02_r, swing1LimitDegrees=20, swing2LimitDegrees=10, twistLimitDegrees=2}
 {bone1Name=dyn_skirt_front_side_02_r, bone2Name=dyn_skirt_front_side_01_r, swing1LimitDegrees=20, swing2LimitDegrees=10, twistLimitDegrees=2}
 {bone1Name=dyn_skirt_front_side_03_r, bone2Name=dyn_skirt_front_side_02_r, swing1LimitDegrees=20, swing2LimitDegrees=10, twistLimitDegrees=2}
 {bone1Name=dyn_skirt_side_01_r, bone2Name=dyn_shirt_side_02_r, swing1LimitDegrees=20, swing2LimitDegrees=7.5, twistLimitDegrees=2}
 {bone1Name=dyn_skirt_side_02_r, bone2Name=dyn_skirt_side_01_r, swing1LimitDegrees=20, swing2LimitDegrees=10, twistLimitDegrees=2}
 {bone1Name=dyn_skirt_side_03_r, bone2Name=dyn_skirt_side_02_r, swing1LimitDegrees=20, swing2LimitDegrees=10, twistLimitDegrees=2}
 {bone1Name=dyn_skirt_back_side_01_r, bone2Name=dyn_shirt_back_side_02_r, swing1LimitDegrees=10, swing2LimitDegrees=7.5, twistLimitDegrees=2}
 {bone1Name=dyn_skirt_back_side_02_r, bone2Name=dyn_skirt_back_side_01_r, swing1LimitDegrees=20, swing2LimitDegrees=5, twistLimitDegrees=2}
 {bone1Name=dyn_skirt_back_side_03_r, bone2Name=dyn_skirt_back_side_02_r, swing1LimitDegrees=10, swing2LimitDegrees=5, twistLimitDegrees=2}
 {
  bone1Name=dyn_skirt_front_02_l
  bone2Name=dyn_skirt_front_side_02_l
  swing1LimitDegrees=45
  swing1Motion=Free
  swing2LimitDegrees=45
  swing2Motion=Free
  twistLimitDegrees=45
  twistMotion=Free
 }
 {
  bone1Name=dyn_skirt_front_side_02_l
  bone2Name=dyn_skirt_side_02_l
  swing1LimitDegrees=45
  swing1Motion=Free
  swing2LimitDegrees=45
  swing2Motion=Free
  twistLimitDegrees=45
  twistMotion=Free
 }
 {
  bone1Name=dyn_skirt_side_02_l
  bone2Name=dyn_skirt_back_side_02_l
  swing1LimitDegrees=45
  swing1Motion=Free
  swing2LimitDegrees=45
  swing2Motion=Free
  twistLimitDegrees=45
  twistMotion=Free
 }
 {
  bone1Name=dyn_skirt_back_side_02_l
  bone2Name=dyn_skirt_back_02
  swing1LimitDegrees=45
  swing1Motion=Free
  swing2LimitDegrees=45
  swing2Motion=Free
  twistLimitDegrees=45
  twistMotion=Free
 }
 {
  bone1Name=dyn_skirt_front_02_r
  bone2Name=dyn_skirt_front_side_02_r
  swing1LimitDegrees=45
  swing1Motion=Free
  swing2LimitDegrees=45
  swing2Motion=Free
  twistLimitDegrees=45
  twistMotion=Free
 }
 {
  bone1Name=dyn_skirt_front_side_02_r
  bone2Name=dyn_skirt_side_02_r
  swing1LimitDegrees=45
  swing1Motion=Free
  swing2LimitDegrees=45
  swing2Motion=Free
  twistLimitDegrees=45
  twistMotion=Free
 }
 {
  bone1Name=dyn_skirt_side_02_r
  bone2Name=dyn_skirt_back_side_02_r
  swing1LimitDegrees=45
  swing1Motion=Free
  swing2LimitDegrees=45
  swing2Motion=Free
  twistLimitDegrees=45
  twistMotion=Free
 }
 {
  bone1Name=dyn_skirt_back_side_02_r
  bone2Name=dyn_skirt_back_02
  swing1LimitDegrees=45
  swing1Motion=Free
  swing2LimitDegrees=45
  swing2Motion=Free
  twistLimitDegrees=45
  twistMotion=Free
 }
 {bone1Name=dyn_shirt_side_02_r, bone2Name=dyn_shirt_side_01_r, swing1LimitDegrees=15, swing2LimitDegrees=5, twistLimitDegrees=2}
 {bone1Name=dyn_shirt_side_02_l, bone2Name=dyn_shirt_side_01_l, swing1LimitDegrees=15, swing2LimitDegrees=5, twistLimitDegrees=2}
 {bone1Name=dyn_shirt_front_side_02_r, bone2Name=dyn_shirt_front_side_01_r, swing1LimitDegrees=15, swing2LimitDegrees=10, twistLimitDegrees=2}
 {bone1Name=dyn_shirt_front_side_02_l, bone2Name=dyn_shirt_front_side_01_l, swing1LimitDegrees=25, swing2LimitDegrees=20, twistLimitDegrees=2}
 {bone1Name=dyn_shirt_front_02_r, bone2Name=dyn_shirt_front_01_r, swing1LimitDegrees=25, swing2LimitDegrees=20, twistLimitDegrees=2}
 {bone1Name=dyn_shirt_front_02_l, bone2Name=dyn_shirt_front_01_l, swing1LimitDegrees=25, swing2LimitDegrees=20, twistLimitDegrees=2}
 {bone1Name=dyn_shirt_back_side_02_r, bone2Name=dyn_shirt_back_side_01_r, swing1LimitDegrees=15, swing2LimitDegrees=7.5, twistLimitDegrees=2}
 {bone1Name=dyn_shirt_back_side_02_l, bone2Name=dyn_shirt_back_side_01_l, swing1LimitDegrees=15, swing2LimitDegrees=7.5, twistLimitDegrees=2}
 {bone1Name=dyn_shirt_back_02, bone2Name=dyn_shirt_back_01, swing1LimitDegrees=15, swing2LimitDegrees=7.5, twistLimitDegrees=2}
]
