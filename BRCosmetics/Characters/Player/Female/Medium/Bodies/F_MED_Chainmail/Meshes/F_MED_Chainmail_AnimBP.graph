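# Post-process-style AnimBlueprint for the female medium Chainmail body cosmetic, adding secondary motion and deformation to the pose copied from the parent mesh.
# A deform Control Rig feeds weighted bone constraints for rings, necklace and shirt, then RigidBody necklace physics, AnimDynamics rings, laces and sleeves, and body and necklace Control Rigs.
# The cached result enters one of two RigidBody shirt simulations, selected by whether the backbling is a cape.
# Updates refresh gravity override and pelvis-relative spine twist alpha.
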AnimBlueprint /BRCosmetics/Characters/Player/Female/Medium/Bodies/F_MED_Chainmail/Meshes/F_MED_Chainmail_AnimBP
target_skeleton=&/F_MED_Chainmail_Skeleton
graphs [
 AnimGraph {nodes [
   n1 = RigidBody!(ComponentPose=Local To Component(LocalPose=Use cached pose 'pre_sim_cache'()), Alpha, OverrideWorldGravity:Vector=Get Gravity Override()) -> (Pose) {
    bClampLinearTranslationLimitToRefPose=true
    bOverrideWorldGravity=true
    overridePhysicsAsset=&/F_MED_Chainmail_shirt_Physics
    simulationSpace=BaseBoneSpace
    alphaBoolBlend {blendOption=Linear}
    baseBoneRef {boneName=pelvis}
    componentAppliedLinearAccClamp=<2500 2500 2500>
    componentLinearAccScale=<0.4 0.3 0.1>
    componentLinearVelScale=<0.4 0.3 0.1>
    overrideWorldGravity {z=-980}
    simSpaceSettings {maxAngularAcceleration=1000, maxAngularVelocity=3, maxLinearAcceleration=250, maxLinearVelocity=500, velocityScaleZ=0, worldAlpha=1}
   }
   n2 = RigidBody(ComponentPose=Local To Component(LocalPose=Use cached pose 'pre_sim_cache'()), OverrideWorldGravity=Get Gravity Override()) {
    bClampLinearTranslationLimitToRefPose=true
    bOverrideWorldGravity=true
    overridePhysicsAsset=&/F_MED_Chainmail_wearing_cape_Physics
    simulationSpace=BaseBoneSpace
    alphaBoolBlend {blendOption=Linear}
    baseBoneRef {boneName=pelvis}
    componentAppliedLinearAccClamp=<2500 2500 2500>
    componentLinearAccScale=<0.4 0.3 0.1>
    componentLinearVelScale=<0.4 0.3 0.1>
    overrideWorldGravity {z=-980}
    simSpaceSettings {maxAngularAcceleration=1000, maxAngularVelocity=3, maxLinearAcceleration=250, maxLinearVelocity=500, velocityScaleZ=0, worldAlpha=1}
   }
   Output Pose(Result=Blend Poses by bool(
     BlendPose_0=Component To Local(ComponentPose=@n2)
     BlendPose_1=Component To Local(ComponentPose=@n1)
     bActiveValue=Get is_wearing_cape() -> (is_wearing_cape:bool)
    ) {blendType=Linear}) {layerGroup=None}
   n3 = Control Rig(Source=Copy Pose From Mesh() {bUseAttachedParent=true}) {
    alphaBoolBlend {blendOption=Linear}
    controlRigAssetReference {blueprintRigClass=&/F_MED_Chainmail_Deform_ControlRig.F_MED_Chainmail_Deform_ControlRig_C}
   }
   n4 = "Constraint - ring_01_mstr_l (PPP)"!(
    ComponentPose=Local To Component(LocalPose=@n3)
    Alpha=1
    ConstraintWeights_0=0.3561
    ConstraintWeights_1:float=0.3538
    ConstraintWeights_2:float=0.2901
   ) -> (Pose) {
    alphaBoolBlend {blendOption=Linear}
    boneToModify {boneName=ring_01_mstr_l}
    constraintSetup [{targetBone {boneName=thigh_l}}, {targetBone {boneName=pelvis}}, {targetBone {boneName=deform_groin_l}}]
    constraintWeights [1,1,1]
   }
   n5 = "Constraint - ring_02_mstr_l (PPP)"!(ComponentPose=@n4, Alpha=1, ConstraintWeights_0=0.7207, ConstraintWeights_1:float=0.1508, ConstraintWeights_2:float=0.1285) -> (Pose) {
    alphaBoolBlend {blendOption=Linear}
    boneToModify {boneName=ring_02_mstr_l}
    constraintSetup [{targetBone {boneName=thigh_l}}, {targetBone {boneName=deform_groin_l}}, {targetBone {boneName=pelvis}}]
    constraintWeights [1,1,1]
   }
   n6 = "Constraint - ring_03_mstr_l (PPP)"!(ComponentPose=@n5, Alpha=1, ConstraintWeights_0=0.5976, ConstraintWeights_1:float=0.2067, ConstraintWeights_2:float=0.1957) -> (Pose) {
    alphaBoolBlend {blendOption=Linear}
    boneToModify {boneName=ring_03_mstr_l}
    constraintSetup [{targetBone {boneName=thigh_l}}, {targetBone {boneName=deform_groin_l}}, {targetBone {boneName=pelvis}}]
    constraintWeights [1,1,1]
   }
   n7 = "Constraint - dyn_ring_mstr (P)"!(ComponentPose=@n6, Alpha=1, ConstraintWeights_0=1, ConstraintWeights_1:float=0) -> (Pose) {
    alphaBoolBlend {blendOption=Linear}
    boneToModify {boneName=dyn_ring_mstr}
    constraintSetup [{targetBone {boneName=pelvis}}]
    constraintWeights [1,0]
   }
   n8 = "Constraint - ring_01_mstr_r (PPP)"!(ComponentPose=@n7, Alpha=1, ConstraintWeights_0=0.3561, ConstraintWeights_1:float=0.3538, ConstraintWeights_2:float=0.2901) -> (Pose) {
    alphaBoolBlend {blendOption=Linear}
    boneToModify {boneName=ring_01_mstr_r}
    constraintSetup [{targetBone {boneName=thigh_r}}, {targetBone {boneName=pelvis}}, {targetBone {boneName=deform_groin_r}}]
    constraintWeights [1,1,1]
   }
   n9 = "Constraint - ring_02_mstr_r (PPP)"!(ComponentPose=@n8, Alpha=1, ConstraintWeights_0=0.7207, ConstraintWeights_1:float=0.1508, ConstraintWeights_2:float=0.1285) -> (Pose) {
    alphaBoolBlend {blendOption=Linear}
    boneToModify {boneName=ring_02_mstr_r}
    constraintSetup [{targetBone {boneName=thigh_r}}, {targetBone {boneName=deform_groin_r}}, {targetBone {boneName=pelvis}}]
    constraintWeights [1,1,1]
   }
   n10 = "Constraint - ring_03_mstr_r (PPP)"!(ComponentPose=@n9, Alpha=1, ConstraintWeights_0=0.5976, ConstraintWeights_1:float=0.2067, ConstraintWeights_2:float=0.1957) -> (Pose) {
    alphaBoolBlend {blendOption=Linear}
    boneToModify {boneName=ring_03_mstr_r}
    constraintSetup [{targetBone {boneName=thigh_r}}, {targetBone {boneName=deform_groin_r}}, {targetBone {boneName=pelvis}}]
    constraintWeights [1,1,1]
   }
   n11 = Anim Dynamics - Bone: dyn_ring_01_l(ComponentPose=@n10, Alpha=1, ExternalForce="0,0,0") {
    angularDampingOverride=0.9999
    angularSpringConstant=750
    bAngularSpring=true
    bGravityOverrideInSimSpace=true
    bOverrideAngularDamping=true
    simulationSpace=BoneRelative
    alphaBoolBlend {blendOption=Linear}
    boundBone {boneName=dyn_ring_01_l}
    componentAppliedLinearAccClamp=<5000 5000 5000>
    componentLinearAccScale=<1.2 1.2 1.2>
    componentLinearVelScale=<1.2 1.2 1.2>
    physicsBodyDefinitions [{
      sphereCollisionRadius=0
      boundBone {boneName=dyn_ring_01_l}
      boxExtents=<12.5 12.5 12.5>
      constraintSetup {angularLimitsMax=<1 2.5 0>, angularLimitsMin=<-1 -2.5 -25>, angularTarget=<1 0 0>}
      localJointOffset=<1.5 0.1 0>
     }]
    relativeSpaceBone {boneName=root}
   }
   n12 = Anim Dynamics - Bone: dyn_ring_02_l(ComponentPose=@n11, Alpha=1, ExternalForce="0,0,0") {
    angularDampingOverride=0.9999
    angularSpringConstant=750
    bAngularSpring=true
    bGravityOverrideInSimSpace=true
    bOverrideAngularDamping=true
    simulationSpace=BoneRelative
    alphaBoolBlend {blendOption=Linear}
    boundBone {boneName=dyn_ring_02_l}
    componentAppliedLinearAccClamp=<5000 5000 5000>
    componentLinearAccScale=<1.2 1.2 1.2>
    componentLinearVelScale=<1.2 1.2 1.2>
    physicsBodyDefinitions [{
      sphereCollisionRadius=0
      boundBone {boneName=dyn_ring_02_l}
      boxExtents=<12.25 12.25 12.25>
      constraintSetup {angularLimitsMax=<1 2.5 0>, angularLimitsMin=<-1 -2.5 -25>, angularTarget=<1 0 0>}
      localJointOffset=<1.45 0.1 0>
     }]
    relativeSpaceBone {boneName=root}
   }
   n13 = Anim Dynamics - Bone: dyn_ring_03_l(ComponentPose=@n12, Alpha=1, ExternalForce="0,0,0") {
    angularDampingOverride=0.9999
    angularSpringConstant=750
    bAngularSpring=true
    bGravityOverrideInSimSpace=true
    bOverrideAngularDamping=true
    simulationSpace=BoneRelative
    alphaBoolBlend {blendOption=Linear}
    boundBone {boneName=dyn_ring_03_l}
    componentAppliedLinearAccClamp=<5000 5000 5000>
    componentLinearAccScale=<1.2 1.2 1.2>
    componentLinearVelScale=<1.2 1.2 1.2>
    physicsBodyDefinitions [{
      sphereCollisionRadius=0
      boundBone {boneName=dyn_ring_03_l}
      boxExtents=<12.6 12.6 12.6>
      constraintSetup {angularLimitsMax=<1 2.5 0>, angularLimitsMin=<-1 -2.5 -25>, angularTarget=<1 0 0>}
      localJointOffset=<1.55 0.1 0>
     }]
    relativeSpaceBone {boneName=root}
   }
   n14 = Anim Dynamics - Bone: dyn_ring(ComponentPose=@n13, Alpha=1, ExternalForce="0,0,0") {
    angularDampingOverride=0.9999
    angularSpringConstant=750
    bAngularSpring=true
    bGravityOverrideInSimSpace=true
    bOverrideAngularDamping=true
    simulationSpace=BoneRelative
    alphaBoolBlend {blendOption=Linear}
    boundBone {boneName=dyn_ring}
    componentAppliedLinearAccClamp=<5000 5000 5000>
    componentLinearAccScale=<1.2 1.2 1.2>
    componentLinearVelScale=<1.2 1.2 1.2>
    physicsBodyDefinitions [{
      sphereCollisionRadius=0
      boundBone {boneName=dyn_ring}
      boxExtents=<12.5 12.5 12.5>
      constraintSetup {angularLimitsMax=<1 2.5 0>, angularLimitsMin=<-1 -2.5 -25>, angularTarget=<1 0 0>}
      localJointOffset=<1.5 0.1 0>
     }]
    relativeSpaceBone {boneName=root}
   }
   n15 = Anim Dynamics - Bone: dyn_ring_01_r(ComponentPose=@n14, Alpha=1, ExternalForce="0,0,0") {
    angularDampingOverride=0.9999
    angularSpringConstant=750
    bAngularSpring=true
    bGravityOverrideInSimSpace=true
    bOverrideAngularDamping=true
    simulationSpace=BoneRelative
    alphaBoolBlend {blendOption=Linear}
    boundBone {boneName=dyn_ring_01_r}
    componentAppliedLinearAccClamp=<5000 5000 5000>
    componentLinearAccScale=<1.2 1.2 1.2>
    componentLinearVelScale=<1.2 1.2 1.2>
    physicsBodyDefinitions [{
      sphereCollisionRadius=0
      boundBone {boneName=dyn_ring_01_r}
      boxExtents=<12.2 12.2 12.2>
      constraintSetup {angularLimitsMax=<1 2.5 0>, angularLimitsMin=<-1 -2.5 -25>, angularTarget=<1 0 0>}
      localJointOffset=<1.4 0.1 0>
     }]
    relativeSpaceBone {boneName=root}
   }
   n16 = Anim Dynamics - Bone: dyn_ring_02_r(ComponentPose=@n15, Alpha=1, ExternalForce="0,0,0") {
    angularDampingOverride=0.9999
    angularSpringConstant=750
    bAngularSpring=true
    bGravityOverrideInSimSpace=true
    bOverrideAngularDamping=true
    simulationSpace=BoneRelative
    alphaBoolBlend {blendOption=Linear}
    boundBone {boneName=dyn_ring_02_r}
    componentAppliedLinearAccClamp=<5000 5000 5000>
    componentLinearAccScale=<1.2 1.2 1.2>
    componentLinearVelScale=<1.2 1.2 1.2>
    physicsBodyDefinitions [{
      sphereCollisionRadius=0
      boundBone {boneName=dyn_ring_02_r}
      boxExtents=<12.5 12.5 12.5>
      constraintSetup {angularLimitsMax=<1 2.5 0>, angularLimitsMin=<-1 -2.5 -25>, angularTarget=<1 0 0>}
      localJointOffset=<1.5 0.1 0>
     }]
    relativeSpaceBone {boneName=root}
   }
   n17 = Anim Dynamics - Bone: dyn_ring_03_r(ComponentPose=@n16, Alpha=1, ExternalForce="0,0,0") {
    angularDampingOverride=0.9999
    angularSpringConstant=750
    bAngularSpring=true
    bGravityOverrideInSimSpace=true
    bOverrideAngularDamping=true
    simulationSpace=BoneRelative
    alphaBoolBlend {blendOption=Linear}
    boundBone {boneName=dyn_ring_03_r}
    componentAppliedLinearAccClamp=<5000 5000 5000>
    componentLinearAccScale=<1.2 1.2 1.2>
    componentLinearVelScale=<1.2 1.2 1.2>
    physicsBodyDefinitions [{
      sphereCollisionRadius=0
      boundBone {boneName=dyn_ring_03_r}
      boxExtents=<12.6 12.6 12.6>
      constraintSetup {angularLimitsMax=<1 2.5 0>, angularLimitsMin=<-1 -2.5 -25>, angularTarget=<1 0 0>}
      localJointOffset=<1.54 0.1 0>
     }]
    relativeSpaceBone {boneName=root}
   }
   n18 = "Constraint - dfrm_sphere_necklace_mstr_01 (PP)"!(ComponentPose=@n17, Alpha=1, ConstraintWeights_0=0.8233, ConstraintWeights_1:float=0.1767) -> (Pose) {
    alphaBoolBlend {blendOption=Linear}
    boneToModify {boneName=dfrm_sphere_necklace_mstr_01}
    constraintSetup [{targetBone {boneName=spine_05}}, {targetBone {boneName=spine_04}}]
    constraintWeights [1,1]
   }
   n19 = "Constraint - dfrm_sphere_necklace_mstr_02 (PPPP)"!(
    ComponentPose=@n18
    Alpha=1
    ConstraintWeights_0=0.4667
    ConstraintWeights_1:float=0.2902
    ConstraintWeights_2:float=0.1885
    ConstraintWeights_3:float=0.0545
   ) -> (Pose) {
    alphaBoolBlend {blendOption=Linear}
    boneToModify {boneName=dfrm_sphere_necklace_mstr_02}
    constraintSetup [{targetBone {boneName=spine_05}}, {targetBone {boneName=spine_04}}, {targetBone {boneName=deform_pec_l}}, {targetBone {boneName=clavicle_l}}]
    constraintWeights [1,1,1,1]
   }
   n20 = "Constraint - dfrm_sphere_necklace_mstr_03 (PPPPPP)"!(
    ComponentPose=@n19
    Alpha=1
    ConstraintWeights_0=0.4007
    ConstraintWeights_1:float=0.3026
    ConstraintWeights_2:float=0.1413
    ConstraintWeights_3:float=0.0838
    ConstraintWeights_4:float=0.0368
    ConstraintWeights_5:float=0.0349
   ) -> (Pose) {
    alphaBoolBlend {blendOption=Linear}
    boneToModify {boneName=dfrm_sphere_necklace_mstr_03}
    constraintSetup [
     {targetBone {boneName=spine_05}}
     {targetBone {boneName=clavicle_l}}
     {targetBone {boneName=spine_04}}
     {targetBone {boneName=deform_pec_l}}
     {targetBone {boneName=deform_upperarm_fr_l}}
     {targetBone {boneName=upperarm_twist_01_l}}
    ]
    constraintWeights [1,1,1,1,1,1]
   }
   n21 = "Constraint - dfrm_sphere_necklace_mstr_04 (PPPPPPP)"!(
    ComponentPose=@n20
    Alpha=1
    ConstraintWeights_0=0.4375
    ConstraintWeights_1:float=0.3748
    ConstraintWeights_2:float=0.0981
    ConstraintWeights_3:float=0.0309
    ConstraintWeights_4:float=0.0283
    ConstraintWeights_5:float=0.0269
    ConstraintWeights_6:float=0.0035
   ) -> (Pose) {
    alphaBoolBlend {blendOption=Linear}
    boneToModify {boneName=dfrm_sphere_necklace_mstr_04}
    constraintSetup [
     {targetBone {boneName=clavicle_l}}
     {targetBone {boneName=spine_05}}
     {targetBone {boneName=spine_04}}
     {targetBone {boneName=upperarm_twist_01_l}}
     {targetBone {boneName=deform_pec_l}}
     {targetBone {boneName=deform_upperarm_fr_l}}
     {targetBone {boneName=deform_upperarm_up_l}}
    ]
    constraintWeights [1,1,1,1,1,1,1]
   }
   n22 = "Constraint - dfrm_sphere_necklace_mstr_05 (PPPPP)"!(
    ComponentPose=@n21
    Alpha=1
    ConstraintWeights_0=0.6876
    ConstraintWeights_1:float=0.2859
    ConstraintWeights_2:float=0.0222
    ConstraintWeights_3:float=0.0028
    ConstraintWeights_4:float=0.0015
   ) -> (Pose) {
    alphaBoolBlend {blendOption=Linear}
    boneToModify {boneName=dfrm_sphere_necklace_mstr_05}
    constraintSetup [
     {targetBone {boneName=clavicle_l}}
     {targetBone {boneName=spine_05}}
     {targetBone {boneName=spine_04}}
     {targetBone {boneName=upperarm_twist_01_l}}
     {targetBone {boneName=deform_upperarm_up_l}}
    ]
    constraintWeights [1,1,1,1,1]
   }
   n23 = "Constraint - dfrm_sphere_necklace_mstr_06 (PPPPPP)"!(
    ComponentPose=@n22
    Alpha=1
    ConstraintWeights_0=0.4977
    ConstraintWeights_1:float=0.4633
    ConstraintWeights_2:float=0.0194
    ConstraintWeights_3:float=0.0107
    ConstraintWeights_4:float=0.0074
    ConstraintWeights_5:float=0.0015
   ) -> (Pose) {
    alphaBoolBlend {blendOption=Linear}
    boneToModify {boneName=dfrm_sphere_necklace_mstr_06}
    constraintSetup [
     {targetBone {boneName=clavicle_l}}
     {targetBone {boneName=spine_05}}
     {targetBone {boneName=neck_01}}
     {targetBone {boneName=spine_03}}
     {targetBone {boneName=neck_02}}
     {targetBone {boneName=spine_04}}
    ]
    constraintWeights [1,1,1,1,1,1]
   }
   n24 = "Constraint - dfrm_sphere_necklace_mstr_07 (PPPP)"!(
    ComponentPose=@n23
    Alpha=1
    ConstraintWeights_0=0.6708
    ConstraintWeights_1:float=0.2742
    ConstraintWeights_2:float=0.0286
    ConstraintWeights_3:float=0.0264
   ) -> (Pose) {
    alphaBoolBlend {blendOption=Linear}
    boneToModify {boneName=dfrm_sphere_necklace_mstr_07}
    constraintSetup [{targetBone {boneName=spine_05}}, {targetBone {boneName=clavicle_l}}, {targetBone {boneName=spine_03}}, {targetBone {boneName=neck_01}}]
    constraintWeights [1,1,1,1]
   }
   n25 = "Constraint - dfrm_sphere_necklace_mstr_08 (PPPPPP)"!(
    ComponentPose=@n24
    Alpha=1
    ConstraintWeights_0=0.7956
    ConstraintWeights_1:float=0.0598
    ConstraintWeights_2:float=0.0547
    ConstraintWeights_3:float=0.0412
    ConstraintWeights_4:float=0.0349
    ConstraintWeights_5:float=0.0138
   ) -> (Pose) {
    alphaBoolBlend {blendOption=Linear}
    boneToModify {boneName=dfrm_sphere_necklace_mstr_08}
    constraintSetup [
     {targetBone {boneName=spine_05}}
     {targetBone {boneName=spine_03}}
     {targetBone {boneName=neck_01}}
     {targetBone {boneName=clavicle_r}}
     {targetBone {boneName=clavicle_l}}
     {targetBone {boneName=neck_02}}
    ]
    constraintWeights [1,1,1,1,1,1]
   }
   n26 = "Constraint - dfrm_sphere_necklace_mstr_09 (PPPP)"!(
    ComponentPose=@n25
    Alpha=1
    ConstraintWeights_0=0.6708
    ConstraintWeights_1:float=0.2742
    ConstraintWeights_2:float=0.0286
    ConstraintWeights_3:float=0.0264
   ) -> (Pose) {
    alphaBoolBlend {blendOption=Linear}
    boneToModify {boneName=dfrm_sphere_necklace_mstr_09}
    constraintSetup [{targetBone {boneName=spine_05}}, {targetBone {boneName=clavicle_r}}, {targetBone {boneName=spine_03}}, {targetBone {boneName=neck_01}}]
    constraintWeights [1,1,1,1]
   }
   n27 = "Constraint - dfrm_sphere_necklace_mstr_10 (PPPPPP)"!(
    ComponentPose=@n26
    Alpha=1
    ConstraintWeights_0=0.4977
    ConstraintWeights_1:float=0.4633
    ConstraintWeights_2:float=0.0194
    ConstraintWeights_3:float=0.0107
    ConstraintWeights_4:float=0.0074
    ConstraintWeights_5:float=0.0015
   ) -> (Pose) {
    alphaBoolBlend {blendOption=Linear}
    boneToModify {boneName=dfrm_sphere_necklace_mstr_10}
    constraintSetup [
     {targetBone {boneName=clavicle_r}}
     {targetBone {boneName=spine_05}}
     {targetBone {boneName=neck_01}}
     {targetBone {boneName=spine_03}}
     {targetBone {boneName=neck_02}}
     {targetBone {boneName=spine_04}}
    ]
    constraintWeights [1,1,1,1,1,1]
   }
   n28 = "Constraint - dfrm_sphere_necklace_mstr_11 (PPPPP)"!(
    ComponentPose=@n27
    Alpha=1
    ConstraintWeights_0=0.6876
    ConstraintWeights_1:float=0.2859
    ConstraintWeights_2:float=0.0222
    ConstraintWeights_3:float=0.0028
    ConstraintWeights_4:float=0.0015
   ) -> (Pose) {
    alphaBoolBlend {blendOption=Linear}
    boneToModify {boneName=dfrm_sphere_necklace_mstr_11}
    constraintSetup [
     {targetBone {boneName=clavicle_r}}
     {targetBone {boneName=spine_05}}
     {targetBone {boneName=spine_04}}
     {targetBone {boneName=upperarm_twist_01_r}}
     {targetBone {boneName=deform_upperarm_up_r}}
    ]
    constraintWeights [1,1,1,1,1]
   }
   n29 = "Constraint - dfrm_sphere_necklace_mstr_12 (PPPPPPP)"!(
    ComponentPose=@n28
    Alpha=1
    ConstraintWeights_0=0.4375
    ConstraintWeights_1:float=0.3748
    ConstraintWeights_2:float=0.0981
    ConstraintWeights_3:float=0.0309
    ConstraintWeights_4:float=0.0283
    ConstraintWeights_5:float=0.0269
    ConstraintWeights_6:float=0.0035
   ) -> (Pose) {
    alphaBoolBlend {blendOption=Linear}
    boneToModify {boneName=dfrm_sphere_necklace_mstr_12}
    constraintSetup [
     {targetBone {boneName=clavicle_r}}
     {targetBone {boneName=spine_05}}
     {targetBone {boneName=spine_04}}
     {targetBone {boneName=upperarm_twist_01_r}}
     {targetBone {boneName=deform_pec_r}}
     {targetBone {boneName=deform_upperarm_fr_r}}
     {targetBone {boneName=deform_upperarm_up_r}}
    ]
    constraintWeights [1,1,1,1,1,1,1]
   }
   n30 = "Constraint - dfrm_sphere_necklace_mstr_13 (PPPPPP)"!(
    ComponentPose=@n29
    Alpha=1
    ConstraintWeights_0=0.4007
    ConstraintWeights_1:float=0.3026
    ConstraintWeights_2:float=0.1413
    ConstraintWeights_3:float=0.0838
    ConstraintWeights_4:float=0.0368
    ConstraintWeights_5:float=0.0349
   ) -> (Pose) {
    alphaBoolBlend {blendOption=Linear}
    boneToModify {boneName=dfrm_sphere_necklace_mstr_13}
    constraintSetup [
     {targetBone {boneName=spine_05}}
     {targetBone {boneName=clavicle_r}}
     {targetBone {boneName=spine_04}}
     {targetBone {boneName=deform_pec_r}}
     {targetBone {boneName=deform_upperarm_fr_r}}
     {targetBone {boneName=upperarm_twist_01_r}}
    ]
    constraintWeights [1,1,1,1,1,1]
   }
   n31 = "Constraint - dfrm_sphere_necklace_mstr_14 (PPPP)"!(
    ComponentPose=@n30
    Alpha=1
    ConstraintWeights_0=0.4667
    ConstraintWeights_1:float=0.2902
    ConstraintWeights_2:float=0.1885
    ConstraintWeights_3:float=0.0545
   ) -> (Pose) {
    alphaBoolBlend {blendOption=Linear}
    boneToModify {boneName=dfrm_sphere_necklace_mstr_14}
    constraintSetup [{targetBone {boneName=spine_05}}, {targetBone {boneName=spine_04}}, {targetBone {boneName=deform_pec_r}}, {targetBone {boneName=clavicle_r}}]
    constraintWeights [1,1,1,1]
   }
   n32 = "Constraint - dyn_shirt_front_cnstr_l (PP)"!(ComponentPose=@n31, Alpha=1, ConstraintWeights_0=0.9, ConstraintWeights_1:float=0.1) -> (Pose) {
    alphaBoolBlend {blendOption=Linear}
    boneToModify {boneName=dyn_shirt_front_cnstr_l}
    constraintSetup [{targetBone {boneName=spine_03}}, {targetBone {boneName=dfrm_clavicle_l}}]
    constraintWeights [1,1]
   }
   n33 = "Constraint - dyn_shirt_front_side_cnstr_l (PP)"!(ComponentPose=@n32, Alpha=1, ConstraintWeights_0=0.8, ConstraintWeights_1:float=0.2) -> (Pose) {
    alphaBoolBlend {blendOption=Linear}
    boneToModify {boneName=dyn_shirt_front_side_cnstr_l}
    constraintSetup [{targetBone {boneName=spine_03}}, {targetBone {boneName=dfrm_clavicle_l}}]
    constraintWeights [1,1]
   }
   n34 = "Constraint - dyn_shirt_front_side_cnstr_r (PP)"!(ComponentPose=@n33, Alpha=1, ConstraintWeights_0=0.8, ConstraintWeights_1:float=0.2) -> (Pose) {
    alphaBoolBlend {blendOption=Linear}
    boneToModify {boneName=dyn_shirt_front_side_cnstr_r}
    constraintSetup [{targetBone {boneName=spine_03}}, {targetBone {boneName=dfrm_clavicle_l}}]
    constraintWeights [1,1]
   }
   n35 = "Constraint - dyn_shirt_front_cnstr_r (PP)"!(ComponentPose=@n34, Alpha=1, ConstraintWeights_0=0.9, ConstraintWeights_1:float=0.1) -> (Pose) {
    alphaBoolBlend {blendOption=Linear}
    boneToModify {boneName=dyn_shirt_front_cnstr_r}
    constraintSetup [{targetBone {boneName=spine_03}}, {targetBone {boneName=dfrm_clavicle_r}}]
    constraintWeights [1,1]
   }
   n36 = RigidBody!(ComponentPose=@n35, Alpha, OverrideWorldGravity=Get Gravity Override(), ExternalForce:Vector) -> (Pose) {
    bClampLinearTranslationLimitToRefPose=true
    bOverrideWorldGravity=true
    overridePhysicsAsset=&/F_MED_Chainmail_Necklace_Physics
    simulationSpace=BaseBoneSpace
    alphaBoolBlend {blendOption=Linear}
    baseBoneRef {boneName=dyn_neckace_mstr}
    componentAppliedLinearAccClamp=<3000 3000 3000>
    componentLinearAccScale=<0.4 0.3 0.3>
    componentLinearVelScale=<0.4 0.3 0.3>
    simSpaceSettings {maxAngularAcceleration=1000, maxAngularVelocity=3, maxLinearAcceleration=250, maxLinearVelocity=500, velocityScaleZ=0, worldAlpha=1}
   }
   n37 = Anim Dynamics - Bone: dyn_outer_lace_01_l(ComponentPose=@n36, Alpha=1, ExternalForce="0,0,0") {
    angularDampingOverride=0.9999
    angularSpringConstant=500
    bAngularSpring=true
    bGravityOverrideInSimSpace=true
    bOverrideAngularDamping=true
    simulationSpace=BoneRelative
    alphaBoolBlend {blendOption=Linear}
    boundBone {boneName=dyn_outer_lace_01_l}
    componentAppliedLinearAccClamp=<5000 5000 5000>
    componentLinearAccScale=<1.2 1.2 1.2>
    componentLinearVelScale=<1.2 1.2 1.2>
    physicsBodyDefinitions [{
      sphereCollisionRadius=0
      boundBone {boneName=dyn_outer_lace_01_l}
      boxExtents=<15 20 20>
      constraintSetup {angularLimitsMax=<1 20 10>, angularLimitsMin=<-1 -20 -45>, angularTarget=<1 0 0>}
      localJointOffset=<3 0 0>
     }]
    relativeSpaceBone {boneName=root}
   }
   n38 = Anim Dynamics - Bone: dyn_outer_lace_02_l(ComponentPose=@n37, Alpha=1, ExternalForce="0,0,0") {
    angularDampingOverride=0.9999
    angularSpringConstant=500
    bAngularSpring=true
    bGravityOverrideInSimSpace=true
    bOverrideAngularDamping=true
    simulationSpace=BoneRelative
    alphaBoolBlend {blendOption=Linear}
    boundBone {boneName=dyn_outer_lace_02_l}
    componentAppliedLinearAccClamp=<5000 5000 5000>
    componentLinearAccScale=<1.2 1.2 1.2>
    componentLinearVelScale=<1.2 1.2 1.2>
    physicsBodyDefinitions [{
      sphereCollisionRadius=0
      boundBone {boneName=dyn_outer_lace_02_l}
      boxExtents=<15 20 20>
      constraintSetup {angularLimitsMax=<1 40 10>, angularLimitsMin=<-1 -60 -45>, angularTarget=<1 0 0>}
      localJointOffset=<3 0 0>
     }]
    relativeSpaceBone {boneName=root}
   }
   n39 = Anim Dynamics - Bone: dyn_outer_lace_03_l(ComponentPose=@n38, Alpha=1, ExternalForce="0,0,0") {
    angularDampingOverride=0.9999
    angularSpringConstant=500
    bAngularSpring=true
    bGravityOverrideInSimSpace=true
    bOverrideAngularDamping=true
    simulationSpace=BoneRelative
    alphaBoolBlend {blendOption=Linear}
    boundBone {boneName=dyn_outer_lace_03_l}
    componentAppliedLinearAccClamp=<5000 5000 5000>
    componentLinearAccScale=<1.2 1.2 1.2>
    componentLinearVelScale=<1.2 1.2 1.2>
    physicsBodyDefinitions [{
      sphereCollisionRadius=0
      boundBone {boneName=dyn_outer_lace_03_l}
      boxExtents=<15 20 20>
      constraintSetup {angularLimitsMax=<1 20 10>, angularLimitsMin=<-1 -60 -40>, angularTarget=<1 0 0>}
      localJointOffset=<3 0 0>
     }]
    relativeSpaceBone {boneName=root}
   }
   n40 = Anim Dynamics - Bone: dyn_inner_lace_01_l(ComponentPose=@n39, Alpha=1, ExternalForce="0,0,0") {
    angularDampingOverride=0.9999
    angularSpringConstant=650
    bAngularSpring=true
    bGravityOverrideInSimSpace=true
    bOverrideAngularDamping=true
    simulationSpace=BoneRelative
    alphaBoolBlend {blendOption=Linear}
    boundBone {boneName=dyn_inner_lace_01_l}
    componentAppliedLinearAccClamp=<5000 5000 5000>
    componentLinearAccScale=<1.2 1.2 1.2>
    componentLinearVelScale=<1.2 1.2 1.2>
    physicsBodyDefinitions [{
      sphereCollisionRadius=0
      boundBone {boneName=dyn_inner_lace_01_l}
      boxExtents=<15 20 20>
      constraintSetup {angularLimitsMax=<1 20 10>, angularLimitsMin=<-1 -20 -45>, angularTarget=<1 0 0>}
      localJointOffset=<3 0 0>
     }]
    relativeSpaceBone {boneName=root}
   }
   n41 = Anim Dynamics - Bone: dyn_inner_lace_02_l(ComponentPose=@n40, Alpha=1, ExternalForce="0,0,0") {
    angularDampingOverride=0.9999
    angularSpringConstant=650
    bAngularSpring=true
    bGravityOverrideInSimSpace=true
    bOverrideAngularDamping=true
    simulationSpace=BoneRelative
    alphaBoolBlend {blendOption=Linear}
    boundBone {boneName=dyn_inner_lace_02_l}
    componentAppliedLinearAccClamp=<5000 5000 5000>
    componentLinearAccScale=<1.2 1.2 1.2>
    componentLinearVelScale=<1.2 1.2 1.2>
    physicsBodyDefinitions [{
      sphereCollisionRadius=0
      boundBone {boneName=dyn_inner_lace_02_l}
      boxExtents=<15 20 20>
      constraintSetup {angularLimitsMax=<1 40 10>, angularLimitsMin=<-1 -60 -45>, angularTarget=<1 0 0>}
      localJointOffset=<3 0 0>
     }]
    relativeSpaceBone {boneName=root}
   }
   n42 = Anim Dynamics - Bone: dyn_inner_lace_03_l(ComponentPose=@n41, Alpha=1, ExternalForce="0,0,0") {
    angularDampingOverride=0.9999
    angularSpringConstant=650
    bAngularSpring=true
    bGravityOverrideInSimSpace=true
    bOverrideAngularDamping=true
    simulationSpace=BoneRelative
    alphaBoolBlend {blendOption=Linear}
    boundBone {boneName=dyn_inner_lace_03_l}
    componentAppliedLinearAccClamp=<5000 5000 5000>
    componentLinearAccScale=<1.2 1.2 1.2>
    componentLinearVelScale=<1.2 1.2 1.2>
    physicsBodyDefinitions [{
      sphereCollisionRadius=0
      boundBone {boneName=dyn_inner_lace_03_l}
      boxExtents=<15 20 20>
      constraintSetup {angularLimitsMax=<1 20 10>, angularLimitsMin=<-1 -60 -40>, angularTarget=<1 0 0>}
      localJointOffset=<3 0 0>
     }]
    relativeSpaceBone {boneName=root}
   }
   n43 = Anim Dynamics - Bone: dyn_outer_lace_01_r(ComponentPose=@n42, Alpha=1, ExternalForce="0,0,0") {
    angularDampingOverride=0.9999
    angularSpringConstant=500
    bAngularSpring=true
    bGravityOverrideInSimSpace=true
    bOverrideAngularDamping=true
    simulationSpace=BoneRelative
    alphaBoolBlend {blendOption=Linear}
    boundBone {boneName=dyn_outer_lace_01_r}
    componentAppliedLinearAccClamp=<5000 5000 5000>
    componentLinearAccScale=<1.2 1.2 1.2>
    componentLinearVelScale=<1.2 1.2 1.2>
    physicsBodyDefinitions [{
      sphereCollisionRadius=0
      boundBone {boneName=dyn_outer_lace_01_r}
      boxExtents=<15 20 20>
      constraintSetup {angularLimitsMax=<1 20 10>, angularLimitsMin=<-1 -20 -45>, angularTarget=<1 0 0>}
      localJointOffset=<3 0 0>
     }]
    relativeSpaceBone {boneName=root}
   }
   n44 = Anim Dynamics - Bone: dyn_outer_lace_02_r(ComponentPose=@n43, Alpha=1, ExternalForce="0,0,0") {
    angularDampingOverride=0.9999
    angularSpringConstant=500
    bAngularSpring=true
    bGravityOverrideInSimSpace=true
    bOverrideAngularDamping=true
    simulationSpace=BoneRelative
    alphaBoolBlend {blendOption=Linear}
    boundBone {boneName=dyn_outer_lace_02_r}
    componentAppliedLinearAccClamp=<5000 5000 5000>
    componentLinearAccScale=<1.2 1.2 1.2>
    componentLinearVelScale=<1.2 1.2 1.2>
    physicsBodyDefinitions [{
      sphereCollisionRadius=0
      boundBone {boneName=dyn_outer_lace_02_r}
      boxExtents=<15 20 20>
      constraintSetup {angularLimitsMax=<1 40 10>, angularLimitsMin=<-1 -60 -45>, angularTarget=<1 0 0>}
      localJointOffset=<3 0 0>
     }]
    relativeSpaceBone {boneName=root}
   }
   n45 = Anim Dynamics - Bone: dyn_outer_lace_03_r(ComponentPose=@n44, Alpha=1, ExternalForce="0,0,0") {
    angularDampingOverride=0.9999
    angularSpringConstant=500
    bAngularSpring=true
    bGravityOverrideInSimSpace=true
    bOverrideAngularDamping=true
    simulationSpace=BoneRelative
    alphaBoolBlend {blendOption=Linear}
    boundBone {boneName=dyn_outer_lace_03_r}
    componentAppliedLinearAccClamp=<5000 5000 5000>
    componentLinearAccScale=<1.2 1.2 1.2>
    componentLinearVelScale=<1.2 1.2 1.2>
    physicsBodyDefinitions [{
      sphereCollisionRadius=0
      boundBone {boneName=dyn_outer_lace_03_r}
      boxExtents=<15 20 20>
      constraintSetup {angularLimitsMax=<1 20 10>, angularLimitsMin=<-1 -60 -40>, angularTarget=<1 0 0>}
      localJointOffset=<3 0 0>
     }]
    relativeSpaceBone {boneName=root}
   }
   n46 = Anim Dynamics - Bone: dyn_inner_lace_01_r(ComponentPose=@n45, Alpha=1, ExternalForce="0,0,0") {
    angularDampingOverride=0.9999
    angularSpringConstant=650
    bAngularSpring=true
    bGravityOverrideInSimSpace=true
    bOverrideAngularDamping=true
    simulationSpace=BoneRelative
    alphaBoolBlend {blendOption=Linear}
    boundBone {boneName=dyn_inner_lace_01_r}
    componentAppliedLinearAccClamp=<5000 5000 5000>
    componentLinearAccScale=<1.2 1.2 1.2>
    componentLinearVelScale=<1.2 1.2 1.2>
    physicsBodyDefinitions [{
      sphereCollisionRadius=0
      boundBone {boneName=dyn_inner_lace_01_r}
      boxExtents=<15 20 20>
      constraintSetup {angularLimitsMax=<1 20 10>, angularLimitsMin=<-1 -20 -45>, angularTarget=<1 0 0>}
      localJointOffset=<3 0 0>
     }]
    relativeSpaceBone {boneName=root}
   }
   n47 = Anim Dynamics - Bone: dyn_inner_lace_02_r(ComponentPose=@n46, Alpha=1, ExternalForce="0,0,0") {
    angularDampingOverride=0.9999
    angularSpringConstant=650
    bAngularSpring=true
    bGravityOverrideInSimSpace=true
    bOverrideAngularDamping=true
    simulationSpace=BoneRelative
    alphaBoolBlend {blendOption=Linear}
    boundBone {boneName=dyn_inner_lace_02_r}
    componentAppliedLinearAccClamp=<5000 5000 5000>
    componentLinearAccScale=<1.2 1.2 1.2>
    componentLinearVelScale=<1.2 1.2 1.2>
    physicsBodyDefinitions [{
      sphereCollisionRadius=0
      boundBone {boneName=dyn_inner_lace_02_r}
      boxExtents=<15 20 20>
      constraintSetup {angularLimitsMax=<1 40 10>, angularLimitsMin=<-1 -60 -45>, angularTarget=<1 0 0>}
      localJointOffset=<3 0 0>
     }]
    relativeSpaceBone {boneName=root}
   }
   n48 = Anim Dynamics - Bone: dyn_inner_lace_03_r(ComponentPose=@n47, Alpha=1, ExternalForce="0,0,0") {
    angularDampingOverride=0.9999
    angularSpringConstant=650
    bAngularSpring=true
    bGravityOverrideInSimSpace=true
    bOverrideAngularDamping=true
    simulationSpace=BoneRelative
    alphaBoolBlend {blendOption=Linear}
    boundBone {boneName=dyn_inner_lace_03_r}
    componentAppliedLinearAccClamp=<5000 5000 5000>
    componentLinearAccScale=<1.2 1.2 1.2>
    componentLinearVelScale=<1.2 1.2 1.2>
    physicsBodyDefinitions [{
      sphereCollisionRadius=0
      boundBone {boneName=dyn_inner_lace_03_r}
      boxExtents=<15 20 20>
      constraintSetup {angularLimitsMax=<1 20 10>, angularLimitsMin=<-1 -60 -40>, angularTarget=<1 0 0>}
      localJointOffset=<3 0 0>
     }]
    relativeSpaceBone {boneName=root}
   }
   n49 = Anim Dynamics - Bone: dfrm_sleeve_l(ComponentPose=@n48, Alpha=1, ExternalForce="0,0,0") {
    angularDampingOverride=1
    angularSpringConstant=20000
    bAngularSpring=true
    bOverrideAngularDamping=true
    bUseGravityOverride=true
    gravityScale=0.35
    simulationSpace=RootRelative
    alphaBoolBlend {blendOption=Linear}
    boundBone {boneName=dfrm_sleeve_l}
    componentLinearAccScale=<0.7 0.7 0.7>
    componentLinearVelScale=<0.7 0.7 0.7>
    gravityOverride {z=-980}
    physicsBodyDefinitions [{
      sphereCollisionRadius=0
      boundBone {boneName=dfrm_sleeve_l}
      boxExtents=<25 35 35>
      constraintSetup {angularLimitsMax=<0 3 5>, angularLimitsMin=<0 -3 -5>, angularTarget=<1 0 0>}
      localJointOffset=<1 0 0>
     }]
   }
   n50 = Anim Dynamics - Bone: dfrm_sleeve_r(ComponentPose=@n49, Alpha=1, ExternalForce="0,0,0") {
    angularDampingOverride=1
    angularSpringConstant=20000
    bAngularSpring=true
    bOverrideAngularDamping=true
    bUseGravityOverride=true
    gravityScale=0.35
    simulationSpace=RootRelative
    alphaBoolBlend {blendOption=Linear}
    boundBone {boneName=dfrm_sleeve_r}
    componentLinearAccScale=<0.7 0.7 0.7>
    componentLinearVelScale=<0.7 0.7 0.7>
    gravityOverride {z=-980}
    physicsBodyDefinitions [{
      sphereCollisionRadius=0
      boundBone {boneName=dfrm_sleeve_r}
      boxExtents=<25 35 35>
      constraintSetup {angularLimitsMax=<0 3 5>, angularLimitsMin=<0 -3 -5>, angularTarget=<1 0 0>}
      localJointOffset=<1 0 0>
     }]
   }
   n51 = Control Rig!(
    Source=Component To Local(ComponentPose=@n50)
    Alpha
    spine_rotate_down_alpha:double=Get spine_rotate_down_alpha() -> (spine_rotate_down_alpha:double)
    is_crouching:bool=Get bIsCrouching()
    is_moving_2d:bool=Get bIsMoving2D()
   ) -> (Pose) {
    alphaBoolBlend {blendOption=Linear}
    controlRigAssetReference {blueprintRigClass=&/F_MED_Chainmail_CtrlRig.F_MED_Chainmail_CtrlRig_C}
   }
   n52 = Control Rig!(Source=@n51, Alpha) -> (Pose) {
    alphaBoolBlend {blendOption=Linear}
    controlRigAssetReference {blueprintRigClass=&/F_MED_Chainmail_Necklace_CtrlRig.F_MED_Chainmail_Necklace_CtrlRig_C}
   }
   Save cached pose 'pre_sim_cache'(Pose=@n52) {cacheName=pre_sim_cache}
  ]}
 EventGraph {nodes [
   Event BlueprintUpdateAnimation()
   > n1 = Sequence()
   > n2 = Gravity Override State Function DEPRECATED(CharacterPartAnimInstance=Self-Reference(), GameTimeInSeconds=GetGameTimeInSeconds())
   > Set Gravity Override(Gravity Override=@n2)
   Set spine_rotate_down_alpha(
    execute=@n1.then_1
    spine_rotate_down_alpha:double=GetAlphaFromReferenceTransformDelta(
     SkeletalMeshComponent=Get skeletal_mesh() -> (skeletal_mesh:Skeletal Mesh Component Object Reference)
     BoneName=spine_05
     BaseName=pelvis
     DeltaComponentType=QuaternionTwist
     InputRange=(X=-8,Y=-32)
     TwistAxis="0,0,1"
    )
   ) -> (Output_Get:double)
   Event OnCharacterCustomizationLoaded()
   > Set is_wearing_cape(is_wearing_cape:bool=CheckForBackbling().bIsBackblingACape) -> (Output_Get:bool)
   Event BlueprintInitializeAnimation()
   > Set skeletal_mesh(skeletal_mesh:Skeletal Mesh Component Object Reference=GetOwnerSkeletalMeshComponent()) -> (Output_Get:Skeletal Mesh Component Object Reference)
  ]}
]
variables [
 Gravity Override {type=Vector}
 is_wearing_cape {type=boolean}
 skeletal_mesh {type=SkeletalMeshComponent}
 spine_rotate_down_alpha {type=number}
]
