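# Corrective Control Rig for the female medium Chainmail face accessory part, posing hair and hat-back bones as the character moves.
# Spherical pose readers on head and upper arms detect looking up and rifle or pistol aiming, multiplied into combined alphas.
# These alphas, smoothed for plain look-up, weight additive bone offsets alongside twist-left and spine-to-head alphas, which the graph reads but never computes.
# A parent constraint finally ties one right hair bone to the side master.
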ControlRigBlueprint /BRCosmetics/Characters/Player/Female/Medium/Bodies/F_MED_Chainmail/Meshes/Parts/F_MED_Chainmail_FaceAcc_CtrlRig
graphs [
 ContainedGraph {nodes [
   n1 = RigVMFunctionEntryNode()
   SequenceAggregate_1 = &RigVMFunction_Sequence(ExecuteContext=@n1.ExecuteContext)
   SequenceAggregate_1_1 = &RigVMFunction_Sequence(ExecuteContext=@SequenceAggregate_1.B)
   SequenceAggregate_1_2 = &RigVMFunction_Sequence(ExecuteContext=@SequenceAggregate_1_1.B)
   SequenceAggregate_1_3 = &RigVMFunction_Sequence(ExecuteContext=@SequenceAggregate_1_2.B)
   RigVMFunctionReturnNode(A=@SequenceAggregate_1.A, B=@SequenceAggregate_1_1.A, C=@SequenceAggregate_1_2.A, D=@SequenceAggregate_1_3.A, E=@SequenceAggregate_1_3.B)
  ]}
 ContainedGraph {nodes [
   n1 = RigVMFunctionEntryNode()
   SequenceAggregate = &RigVMFunction_Sequence(ExecuteContext=@n1.ExecuteContext)
   SequenceAggregate_1 = &RigVMFunction_Sequence(ExecuteContext=@SequenceAggregate.B)
   SequenceAggregate_2 = &RigVMFunction_Sequence(ExecuteContext=@SequenceAggregate_1.B)
   SequenceAggregate_3 = &RigVMFunction_Sequence(ExecuteContext=@SequenceAggregate_2.B)
   SequenceAggregate_4 = &RigVMFunction_Sequence(ExecuteContext=@SequenceAggregate_3.B)
   RigVMFunctionReturnNode(
    A=@SequenceAggregate.A
    B=@SequenceAggregate_1.A
    C=@SequenceAggregate_2.A
    D=@SequenceAggregate_3.A
    E=@SequenceAggregate_4.A
    F=@SequenceAggregate_4.B
   )
  ]}
 ContainedGraph {nodes [
   n1 = RigVMFunctionEntryNode()
   SequenceAggregate_2 = &RigVMFunction_Sequence(ExecuteContext=@n1.ExecuteContext)
   SequenceAggregate_2_1 = &RigVMFunction_Sequence(ExecuteContext=@SequenceAggregate_2.B)
   SequenceAggregate_2_2 = &RigVMFunction_Sequence(ExecuteContext=@SequenceAggregate_2_1.B)
   RigVMFunctionReturnNode(A=@SequenceAggregate_2.A, B=@SequenceAggregate_2_1.A, C=@SequenceAggregate_2_2.A, D=@SequenceAggregate_2_2.B)
  ]}
 ContainedGraph {nodes [
   n1 = RigVMFunctionEntryNode()
   SequenceAggregate_3 = &RigVMFunction_Sequence(ExecuteContext=@n1.ExecuteContext)
   SequenceAggregate_3_1 = &RigVMFunction_Sequence(ExecuteContext=@SequenceAggregate_3.B)
   RigVMFunctionReturnNode(A=@SequenceAggregate_3.A, B=@SequenceAggregate_3_1.A, C=@SequenceAggregate_3_1.B)
  ]}
 RigVMFunctionLibrary {links [], nodes []}
 RigVMModel {nodes [
   RigUnit_BeginExecution = &RigUnit_BeginExecution()
   SequenceAggregate_1 = RigVMAggregateNode(ExecuteContext=@RigUnit_BeginExecution.ExecutePin)
   SequenceAggregate = RigVMAggregateNode(ExecuteContext=@SequenceAggregate_1.A)
   n1 = &RigUnit_SphericalPoseReader(
    ActiveRegionScaleFactors=(PositiveWidth=2,NegativeWidth=0.6,PositiveHeight=2,NegativeHeight=2)
    ActiveRegionSize=0.1
    Debug=(bDrawDebug=false,DebugScale=25,DebugSegments=20,DebugThickness=0.25)
    DriverAxis=(X=0,Y=1,Z=0)
    DriverItem=(Type=Bone,Name="head")
    FalloffRegionScaleFactors=(PositiveWidth=1,NegativeWidth=0.8,PositiveHeight=1,NegativeHeight=1)
    FalloffSize=0.2
    OptionalParentItem=(Type=Bone,Name="None")
    RotationOffset=(X=90,Y=0,Z=-30)
    ExecutePin=@SequenceAggregate.A
   )
   > RigVMVariableNode(Value=@n1.OutputParam) {variable=spr_look_up_alpha}
   SequenceAggregate_2 = RigVMAggregateNode(ExecuteContext=@SequenceAggregate.F)
   n2 = RigVMVariableNode() {variable=spr_upperarm_l_rifle_aim_alpha}
   n3 = RigVMVariableNode() {variable=spr_look_up_alpha}
   RigVMVariableNode(
    Value=&RigVMFunction_MathDoubleMul(
     A=&RigVMFunction_MathDoubleMul(B=@n2.Value, A=@n3.Value).Result
     B=RigVMVariableNode() {variable=spr_upperarm_r_rifle_aim_alpha}.Value
    ).Result
    ExecuteContext=@SequenceAggregate_2.B
   ) {variable=look_up_with_rifle_mult_alpha}
   RigVMVariableNode(
    Value=&RigVMFunction_MathDoubleMul(B=RigVMVariableNode() {variable=twist_left_alpha}.Value, A=RigVMVariableNode() {variable=spr_look_up_alpha}.Value).Result
    ExecuteContext=@SequenceAggregate_2.D
   ) {variable=twist_left_head_mult_alpha}
   RigVMVariableNode(
    Value=&RigVMFunction_MathDoubleAdd(A=RigVMVariableNode() {variable=look_up_with_rifle_mult_alpha}.Value, B=RigVMVariableNode() {variable=look_up_with_pistol_mult_alpha}.Value).Result
    ExecuteContext=@SequenceAggregate_2.C
   ) {variable=look_up_weapons_no_twist_alpha}
   Multiply = &RigVMFunction_MathDoubleMul(B=RigVMVariableNode() {variable=spr_upperarm_l_pistol_aim_alpha}.Value, A=RigVMVariableNode() {variable=spr_look_up_alpha}.Value)
   RigVMVariableNode(
    Value=&RigVMFunction_MathDoubleMul(A=@Multiply.Result, B=RigVMVariableNode() {variable=spr_upperarm_r_pistol_aim_alpha}.Value).Result
    ExecuteContext=@SequenceAggregate_2.A
   ) {variable=look_up_with_pistol_mult_alpha}
   n4 = &RigUnit_SphericalPoseReader(
    ActiveRegionScaleFactors=(PositiveWidth=1,NegativeWidth=1,PositiveHeight=1,NegativeHeight=1)
    ActiveRegionSize=0.04
    Debug=(bDrawDebug=false,DebugScale=25,DebugSegments=20,DebugThickness=0.25)
    DriverAxis=(X=-1,Y=0,Z=0)
    DriverItem=(Type=Bone,Name="upperarm_r")
    FalloffRegionScaleFactors=(PositiveWidth=1,NegativeWidth=1,PositiveHeight=1,NegativeHeight=1)
    FalloffSize=0.1
    OptionalParentItem=(Type=Bone)
    RotationOffset=(X=-60,Y=22.5,Z=0)
    ExecutePin=@SequenceAggregate.E
   )
   > n5 = &RigUnit_SphericalPoseReader(
    ActiveRegionScaleFactors=(PositiveWidth=1,NegativeWidth=1,PositiveHeight=1,NegativeHeight=1)
    ActiveRegionSize=0.04
    Debug=(bDrawDebug=false,DebugScale=25,DebugSegments=20,DebugThickness=0.25)
    DriverAxis=(X=-1,Y=0,Z=0)
    DriverItem=(Type=Bone,Name="upperarm_r")
    FalloffRegionScaleFactors=(PositiveWidth=1,NegativeWidth=1,PositiveHeight=1,NegativeHeight=1)
    FalloffSize=0.1
    OptionalParentItem=(Type=Bone)
    RotationOffset=(X=-47,Y=15,Z=0)
   )
   Maximum = &RigVMFunction_MathDoubleMax(A=@n5.OutputParam, B=@n4.OutputParam)
   > RigVMVariableNode(Value=@Maximum.Result) {variable=spr_upperarm_r_rifle_aim_alpha}
   n6 = &RigUnit_SphericalPoseReader(
    ActiveRegionScaleFactors=(PositiveWidth=1,NegativeWidth=1,PositiveHeight=1,NegativeHeight=1)
    ActiveRegionSize=0.04
    Debug=(bDrawDebug=false,DebugScale=25,DebugSegments=20,DebugThickness=0.25)
    DriverAxis=(X=1,Y=0,Z=0)
    DriverItem=(Type=Bone,Name="upperarm_l")
    FalloffRegionScaleFactors=(PositiveWidth=1,NegativeWidth=1,PositiveHeight=1,NegativeHeight=1)
    FalloffSize=0.1
    OptionalParentItem=(Type=Bone)
    RotationOffset=(X=12.5,Y=-67.5,Z=0)
    ExecutePin=@SequenceAggregate.D
   )
   > RigVMVariableNode(Value=@n6.OutputParam) {variable=spr_upperarm_l_rifle_aim_alpha}
   n7 = &RigUnit_SphericalPoseReader(
    ActiveRegionScaleFactors=(PositiveWidth=3,NegativeWidth=1,PositiveHeight=1,NegativeHeight=3)
    ActiveRegionSize=0.05
    Debug=(bDrawDebug=false,DebugScale=25,DebugSegments=20,DebugThickness=0.25)
    DriverAxis=(X=-1,Y=0,Z=0)
    DriverItem=(Type=Bone,Name="upperarm_r")
    FalloffRegionScaleFactors=(PositiveWidth=1,NegativeWidth=1,PositiveHeight=1,NegativeHeight=1)
    FalloffSize=0.15
    OptionalParentItem=(Type=Bone)
    RotationOffset=(X=-82.5,Y=0,Z=-15)
    ExecutePin=@SequenceAggregate.B
   )
   > RigVMVariableNode(Value=@n7.OutputParam) {variable=spr_upperarm_r_pistol_aim_alpha}
   n8 = &RigUnit_SphericalPoseReader(
    ActiveRegionScaleFactors=(PositiveWidth=1,NegativeWidth=1,PositiveHeight=1,NegativeHeight=1)
    ActiveRegionSize=0.025
    Debug=(bDrawDebug=false,DebugScale=25,DebugSegments=20,DebugThickness=0.25)
    DriverAxis=(X=1,Y=0,Z=0)
    DriverItem=(Type=Bone,Name="upperarm_l")
    FalloffRegionScaleFactors=(PositiveWidth=1,NegativeWidth=1,PositiveHeight=1,NegativeHeight=1)
    FalloffSize=0.1
    OptionalParentItem=(Type=Bone)
    RotationOffset=(X=60,Y=0,Z=-32.5)
    ExecutePin=@SequenceAggregate.C
   )
   > RigVMVariableNode(Value=@n8.OutputParam) {variable=spr_upperarm_l_pistol_aim_alpha}
   &RigUnit_ParentConstraint(
    AdvancedSettings=(RotationOrderForFilter=XZY)
    Child=(Type=Bone,Name="dyn_hair_06_mstr_r")
    Parents=((Item=(Type=Bone,Name="dyn_hair_side_mstr_r"),Weight=0.4))
    Weight=1
    ExecutePin=@SequenceAggregate_1.E
   )
   SequenceAggregate_3 = RigVMAggregateNode(ExecuteContext=@SequenceAggregate_1.C)
   &RigUnit_ModifyTransforms(
    ItemToModify=((Item=(Type=Bone,Name="dyn_hair_back_mstr_l"),Transform=(Rotation=(X=0,Y=0,Z=0,W=1),Translation=(X=0,Y=0,Z=0),Scale3D=(X=1,Y=1,Z=1))))
    ItemToModify.0.Transform.Rotation=&RigVMFunction_MathQuaternionFromEuler(Euler=(X=0,Y=0,Z=-22.5)).Result
    ExecutePin=@SequenceAggregate_3.A
    Weight=RigVMVariableNode() {variable=look_up_weapons_no_twist_alpha}.Value
   )
   > &RigUnit_ModifyTransforms(
    ItemToModify=((Item=(Type=Bone,Name="dyn_hair_11_mstr_r"),Transform=(Rotation=(X=0,Y=0,Z=0,W=1),Translation=(X=0,Y=0,Z=0),Scale3D=(X=1,Y=1,Z=1))))
    ItemToModify.0.Transform.Rotation=&RigVMFunction_MathQuaternionFromEuler(Euler=(X=0,Y=0,Z=-5)).Result
    Weight=RigVMVariableNode() {variable=look_up_weapons_no_twist_alpha}.Value
   )
   > &RigUnit_ModifyTransforms(
    ItemToModify=((Item=(Type=Bone,Name="dyn_hair_side_mstr_r"),Transform=(Rotation=(X=0,Y=0,Z=0,W=1),Translation=(X=0,Y=0,Z=0),Scale3D=(X=1,Y=1,Z=1))))
    ItemToModify.0.Transform.Rotation=&RigVMFunction_MathQuaternionFromEuler(Euler=(X=0,Y=0,Z=-30)).Result
    Weight=RigVMVariableNode() {variable=look_up_weapons_no_twist_alpha}.Value
   )
   > &RigUnit_ModifyTransforms(
    ItemToModify=((Item=(Type=Bone,Name="dyn_hair_08_mstr_r"),Transform=(Rotation=(X=0,Y=0,Z=0,W=1),Translation=(X=0,Y=0,Z=0),Scale3D=(X=1,Y=1,Z=1))))
    ItemToModify.0.Transform.Rotation=&RigVMFunction_MathQuaternionFromEuler(Euler=(X=0,Y=0,Z=-10)).Result
    Weight=RigVMVariableNode() {variable=look_up_weapons_no_twist_alpha}.Value
   )
   > &RigUnit_ModifyTransforms(
    ItemToModify=((Item=(Type=Bone,Name="dyn_hair_09_mstr_r"),Transform=(Rotation=(X=0,Y=0,Z=0,W=1),Translation=(X=0,Y=0,Z=0),Scale3D=(X=1,Y=1,Z=1))))
    ItemToModify.0.Transform.Rotation=&RigVMFunction_MathQuaternionFromEuler(Euler=(X=0,Y=0,Z=-20)).Result
    Weight=RigVMVariableNode() {variable=look_up_weapons_no_twist_alpha}.Value
   )
   &RigUnit_ModifyTransforms(
    ItemToModify=((Item=(Type=Bone,Name="dyn_hair_09_l"),Transform=(Rotation=(X=0,Y=0,Z=0,W=1),Translation=(X=0,Y=0,Z=0),Scale3D=(X=1,Y=1,Z=1))))
    ItemToModify.0.Transform.Rotation=&RigVMFunction_MathQuaternionFromEuler(Euler=(X=0,Y=0,Z=-35)).Result
    Weight=RigVMVariableNode() {variable=spine_05_to_head_dRefz_alpha}.Value
    ExecutePin=@SequenceAggregate_3.C
   )
   > &RigUnit_ModifyTransforms(
    ItemToModify=((Item=(Type=Bone,Name="dyn_hair_back_mstr_l"),Transform=(Rotation=(X=0,Y=0,Z=0,W=1),Translation=(X=-3,Y=0,Z=0),Scale3D=(X=1,Y=1,Z=1))))
    ItemToModify.0.Transform.Rotation=&RigVMFunction_MathQuaternionFromEuler(Euler=(X=0,Y=0,Z=-40)).Result
    Weight=RigVMVariableNode() {variable=spine_05_to_head_dRefz_alpha}.Value
   )
   > &RigUnit_ModifyTransforms(
    ItemToModify=((Item=(Type=Bone,Name="dyn_hair_11_mstr_r"),Transform=(Rotation=(X=0,Y=0,Z=0,W=1),Translation=(X=0,Y=0,Z=0),Scale3D=(X=1,Y=1,Z=1))))
    ItemToModify.0.Transform.Rotation=&RigVMFunction_MathQuaternionFromEuler(Euler=(X=0,Y=0,Z=-35)).Result
    Weight=RigVMVariableNode() {variable=spine_05_to_head_dRefz_alpha}.Value
   )
   > &RigUnit_ModifyTransforms(
    ItemToModify=((Item=(Type=Bone,Name="dfrm_hat_back"),Transform=(Rotation=(X=0,Y=0,Z=0,W=1),Translation=(X=-3,Y=-3,Z=0),Scale3D=(X=1,Y=1,Z=1))))
    ItemToModify.0.Transform.Rotation=&RigVMFunction_MathQuaternionFromEuler(Euler=(X=0,Y=0,Z=-90)).Result
    Weight=RigVMVariableNode() {variable=spine_05_to_head_dRefz_alpha}.Value
   )
   > &RigUnit_ModifyTransforms(
    ItemToModify=((Item=(Type=Bone,Name="dyn_hair_08_mstr_r"),Transform=(Rotation=(X=0,Y=0,Z=0,W=1),Translation=(X=0,Y=0,Z=0),Scale3D=(X=1,Y=1,Z=1))))
    ItemToModify.0.Transform.Rotation=&RigVMFunction_MathQuaternionFromEuler(Euler=(X=0,Y=0,Z=-25)).Result
    Weight=RigVMVariableNode() {variable=spine_05_to_head_dRefz_alpha}.Value
   )
   > &RigUnit_ModifyTransforms(
    ItemToModify=((Item=(Type=Bone,Name="dyn_hair_side_mstr_r"),Transform=(Rotation=(X=0,Y=0,Z=0,W=1),Translation=(X=-3,Y=0,Z=0),Scale3D=(X=1,Y=1,Z=1))))
    ItemToModify.0.Transform.Rotation=&RigVMFunction_MathQuaternionFromEuler(Euler=(X=0,Y=0,Z=-35)).Result
    Weight=RigVMVariableNode() {variable=spine_05_to_head_dRefz_alpha}.Value
   )
   &RigUnit_ModifyTransforms(
    ItemToModify=((Item=(Type=Bone,Name="dyn_hair_back_mstr_l"),Transform=(Rotation=(X=0,Y=0,Z=0,W=1),Translation=(X=0,Y=0,Z=0),Scale3D=(X=1,Y=1,Z=1))))
    ItemToModify.0.Transform.Rotation=&RigVMFunction_MathQuaternionFromEuler(Euler=(X=0,Y=0,Z=-10)).Result
    Weight=RigVMVariableNode() {variable=twist_left_head_mult_alpha}.Value
    ExecutePin=@SequenceAggregate_3.B
   )
   > &RigUnit_ModifyTransforms(
    ItemToModify=((Item=(Type=Bone,Name="dyn_hair_side_mstr_r"),Transform=(Rotation=(X=0,Y=0,Z=0,W=1),Translation=(X=0,Y=0,Z=0),Scale3D=(X=1,Y=1,Z=1))))
    ItemToModify.0.Transform.Rotation=&RigVMFunction_MathQuaternionFromEuler(Euler=(X=-15,Y=0,Z=-20)).Result
    Weight=RigVMVariableNode() {variable=twist_left_head_mult_alpha}.Value
   )
   > &RigUnit_ModifyTransforms(
    ItemToModify=((Item=(Type=Bone,Name="dyn_hair_09_r"),Transform=(Rotation=(X=0,Y=0,Z=0,W=1),Translation=(X=0,Y=0,Z=0),Scale3D=(X=1,Y=1,Z=1))))
    ItemToModify.0.Transform.Rotation=&RigVMFunction_MathQuaternionFromEuler(Euler=(X=-15,Y=0,Z=-20)).Result
    Weight=RigVMVariableNode() {variable=twist_left_head_mult_alpha}.Value
   )
   AlphaInterp = &RigVMFunction_AlphaInterp(
    InterpSpeedDecreasing=10
    InterpSpeedIncreasing=20
    Scale=1
    bClampResult=False
    Value=RigVMVariableNode() {variable=spr_look_up_alpha}.Value
   )
   &RigUnit_ModifyTransforms(
    ItemToModify=((Item=(Type=Bone,Name="dyn_hair_06_l"),Transform=(Rotation=(X=0,Y=0,Z=0,W=1),Translation=(X=0,Y=0,Z=0),Scale3D=(X=1,Y=1,Z=1))))
    ItemToModify.0.Transform.Rotation=&RigVMFunction_MathQuaternionFromEuler(Euler=(X=0,Y=0,Z=-10)).Result
    Weight=@AlphaInterp.Result
    ExecutePin=@SequenceAggregate_1.B
   )
   > &RigUnit_ModifyTransforms(
    ItemToModify=((Item=(Type=Bone,Name="dyn_hair_09_l"),Transform=(Rotation=(X=0,Y=0,Z=0,W=1),Translation=(X=0,Y=0,Z=0),Scale3D=(X=1,Y=1,Z=1))))
    ItemToModify.0.Transform.Rotation=&RigVMFunction_MathQuaternionFromEuler(Euler=(X=0,Y=0,Z=-10)).Result
    Weight=@AlphaInterp.Result
   )
   > &RigUnit_ModifyTransforms(
    ItemToModify=((Item=(Type=Bone,Name="dyn_hair_back_mstr_l"),Transform=(Rotation=(X=0,Y=0,Z=0,W=1),Translation=(X=0,Y=0,Z=0),Scale3D=(X=1,Y=1,Z=1))))
    ItemToModify.0.Transform.Rotation=&RigVMFunction_MathQuaternionFromEuler(Euler=(X=0,Y=0,Z=-27.56)).Result
    Weight=@AlphaInterp.Result
   )
   > &RigUnit_ModifyTransforms(
    ItemToModify=((Item=(Type=Bone,Name="dyn_hair_11_mstr_r"),Transform=(Rotation=(X=0,Y=0,Z=0,W=1),Translation=(X=0,Y=0,Z=0),Scale3D=(X=1,Y=1,Z=1))))
    ItemToModify.0.Transform.Rotation=&RigVMFunction_MathQuaternionFromEuler(Euler=(X=0,Y=0,Z=-20)).Result
    Weight=@AlphaInterp.Result
   )
   > &RigUnit_ModifyTransforms(
    ItemToModify=((Item=(Type=Bone,Name="dyn_hair_side_mstr_r"),Transform=(Rotation=(X=0,Y=0,Z=0,W=1),Translation=(X=0,Y=0,Z=0),Scale3D=(X=1,Y=1,Z=1))))
    ItemToModify.0.Transform.Rotation=&RigVMFunction_MathQuaternionFromEuler(Euler=(X=0,Y=0,Z=-25)).Result
    Weight=@AlphaInterp.Result
   )
   > &RigUnit_ModifyTransforms(
    ItemToModify=((Item=(Type=Bone,Name="dyn_hair_08_mstr_r"),Transform=(Rotation=(X=0,Y=0,Z=0,W=1),Translation=(X=0,Y=0,Z=0),Scale3D=(X=1,Y=1,Z=1))))
    ItemToModify.0.Transform.Rotation=&RigVMFunction_MathQuaternionFromEuler(Euler=(X=0,Y=0,Z=-25)).Result
    Weight=@AlphaInterp.Result
   )
  ]}
]
hierarchy [
 root {loc=<0 0 0>, rot=<0 0 0 1>}
 pelvis {parent=root, loc=<0 -0.5301 98.0571>, rot=<0.0127 0.707 -0.0127 -0.707>}
 spine_01 {parent=pelvis, loc=<6.1208 -0.9328 0>, rot=<0 0 0.034 -0.9994>}
 spine_02 {parent=spine_01, loc=<6.5996 -0.2095 0>, rot=<0 0 0.0238 -0.9997>}
 spine_03 {parent=spine_02, loc=<6.4914 -0.1237 0>, rot=<0 0 0.0149 -0.9999>}
 spine_04 {parent=spine_03, loc=<6.9598 0.0794 0>, rot=<0 0 -0.0132 -0.9999>}
 spine_05 {parent=spine_04, loc=<6.9987 0 0>, rot=<0 0 -0.0015 -1>}
 neck_01 {parent=spine_05, loc=<11.3138 0.1529 0>, rot=<0 0 -0.2003 -0.9797>}
 neck_02 {parent=neck_01, loc=<4.5627 0 0>, rot=<0 0 0.0504 -0.9987>}
 head {parent=neck_02, loc=<5.3914 0 0>, rot=<0 0 0.115 -0.9934>}
 dyn_hair_05_l {parent=head, loc=<0.4069 -5.4582 -6.6485>, rot=<0.3059 -0.7467 -0.5255 -0.2697>}
 dyn_hair_01_r {parent=head, loc=<4.4682 3.3859 7.1196>, rot=<-0.0274 -0.6157 0.7866 0.038>}
 dyn_hair_01 {parent=head, loc=<7.6779 8.7051 0.4007>, rot=<-0.0998 0.0746 0.9846 0.1228>}
 dyn_hair_back_mstr_l {parent=head, loc=<0.638 -5.3291 -0.3166>, rot=<0.1221 -0.9918 -0.0361 0.009>}
 dyn_hair_06_l {parent=dyn_hair_back_mstr_l, loc=<4.2115 -2.0815 4.4139>, rot=<-0.0032 -0.0444 0.0726 -0.9964>}
 dyn_hair_08_l {parent=dyn_hair_back_mstr_l, loc=<1.0321 -2.2807 0.5897>, rot=<0.0758 0.2661 0.3523 -0.894>}
 dyn_hair_07_mstr_l {parent=dyn_hair_back_mstr_l, loc=<-0.0511 -1.9188 3.3005>, rot=<-0.0296 0.2691 0.6562 -0.7044>}
 dyn_hair_07_l {parent=dyn_hair_07_mstr_l, loc=<0 0 0>, rot=<0 0 0 1>}
 dyn_hair_11_mstr_r {parent=dyn_hair_back_mstr_l, loc=<4.9925 -1.1462 -0.3155>, rot=<-0.0457 0.0411 0.0006 -0.9981>}
 dyn_hair_11_r {parent=dyn_hair_11_mstr_r, loc=<0 0 0>, rot=<0 0 0 1>}
 dyn_hair_07_mstr_r {parent=dyn_hair_11_r, loc=<0.5381 -1.385 -0.8913>, rot=<0.0196 -0.2244 0.1882 -0.9559>}
 dyn_hair_07_r {parent=dyn_hair_07_mstr_r, loc=<0 0 0>, rot=<0 0 0 1>}
 dyn_hair_06_mstr_r {parent=dyn_hair_back_mstr_l, loc=<3.124 -1.7062 -3.5146>, rot=<0.0033 -0.0066 0.2868 -0.958>}
 dyn_hair_06_r {parent=dyn_hair_06_mstr_r, loc=<0 0 0>, rot=<0 0 0 1>}
 dyn_hair_09_l {parent=dyn_hair_back_mstr_l, loc=<4.3644 -0.2695 5.444>, rot=<0.2403 -0.034 -0.0214 -0.9699>}
 dyn_hair_side_mstr_l {parent=head, loc=<0.7906 -2.6917 -6.6655>, rot=<-0.0538 -0.7516 -0.6567 -0.0306>}
 dyn_hair_01_l {parent=dyn_hair_side_mstr_l, loc=<1.9556 -0.6654 1.1066>, rot=<0.3538 0.2028 0.1536 -0.9001>}
 dyn_hair_02_l {parent=dyn_hair_side_mstr_l, loc=<3.1404 0.4318 0.0074>, rot=<0.2077 0.13 -0.039 -0.9687>}
 dyn_hair_03_l {parent=dyn_hair_side_mstr_l, loc=<1.1792 -1.9643 -1.5256>, rot=<-0.1672 -0.1988 0.5757 -0.7753>}
 dyn_hair_04_l {parent=dyn_hair_side_mstr_l, loc=<2.762 -1.0348 -1.8502>, rot=<0.0733 -0.0349 0.1637 -0.9832>}
 dyn_hair_side_mstr_r {parent=head, loc=<2.2951 -3.4446 5.2309>, rot=<0.2084 -0.918 0.2991 0.1563>}
 dyn_hair_05_r {parent=dyn_hair_side_mstr_r, loc=<2.1106 -1.2889 2.5111>, rot=<-0.007 0.0592 0.3541 -0.9333>}
 dyn_hair_04_mstr_r {parent=dyn_hair_side_mstr_r, loc=<4.2261 -0.2225 1.2453>, rot=<-0.0475 0.0343 0.1903 -0.98>}
 dyn_hair_04_r {parent=dyn_hair_04_mstr_r, loc=<0 0 0>, rot=<0 0 0 1>}
 dyn_hair_03_mstr_r {parent=dyn_hair_side_mstr_r, loc=<3.2862 -0.1216 -1.2097>, rot=<-0.3593 -0.1868 0.1308 -0.9049>}
 dyn_hair_03_r {parent=dyn_hair_03_mstr_r, loc=<0 0 0>, rot=<0 0 0 1>}
 dyn_hair_02_mstr_r {parent=dyn_hair_side_mstr_r, loc=<1.8253 -0.5218 -2.2347>, rot=<-0.2361 -0.0582 0.3555 -0.9025>}
 dyn_hair_02_r {parent=dyn_hair_02_mstr_r, loc=<0 0 0>, rot=<0 0 0 1>}
 dyn_hair_10_mstr_r {parent=dyn_hair_side_mstr_r, loc=<3.4546 2.6634 -1.1108>, rot=<-0.2132 -0.0592 -0.3385 -0.9146>}
 dyn_hair_10_r {parent=dyn_hair_10_mstr_r, loc=<0 0 0>, rot=<0 0 0 1>}
 dyn_hair_08_mstr_r {parent=dyn_hair_side_mstr_r, loc=<5.1923 1.0909 3.6238>, rot=<0.013 0.1272 -0.0061 -0.9918>}
 dyn_hair_08_r {parent=dyn_hair_08_mstr_r, loc=<0 0 0>, rot=<0 0 0 1>}
 dyn_hair_09_mstr_r {parent=dyn_hair_side_mstr_r, loc=<4.4672 2.0022 -0.3533>, rot=<-0.147 -0.0125 0.0512 -0.9877>}
 dyn_hair_09_r {parent=dyn_hair_09_mstr_r, loc=<0 0 0>, rot=<0 0 0 1>}
 FACIAL_C_FacialRoot {parent=head, loc=<1.0894 1.2094 0>, rot=<0.5021 -0.4979 -0.5021 -0.4979>}
 FACIAL_C_UnderChin {parent=FACIAL_C_FacialRoot, loc=<0 7.4071 2.3312>, rot=<0.5 0 0 -0.866>}
 FACIAL_L_UnderChin {parent=FACIAL_C_FacialRoot, loc=<2.8814 6.7258 1.3392>, rot=<0.4025 -0.355 0.0823 -0.8398>}
 FACIAL_R_UnderChin {parent=FACIAL_C_FacialRoot, loc=<-2.8814 6.7258 1.3392>, rot=<0.4025 0.355 -0.0823 -0.8398>}
 FACIAL_C_Jaw {parent=FACIAL_C_FacialRoot, loc=<0 1.7683 0.8734>, rot=<0.342 0 0 -0.9397>}
 FACIAL_C_Chin {parent=FACIAL_C_Jaw, loc=<0 0.5618 6.7527>, rot=<-0.2588 0 0 -0.9659>}
 dfrm_chin_stud {parent=FACIAL_C_Jaw, loc=<0 -0.3071 7.3365>, rot=<0.9063 0 0 0.4226>}
 FACIAL_L_Jawline {parent=FACIAL_C_FacialRoot, loc=<4.1435 4.6153 2.7944>, rot=<0.2092 0.1578 -0.0547 -0.9635>}
 FACIAL_R_Jawline {parent=FACIAL_C_FacialRoot, loc=<-4.1435 4.6153 2.7944>, rot=<0.2092 -0.1578 0.0547 -0.9635>}
 FACIAL_C_Jawline {parent=FACIAL_C_FacialRoot, loc=<0 7.5075 4.6303>, rot=<0.5736 0 0 -0.8192>}
 FACIAL_L_CheekOuter {parent=FACIAL_C_FacialRoot, loc=<4.0712 0.3088 4.349>, rot=<0.0551 -0.377 0.1003 -0.9191>}
 FACIAL_C_Nose {parent=FACIAL_C_FacialRoot, loc=<0 -0.6762 7.2692>, rot=<-0.1305 0 0 -0.9914>}
 FACIAL_C_NoseLower {parent=FACIAL_C_Nose, loc=<0 3.505 -0.1681>, rot=<0.1305 0 0 -0.9914>}
 FACIAL_C_NoseTip {parent=FACIAL_C_Nose, loc=<0 2.5584 0.5684>, rot=<0.0872 0 0 -0.9962>}
 FACIAL_R_Nostril {parent=FACIAL_C_Nose, loc=<-0.8411 2.8248 -0.2714>, rot=<0.0173 -0.1305 -0.0023 -0.9913>}
 FACIAL_L_Nostril {parent=FACIAL_C_Nose, loc=<0.8411 2.8248 -0.2714>, rot=<0.0173 0.1305 0.0023 -0.9913>}
 FACIAL_L_Masseter {parent=FACIAL_C_FacialRoot, loc=<6.1197 1.0372 2.3351>, rot=<0.0693 0.1129 0.0079 -0.9912>}
 FACIAL_C_LowerLipRotation {parent=FACIAL_C_FacialRoot, loc=<0 1.7683 0.8734>, rot=<0.342 0 0 -0.9397>}
 FACIAL_C_MouthLower {parent=FACIAL_C_LowerLipRotation, loc=<0 -1.6043 4.8627>, rot=<-0.342 0 0 -0.9397>}
 FACIAL_C_LipLower {parent=FACIAL_C_MouthLower, loc=<0 1.019 1.6184>, rot=<0.0872 0 0 -0.9962>}
 dfrm_hat_back {parent=head, loc=<6.431 -1.0092 0>, rot=<0.5268 -0.85 0 0>}
 clavicle_r {parent=spine_05, loc=<5.4831 -0.1053 3.2946>, rot=<0.7289 0.0465 -0.6829 0.009>}
 upperarm_r {parent=clavicle_r, loc=<-9.8658 1.2136 0.4105>, rot=<0.0164 -0.3245 -0.0269 -0.9454>}
 clavicle_l {parent=spine_05, loc=<5.4831 -0.1053 -3.2946>, rot=<0.0465 -0.7289 -0.009 -0.6829>}
 upperarm_l {parent=clavicle_l, loc=<9.8658 -1.2136 -0.4105>, rot=<0.0164 -0.3245 -0.0269 -0.9454>}
]
variables [
 look_up_weapons_no_twist_alpha {type=double, default=0}
 look_up_with_pistol_mult_alpha {type=float, default=1}
 look_up_with_rifle_mult_alpha {type=float, default=1}
 spine_05_to_head_dRefz_alpha {type=double, default=0}
 spr_look_up_alpha {type=double, default=0}
 spr_upperarm_l_pistol_aim_alpha {type=double, default=0}
 spr_upperarm_l_rifle_aim_alpha {type=double, default=0}
 spr_upperarm_r_pistol_aim_alpha {type=double, default=0}
 spr_upperarm_r_rifle_aim_alpha {type=double, default=0}
 twist_left_alpha {type=double, default=0}
 twist_left_head_mult_alpha {type=double, default=0}
]
