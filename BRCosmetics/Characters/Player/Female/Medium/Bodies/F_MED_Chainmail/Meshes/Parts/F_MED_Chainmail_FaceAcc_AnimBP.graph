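# Post-process style AnimBlueprint for the female medium Chainmail body's face accessory, layering secondary motion onto the parent mesh's copied pose.
# Each update derives twist alphas from spine and head bone deltas, feeding two into a Control Rig; leaning_back_alpha is computed but unused here.
# Chained per-bone AnimDynamics springs then simulate the dyn_hair bones relative to root.
# A final constraint drives the chin stud from weighted lower-lip, jaw and chin bones.
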AnimBlueprint /BRCosmetics/Characters/Player/Female/Medium/Bodies/F_MED_Chainmail/Meshes/Parts/F_MED_Chainmail_FaceAcc_AnimBP
target_skeleton=&/F_MED_Chainmail_FaceAcc_Skeleton
graphs [
 AnimGraph {nodes [
   n1 = Control Rig!(
    Source=Copy Pose From Mesh() {bUseAttachedParent=true}
    Alpha
    twist_left_alpha:double=Get twist_left_alpha() -> (twist_left_alpha:double)
    spine_05_to_head_dRefz_alpha:double=Get spine_05_to_head_dRefz_alpha() -> (spine_05_to_head_dRefz_alpha:double)
   ) -> (Pose) {
    alphaBoolBlend {blendOption=Linear}
    controlRigAssetReference {blueprintRigClass=&/F_MED_Chainmail_FaceAcc_CtrlRig.F_MED_Chainmail_FaceAcc_CtrlRig_C}
   }
   n2 = Anim Dynamics - Bone: dyn_hair_01(ComponentPose=Local To Component(LocalPose=@n1), Alpha=1) {
    angularDampingOverride=0.9999
    angularSpringConstant=1500
    bAngularSpring=true
    bGravityOverrideInSimSpace=true
    bOverrideAngularDamping=true
    simulationSpace=BoneRelative
    alphaBoolBlend {blendOption=Linear}
    boundBone {boneName=dyn_hair_01}
    componentAppliedLinearAccClamp=<5000 5000 5000>
    componentLinearAccScale=<1.2 1.2 1.2>
    componentLinearVelScale=<1.2 1.2 1.2>
    physicsBodyDefinitions [{
      sphereCollisionRadius=0
      boundBone {boneName=dyn_hair_01}
      boxExtents=<12.5 12.5 12.5>
      constraintSetup {angularLimitsMax=<1 15 0>, angularLimitsMin=<-1 -3.5 -25>, angularTarget=<1 0 0>}
      localJointOffset=<1 0.5 0.15>
     }]
    relativeSpaceBone {boneName=root}
   }
   n3 = Anim Dynamics - Bone: dyn_hair_02_l(ComponentPose=@n2, Alpha=1, ExternalForce="0,0,0") {
    angularDampingOverride=0.9999
    angularSpringConstant=1500
    bAngularSpring=true
    bGravityOverrideInSimSpace=true
    bOverrideAngularDamping=true
    simulationSpace=BoneRelative
    alphaBoolBlend {blendOption=Linear}
    boundBone {boneName=dyn_hair_02_l}
    componentAppliedLinearAccClamp=<5000 5000 5000>
    componentLinearAccScale=<1.2 1.2 1.2>
    componentLinearVelScale=<1.2 1.2 1.2>
    physicsBodyDefinitions [{
      sphereCollisionRadius=0
      boundBone {boneName=dyn_hair_02_l}
      boxExtents=<15 15 15>
      constraintSetup {angularLimitsMax=<1 25 2.5>, angularLimitsMin=<-1 -5 -7.5>, angularTarget=<1 0 0>}
      localJointOffset=<2 0 0>
     }]
    relativeSpaceBone {boneName=root}
   }
   n4 = Anim Dynamics - Bone: dyn_hair_03_l(ComponentPose=@n3, Alpha=1, ExternalForce="0,0,0") {
    angularDampingOverride=0.9999
    angularSpringConstant=1500
    bAngularSpring=true
    bGravityOverrideInSimSpace=true
    bOverrideAngularDamping=true
    simulationSpace=BoneRelative
    alphaBoolBlend {blendOption=Linear}
    boundBone {boneName=dyn_hair_03_l}
    componentAppliedLinearAccClamp=<5000 5000 5000>
    componentLinearAccScale=<1.2 1.2 1.2>
    componentLinearVelScale=<1.2 1.2 1.2>
    physicsBodyDefinitions [{
      sphereCollisionRadius=0
      boundBone {boneName=dyn_hair_03_l}
      boxExtents=<11 11 11>
      constraintSetup {angularLimitsMax=<1 20 20>, angularLimitsMin=<-1 -20 -20>, angularTarget=<1 0 0>}
      localJointOffset=<1 0 0>
     }]
    relativeSpaceBone {boneName=root}
   }
   n5 = Anim Dynamics - Bone: dyn_hair_04_l(ComponentPose=@n4, Alpha=1, ExternalForce="0,0,0") {
    angularDampingOverride=0.9999
    angularSpringConstant=1250
    bAngularSpring=true
    bGravityOverrideInSimSpace=true
    bOverrideAngularDamping=true
    simulationSpace=BoneRelative
    alphaBoolBlend {blendOption=Linear}
    boundBone {boneName=dyn_hair_04_l}
    componentAppliedLinearAccClamp=<5000 5000 5000>
    componentLinearAccScale=<1.2 1.2 1.2>
    componentLinearVelScale=<1.2 1.2 1.2>
    physicsBodyDefinitions [{
      sphereCollisionRadius=0
      boundBone {boneName=dyn_hair_04_l}
      boxExtents=<12.5 12.5 12.5>
      constraintSetup {angularLimitsMax=<1 20 5>, angularLimitsMin=<-1 -20 -20>, angularTarget=<1 0 0>}
      localJointOffset=<2 0 0>
     }]
    relativeSpaceBone {boneName=root}
   }
   n6 = Anim Dynamics - Bone: dyn_hair_05_l(ComponentPose=@n5, Alpha=1, ExternalForce="0,0,0") {
    angularDampingOverride=0.9999
    angularSpringConstant=1250
    bAngularSpring=true
    bGravityOverrideInSimSpace=true
    bOverrideAngularDamping=true
    simulationSpace=BoneRelative
    alphaBoolBlend {blendOption=Linear}
    boundBone {boneName=dyn_hair_05_l}
    componentAppliedLinearAccClamp=<5000 5000 5000>
    componentLinearAccScale=<1.2 1.2 1.2>
    componentLinearVelScale=<1.2 1.2 1.2>
    physicsBodyDefinitions [{
      sphereCollisionRadius=0
      boundBone {boneName=dyn_hair_05_l}
      boxExtents=<12 12 12>
      constraintSetup {angularLimitsMax=<1 20 5>, angularLimitsMin=<-1 -20 -20>, angularTarget=<1 0 0>}
      localJointOffset=<1.25 0 0>
     }]
    relativeSpaceBone {boneName=root}
   }
   n7 = Anim Dynamics - Bone: dyn_hair_06_l(ComponentPose=@n6, Alpha=1, ExternalForce="0,0,0") {
    angularDampingOverride=0.9999
    angularSpringConstant=1250
    bAngularSpring=true
    bGravityOverrideInSimSpace=true
    bOverrideAngularDamping=true
    simulationSpace=BoneRelative
    alphaBoolBlend {blendOption=Linear}
    boundBone {boneName=dyn_hair_06_l}
    componentAppliedLinearAccClamp=<5000 5000 5000>
    componentLinearAccScale=<1.2 1.2 1.2>
    componentLinearVelScale=<1.2 1.2 1.2>
    physicsBodyDefinitions [{
      sphereCollisionRadius=0
      boundBone {boneName=dyn_hair_06_l}
      boxExtents=<8 8 8>
      constraintSetup {angularLimitsMax=<1 20 5>, angularLimitsMin=<-1 -20 -20>, angularTarget=<1 0 0>}
      localJointOffset=<0.5 0 0>
     }]
    relativeSpaceBone {boneName=root}
   }
   n8 = Anim Dynamics - Bone: dyn_hair_07_l(ComponentPose=@n7, Alpha=1, ExternalForce="0,0,0") {
    angularDampingOverride=0.9999
    angularSpringConstant=1500
    bAngularSpring=true
    bGravityOverrideInSimSpace=true
    bOverrideAngularDamping=true
    simulationSpace=BoneRelative
    alphaBoolBlend {blendOption=Linear}
    boundBone {boneName=dyn_hair_07_l}
    componentAppliedLinearAccClamp=<5000 5000 5000>
    componentLinearAccScale=<1.2 1.2 1.2>
    componentLinearVelScale=<1.2 1.2 1.2>
    physicsBodyDefinitions [{
      sphereCollisionRadius=0
      boundBone {boneName=dyn_hair_07_l}
      boxExtents=<7 7 7>
      constraintSetup {angularLimitsMax=<1 20 20>, angularLimitsMin=<-1 -20 -20>, angularTarget=<1 0 0>}
      localJointOffset=<0.25 0 0>
     }]
    relativeSpaceBone {boneName=root}
   }
   n9 = Anim Dynamics - Bone: dyn_hair_08_l(ComponentPose=@n8, Alpha=1, ExternalForce="0,0,0") {
    angularDampingOverride=0.9999
    angularSpringConstant=1500
    bAngularSpring=true
    bGravityOverrideInSimSpace=true
    bOverrideAngularDamping=true
    simulationSpace=BoneRelative
    alphaBoolBlend {blendOption=Linear}
    boundBone {boneName=dyn_hair_08_l}
    componentAppliedLinearAccClamp=<5000 5000 5000>
    componentLinearAccScale=<1.2 1.2 1.2>
    componentLinearVelScale=<1.2 1.2 1.2>
    physicsBodyDefinitions [{
      sphereCollisionRadius=0
      boundBone {boneName=dyn_hair_08_l}
      boxExtents=<7 7 7>
      constraintSetup {angularLimitsMax=<1 20 5>, angularLimitsMin=<-1 -20 -20>, angularTarget=<1 0 0>}
      localJointOffset=<0.5 0 0>
     }]
    relativeSpaceBone {boneName=root}
   }
   n10 = Anim Dynamics - Bone: dyn_hair_09_l(ComponentPose=@n9, Alpha=1, ExternalForce="0,0,0") {
    angularDampingOverride=0.9999
    angularSpringConstant=1500
    bAngularSpring=true
    bGravityOverrideInSimSpace=true
    bOverrideAngularDamping=true
    simulationSpace=BoneRelative
    alphaBoolBlend {blendOption=Linear}
    boundBone {boneName=dyn_hair_09_l}
    componentAppliedLinearAccClamp=<5000 5000 5000>
    componentLinearAccScale=<1.2 1.2 1.2>
    componentLinearVelScale=<1.2 1.2 1.2>
    physicsBodyDefinitions [{
      sphereCollisionRadius=0
      boundBone {boneName=dyn_hair_09_l}
      boxExtents=<7 7 7>
      constraintSetup {angularLimitsMax=<1 20 5>, angularLimitsMin=<-1 -20 -20>, angularTarget=<1 0 0>}
      localJointOffset=<0.5 0 0>
     }]
    relativeSpaceBone {boneName=root}
   }
   n11 = Anim Dynamics - Bone: dyn_hair_11_r(ComponentPose=@n10, Alpha=1, ExternalForce="0,0,0") {
    angularDampingOverride=0.9999
    angularSpringConstant=1500
    bAngularSpring=true
    bGravityOverrideInSimSpace=true
    bOverrideAngularDamping=true
    simulationSpace=BoneRelative
    alphaBoolBlend {blendOption=Linear}
    boundBone {boneName=dyn_hair_11_r}
    componentAppliedLinearAccClamp=<5000 5000 5000>
    componentLinearAccScale=<1.2 1.2 1.2>
    componentLinearVelScale=<1.2 1.2 1.2>
    physicsBodyDefinitions [{
      sphereCollisionRadius=0
      boundBone {boneName=dyn_hair_11_r}
      boxExtents=<8 8 8>
      constraintSetup {angularLimitsMax=<1 20 5>, angularLimitsMin=<-1 -20 -20>, angularTarget=<1 0 0>}
      localJointOffset=<0.5 0 0>
     }]
    relativeSpaceBone {boneName=root}
   }
   n12 = Anim Dynamics - Bone: dyn_hair_10_r(ComponentPose=@n11, Alpha=1, ExternalForce="0,0,0") {
    angularDampingOverride=0.9999
    angularSpringConstant=1500
    bAngularSpring=true
    bGravityOverrideInSimSpace=true
    bOverrideAngularDamping=true
    simulationSpace=BoneRelative
    alphaBoolBlend {blendOption=Linear}
    boundBone {boneName=dyn_hair_10_r}
    componentAppliedLinearAccClamp=<5000 5000 5000>
    componentLinearAccScale=<1.2 1.2 1.2>
    componentLinearVelScale=<1.2 1.2 1.2>
    physicsBodyDefinitions [{
      sphereCollisionRadius=0
      boundBone {boneName=dyn_hair_10_r}
      boxExtents=<8 8 8>
      constraintSetup {angularLimitsMax=<1 20 5>, angularLimitsMin=<-1 -5 -20>, angularTarget=<1 0 0>}
      localJointOffset=<0.5 0 0>
     }]
    relativeSpaceBone {boneName=root}
   }
   n13 = Anim Dynamics - Bone: dyn_hair_09_r(ComponentPose=@n12, Alpha=1, ExternalForce="0,0,0") {
    angularDampingOverride=0.9999
    angularSpringConstant=1500
    bAngularSpring=true
    bGravityOverrideInSimSpace=true
    bOverrideAngularDamping=true
    simulationSpace=BoneRelative
    alphaBoolBlend {blendOption=Linear}
    boundBone {boneName=dyn_hair_09_r}
    componentAppliedLinearAccClamp=<5000 5000 5000>
    componentLinearAccScale=<1.2 1.2 1.2>
    componentLinearVelScale=<1.2 1.2 1.2>
    physicsBodyDefinitions [{
      sphereCollisionRadius=0
      boundBone {boneName=dyn_hair_09_r}
      boxExtents=<8 8 8>
      constraintSetup {angularLimitsMax=<1 20 5>, angularLimitsMin=<-1 -20 -20>, angularTarget=<1 0 0>}
      localJointOffset=<0.5 0 0>
     }]
    relativeSpaceBone {boneName=root}
   }
   n14 = Anim Dynamics - Bone: dyn_hair_08_r(ComponentPose=@n13, Alpha=1, ExternalForce="0,0,0") {
    angularDampingOverride=0.9999
    angularSpringConstant=1500
    bAngularSpring=true
    bGravityOverrideInSimSpace=true
    bOverrideAngularDamping=true
    simulationSpace=BoneRelative
    alphaBoolBlend {blendOption=Linear}
    boundBone {boneName=dyn_hair_08_r}
    componentAppliedLinearAccClamp=<5000 5000 5000>
    componentLinearAccScale=<1.2 1.2 1.2>
    componentLinearVelScale=<1.2 1.2 1.2>
    physicsBodyDefinitions [{
      sphereCollisionRadius=0
      boundBone {boneName=dyn_hair_08_r}
      boxExtents=<8 8 8>
      constraintSetup {angularLimitsMax=<1 20 5>, angularLimitsMin=<-1 -20 -20>, angularTarget=<1 0 0>}
      localJointOffset=<0.5 0 0>
     }]
    relativeSpaceBone {boneName=root}
   }
   n15 = Anim Dynamics - Bone: dyn_hair_07_r(ComponentPose=@n14, Alpha=1, ExternalForce="0,0,0") {
    angularDampingOverride=0.9999
    angularSpringConstant=1500
    bAngularSpring=true
    bGravityOverrideInSimSpace=true
    bOverrideAngularDamping=true
    simulationSpace=BoneRelative
    alphaBoolBlend {blendOption=Linear}
    boundBone {boneName=dyn_hair_07_r}
    componentAppliedLinearAccClamp=<5000 5000 5000>
    componentLinearAccScale=<1.2 1.2 1.2>
    componentLinearVelScale=<1.2 1.2 1.2>
    physicsBodyDefinitions [{
      sphereCollisionRadius=0
      boundBone {boneName=dyn_hair_07_r}
      boxExtents=<7 7 7>
      constraintSetup {angularLimitsMax=<1 20 5>, angularLimitsMin=<-1 -20 -20>, angularTarget=<1 0 0>}
      localJointOffset=<0.5 0 0>
     }]
    relativeSpaceBone {boneName=root}
   }
   n16 = Anim Dynamics - Bone: dyn_hair_06_r(ComponentPose=@n15, Alpha=1, ExternalForce="0,0,0") {
    angularDampingOverride=0.9999
    angularSpringConstant=1500
    bAngularSpring=true
    bGravityOverrideInSimSpace=true
    bOverrideAngularDamping=true
    simulationSpace=BoneRelative
    alphaBoolBlend {blendOption=Linear}
    boundBone {boneName=dyn_hair_06_r}
    componentAppliedLinearAccClamp=<5000 5000 5000>
    componentLinearAccScale=<1.2 1.2 1.2>
    componentLinearVelScale=<1.2 1.2 1.2>
    physicsBodyDefinitions [{
      sphereCollisionRadius=0
      boundBone {boneName=dyn_hair_06_r}
      boxExtents=<8 8 8>
      constraintSetup {angularLimitsMax=<1 20 20>, angularLimitsMin=<-1 -20 -20>, angularTarget=<1 0 0>}
      localJointOffset=<0.5 0 0>
     }]
    relativeSpaceBone {boneName=root}
   }
   n17 = Anim Dynamics - Bone: dyn_hair_05_r(ComponentPose=@n16, Alpha=1, ExternalForce="0,0,0") {
    angularDampingOverride=0.9999
    angularSpringConstant=1500
    bAngularSpring=true
    bGravityOverrideInSimSpace=true
    bOverrideAngularDamping=true
    simulationSpace=BoneRelative
    alphaBoolBlend {blendOption=Linear}
    boundBone {boneName=dyn_hair_05_r}
    componentAppliedLinearAccClamp=<5000 5000 5000>
    componentLinearAccScale=<1.2 1.2 1.2>
    componentLinearVelScale=<1.2 1.2 1.2>
    physicsBodyDefinitions [{
      sphereCollisionRadius=0
      boundBone {boneName=dyn_hair_05_r}
      boxExtents=<7.5 7.5 7.5>
      constraintSetup {angularLimitsMax=<1 20 20>, angularLimitsMin=<-1 -20 -20>, angularTarget=<1 0 0>}
      localJointOffset=<0.5 0 0>
     }]
    relativeSpaceBone {boneName=root}
   }
   n18 = Anim Dynamics - Bone: dyn_hair_04_r(ComponentPose=@n17, Alpha=1, ExternalForce="0,0,0") {
    angularDampingOverride=0.9999
    angularSpringConstant=1450
    bAngularSpring=true
    bGravityOverrideInSimSpace=true
    bOverrideAngularDamping=true
    simulationSpace=BoneRelative
    alphaBoolBlend {blendOption=Linear}
    boundBone {boneName=dyn_hair_04_r}
    componentAppliedLinearAccClamp=<5000 5000 5000>
    componentLinearAccScale=<1.2 1.2 1.2>
    componentLinearVelScale=<1.2 1.2 1.2>
    physicsBodyDefinitions [{
      sphereCollisionRadius=0
      boundBone {boneName=dyn_hair_04_r}
      boxExtents=<7.25 7.25 7.25>
      constraintSetup {angularLimitsMax=<1 20 20>, angularLimitsMin=<-1 -20 -20>, angularTarget=<1 0 0>}
      localJointOffset=<0.55 0 0>
     }]
    relativeSpaceBone {boneName=root}
   }
   n19 = Anim Dynamics - Bone: dyn_hair_03_r(ComponentPose=@n18, Alpha=1, ExternalForce="0,0,0") {
    angularDampingOverride=0.9999
    angularSpringConstant=1450
    bAngularSpring=true
    bGravityOverrideInSimSpace=true
    bOverrideAngularDamping=true
    simulationSpace=BoneRelative
    alphaBoolBlend {blendOption=Linear}
    boundBone {boneName=dyn_hair_03_r}
    componentAppliedLinearAccClamp=<5000 5000 5000>
    componentLinearAccScale=<1.2 1.2 1.2>
    componentLinearVelScale=<1.2 1.2 1.2>
    physicsBodyDefinitions [{
      sphereCollisionRadius=0
      boundBone {boneName=dyn_hair_03_r}
      boxExtents=<7.25 7.25 7.25>
      constraintSetup {angularLimitsMax=<1 20 20>, angularLimitsMin=<-1 -20 -20>, angularTarget=<1 0 0>}
      localJointOffset=<0.55 0 0>
     }]
    relativeSpaceBone {boneName=root}
   }
   n20 = Anim Dynamics - Bone: dyn_hair_02_r(ComponentPose=@n19, Alpha=1, ExternalForce="0,0,0") {
    angularDampingOverride=0.9999
    angularSpringConstant=1450
    bAngularSpring=true
    bGravityOverrideInSimSpace=true
    bOverrideAngularDamping=true
    simulationSpace=BoneRelative
    alphaBoolBlend {blendOption=Linear}
    boundBone {boneName=dyn_hair_02_r}
    componentAppliedLinearAccClamp=<5000 5000 5000>
    componentLinearAccScale=<1.2 1.2 1.2>
    componentLinearVelScale=<1.2 1.2 1.2>
    physicsBodyDefinitions [{
      sphereCollisionRadius=0
      boundBone {boneName=dyn_hair_02_r}
      boxExtents=<7.25 7.25 7.25>
      constraintSetup {angularLimitsMax=<1 20 20>, angularLimitsMin=<-1 -20 -20>, angularTarget=<1 0 0>}
      localJointOffset=<0.55 0 0>
     }]
    relativeSpaceBone {boneName=root}
   }
   n21 = Anim Dynamics - Bone: dyn_hair_01_r(ComponentPose=@n20, Alpha=1, ExternalForce="0,0,0") {
    angularDampingOverride=0.9999
    angularSpringConstant=1500
    bAngularSpring=true
    bGravityOverrideInSimSpace=true
    bOverrideAngularDamping=true
    simulationSpace=BoneRelative
    alphaBoolBlend {blendOption=Linear}
    boundBone {boneName=dyn_hair_01_r}
    componentAppliedLinearAccClamp=<5000 5000 5000>
    componentLinearAccScale=<1.2 1.2 1.2>
    componentLinearVelScale=<1.2 1.2 1.2>
    physicsBodyDefinitions [{
      sphereCollisionRadius=0
      boundBone {boneName=dyn_hair_01_r}
      boxExtents=<15 15 15>
      constraintSetup {angularLimitsMax=<1 25 2.5>, angularLimitsMin=<-1 -2.5 -7.5>, angularTarget=<1 0 0>}
      localJointOffset=<1.5 0 0>
     }]
    relativeSpaceBone {boneName=root}
   }
   n22 = "Constraint - dfrm_chin_stud (PPP)"!(ComponentPose=@n21, Alpha=1, ConstraintWeights_0=0.32, ConstraintWeights_1:float=0.15, ConstraintWeights_2:float=0.53) -> (Pose) {
    alphaBoolBlend {blendOption=Linear}
    boneToModify {boneName=dfrm_chin_stud}
    constraintSetup [{targetBone {boneName=FACIAL_C_LipLower}}, {targetBone {boneName=FACIAL_C_Jaw}}, {targetBone {boneName=FACIAL_C_Chin}}]
    constraintWeights [1,0,1]
   }
   Output Pose(Result=Component To Local(ComponentPose=@n22)) {layerGroup=None}
  ]}
 EventGraph {nodes [
   Event BlueprintUpdateAnimation()
   > n1 = Sequence!(execute) -> (then_0, then_1, then_2:Exec)
   > Set leaning_back_alpha(leaning_back_alpha:double=GetAlphaFromReferenceTransformDelta(
     SkeletalMeshComponent=Get skeletal_mesh() -> (skeletal_mesh:Skeletal Mesh Component Object Reference)
     BoneName=spine_04
     BaseName=pelvis
     DeltaComponentType=QuaternionTwist
     InputRange=(X=-9,Y=4.25)
     TwistAxis="0,0,1"
    )) -> (Output_Get:double)
   Set twist_left_alpha(
    execute=@n1.then_1
    twist_left_alpha:double=GetAlphaFromReferenceTransformDelta(
     SkeletalMeshComponent=Get skeletal_mesh()
     BoneName=spine_05
     BaseName=pelvis
     DeltaComponentType=QuaternionTwist
     InputRange=(X=-4,Y=8.5)
     TwistAxis="1,0,0"
    )
   ) -> (Output_Get:double)
   Set spine_05_to_head_dRefz_alpha(
    execute=@n1.then_2
    spine_05_to_head_dRefz_alpha:double=GetAlphaFromReferenceTransformDelta(
     SkeletalMeshComponent=Get skeletal_mesh()
     BoneName=head
     BaseName=spine_05
     DeltaComponentType=QuaternionTwist
     InputRange=(X=55,Y=91)
     TwistAxis="0,0,1"
    )
   ) -> (Output_Get:double)
   Event BlueprintInitializeAnimation()
   > Set skeletal_mesh(skeletal_mesh:Skeletal Mesh Component Object Reference=GetOwnerSkeletalMeshComponent()) -> (Output_Get:Skeletal Mesh Component Object Reference)
  ]}
]
variables [
 leaning_back_alpha {type=number}
 skeletal_mesh {type=SkeletalMeshComponent}
 spine_05_to_head_dRefz_alpha {type=number}
 twist_left_alpha {type=number}
]
